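; Creative device component that visualizes an attenuation range with edit-only minimum and maximum distance spheres.
; Initialize stores both spheres and the show flag, then applies visibility and distances.
; Distances are scaled by 100 to centimeters, with maximum equal to minimum plus falloff.
; UpdateSphereScales sizes spheres to those distances or a hidden size, while the visibility function changes only the minimum sphere's visibility and editor-focus bounds.
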(fn Initialize
  (params (MinDistanceSphere :type Creative%20Edit%20Only%20Mesh%20Component%20Object%20Reference) (MaxDistanceSphere :type Creative%20Edit%20Only%20Mesh%20Component%20Object%20Reference) (Visible :type Boolean) (MinDistance :type Float%20%28double-precision%29) (FalloffDistance :type Float%20%28double-precision%29))
  (Variables|Default|SetMinDistanceVisualizationSphere MinDistanceSphere)
  (Variables|Default|SetMaxDistanceVisualizationSphere MaxDistanceSphere)
  (Variables|Default|SetShowSpheres Visible)
  (CallFunction|SetAttenuationVisualsVisible Visible)
  (CallFunction|SetAttenuationValues MinDistance FalloffDistance))

(fn SetAttenuationValues
  (params (MinDistance :type Float%20%28double-precision%29) (FalloffDistance :type Float%20%28double-precision%29))
  (bind _returnvalue (* FalloffDistance 1e+02))
  (bind _returnvalue_2 (* MinDistance 1e+02))
  (bind _returnvalue_3 (+ _returnvalue_2 _returnvalue))
  (bind _output_get (LocalVariables|SetMinDistanceCm _returnvalue_2))
  (if (!= _output_get (Variables|Default|GetAttenuation%20Min%20Distance))
    (Variables|Default|SetAttenuation%20Min%20Distance (LocalVariables|GetMinDistanceCm))
    (else
      (if (!= _returnvalue_3 (Variables|Default|GetAttenuation%20Max%20Distance))
        (Variables|Default|SetAttenuation%20Max%20Distance _returnvalue_3)
        (else
          _))
      (CallFunction|UpdateSphereScales))))

(fn SetAttenuationVisualsVisible
  (params (Visible :type Boolean))
  (bind _mindistancevisualizationsphere (Variables|Default|GetMinDistanceVisualizationSphere))
  (Utilities|FlowControl|Sequence
    (:then_0
      (Rendering|SetVisibility _mindistancevisualizationsphere Visible false))
    (:then_1
      (Editor|SetIgnoreBoundsforEditorFocus _mindistancevisualizationsphere (not Visible)))))

(fn UpdateSphereScales
  (bind _showspheres (Variables|Default|GetShowSpheres))
  (bind _hiddenspheresize (Variables|Default|GetHiddenSphereSize))
  (Transformation|SetWorldScale3D (Variables|Default|GetMinDistanceVisualizationSphere) (Math|Vector|SelectVector (Math|Conversions|ToVector(Float) (Variables|Default|GetAttenuation%20Min%20Distance)) _hiddenspheresize _showspheres))
  (Transformation|SetWorldScale3D (Variables|Default|GetMaxDistanceVisualizationSphere) (Math|Vector|SelectVector (Math|Conversions|ToVector(Float) (Variables|Default|GetAttenuation%20Max%20Distance)) _hiddenspheresize _showspheres)))
