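; Creative device component that sets a named dynamic option for ForTactor templates from a list of objects.
; On authority, when templates are usable, it accepts loaded objects directly or queues soft references from the parent actor.
; Queued references async-load one at a time through recursion; validated results pass through an overridable extraction function before accumulating.
; The option is set once the queue empties, with an optional debug printout of the contents.
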(event Custom|Load%20Remaining%20Soft%20Objects
  (bind _objectstoload (Variables|Default|GetObjectsToLoad))
  ;[0 0.818 1 0.204] Iterate through all objects we want to load. Once we're finalized, load them onto the ForTactorTemplates dynamic options.
  (if (Utilities|Array|IsValidIndex _objectstoload)
    (if (not (Variables|Default|GetIsLoading))
      (if (Utilities|IsValidSoftObjectReference (Utilities|Array|Get(acopy) _objectstoload))
        (Variables|Default|SetIsLoading true)
        (bind _object (Utilities|AsyncLoadAsset (Utilities|Array|Get(acopy) (Variables|Default|GetObjectsToLoad)))
          (:Completed
            (Default|LoadedObjectIsCorrect _object
              (:Yes
                (Utilities|Array|Add (Variables|Default|GetLoadedMemoryObjects) (CallFunction|GetObjectfromLoadedAsset _object))
                (Utilities|Array|RemoveIndex (Variables|Default|GetObjectsToLoad)))
              (:No
                _))
            (Variables|Default|SetIsLoading false)
            (CallFunction|LoadRemainingSoftObjects)))
        (else
          _)))
    (else
      (CallFunction|SetDynamicOption (Variables|Default|GetDynamic%20Memory%20Object%20Name) (Variables|Default|GetLoadedMemoryObjects))
      (if (Variables|Default|GetDebug)
        (Default|PrintDebug)))))

(fn GetObjectFromLoadedAsset
  (params (InputPin :type Object%20Reference))
  ;[0.2086 0.3763 1 0.25] Children May override if we actually want a sub-object here for instance.
  (return InputPin))

(fn PrintDebug
  (bind _returnvalue (Components|GetOwner))
  (bind _returnvalue_2 (Utilities|GetDisplayName _returnvalue))
  ;[1 0.1012 0.5838 0.387] If 'Debug' is set on, will display a list of all the assets that are currently loaded when they change.
  (LocalVariables|SetDebugString (Utilities|String|Append _returnvalue_2 " Dynamic Memory Objects: "))
  (for (_array_element _array_index) (Variables|Default|GetLoadedMemoryObjects)
    (LocalVariables|SetDebugString (Utilities|String|Append (LocalVariables|GetDebugString) (Utilities|String|SelectString "" ", " (== _array_index 0)) (Utilities|GetDisplayName _array_element))) ; Print a list of contents of the array.
  )
  (Development|PrintString (LocalVariables|GetDebugString) true true "(R=1.000000,G=0.000000,B=0.946138,A=1.000000)" 2e+01 (Utilities|String|StringToName _returnvalue_2)))

(fn SetDynamicMemory
  (params (Objects :type Array%20of%20Object%20References))
  (if (CallFunction|CanUseTemplates)
    (bind _output_get (Variables|Default|SetLoadedMemoryObjects Objects))
    (Networking|SwitchHasAuthority
      (:Authority
        (CallFunction|SetDynamicOption (Variables|Default|GetDynamic%20Memory%20Object%20Name) _output_get)
        (Default|PrintDebug)))))

(fn SetDynamicMemorySoft
  (params (Objects :type Array%20of%20Soft%20Object%20References))
  ;[0 1 0.5407 0.204] Called from parent actor. Sets the dynamic memory using a set of soft object references.
  (Networking|SwitchHasAuthority
    (:Authority
      (if (CallFunction|CanUseTemplates)
        (Variables|Default|SetObjectsToLoad Objects)
        (Variables|Default|SetLoadedMemoryObjects (Utilities|Array|MakeArray))
        (CallFunction|LoadRemainingSoftObjects)))))
