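; Creative device base mesh component that snaps to the ground and shows configurable base and emissive beam colors.
; Snapping line-traces down from the attach parent and clamps relative Z to min/max distances on initialization, preview moves and mesh load.
; Color and intensity setters store values and flag overrides, so the dynamic material applies them once mesh and materials load.
; Preview disables collision for snapping and refreshes materials through editor preview setup.
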(event EventBeginPlay
  ;[0 1 0.1587 0.25] Initialization
  (Utilities|IsforPreview
    (:Yes
      (Collision|SetCollisionEnabled "NoCollision") ; Preview needs to not have collision at start in order for\nsnap to ground to work .
      (CallFunction|Initialize))
    (:No
      (CallFunction|Initialize))))

(event Custom|Initialize
  ;[0 1 0.9876 0.25] When we initialize the device, attempt to snap it to the ground, and wait for the Mesh and Materials to load. If we're in preview, attempt to snap to ground each frame.
  (Utilities|IsforPreview
    (:Yes
      (CallFunction|AttemptSnapToGround)
      (Internal|CallUpdatedLocation)
      (Utilities|FlowControl|DelayUntilNextTick))
    (:No
      (CallFunction|AttemptSnapToGround)
      (Internal|CallUpdatedLocation)
      (Networking|CanExecuteCosmeticEvents
        (:True
          (EventDispatchers|BindEventtoOnMeshandMaterialsLoaded (delegate Custom|Update%20MID%20and%20Snap%20to%20Ground)))))))

(event EventSetupForEditorPreview
  (CallFunction|RefreshMaterialParams))

(event Custom|Update%20MID%20and%20Snap%20to%20Ground
  ;[0 1 0.8218 0.25] Create or Reattach the MID, then update the colors from the color variables.
  (Utilities|IsforPreview
    (:No
      (CallFunction|RefreshMaterialParams)
      ;[0.1738 0.7386 1] Try Snap to Ground again, as other objects should now be loaded
      (CallFunction|AttemptSnapToGround)
      (Internal|CallUpdatedLocation))))

(event EventUpdateEditorPreview
  (params (bHasMoved :type Boolean))
  (if bHasMoved
    (CallFunction|AttemptSnapToGround)
    (Internal|CallUpdatedLocation)))

(fn AttemptSnaptoGround
  (bind (_returnvalue_x _returnvalue_y _returnvalue_z) (Math|Vector|BreakVector (Transformation|GetWorldScale)))
  (bind (_returnvalue_x_2 _returnvalue_y_2) (Math|Vector2D|BreakVector2D (* (Variables|Snapping|GetMinMaxSnapDistance) "(X=-1.000000,Y=-1.000000)")))
  (bind (_minmaxsnapdistance_x _minmaxsnapdistance_y) (Math|Vector2D|BreakVector2D (Variables|GetMinMaxSnapDistance_X)))
  (bind (_minmaxsnapdistance_x_2 _minmaxsnapdistance_y_2) (Math|Vector2D|BreakVector2D (Variables|GetMinMaxSnapDistance_X)))
  (if (Variables|Snapping|GetSnapToGround)
    (if (Utilities|Operators|NotEqual(!=) (Variables|Transform|GetRelativeLocation) (Transformation|GetWorldLocation) 0)
      ; LineTrace
      (bind (_outhit _returnvalue) (Collision|LineTraceByChannel (Transformation|GetWorldLocation (Transformation|GetAttachParent)) (Utilities|Operators|Subtract (Transformation|GetWorldLocation (Transformation|GetAttachParent)) (Math|Vector|MakeVector 0.0 0.0 (* _minmaxsnapdistance_y_2 _returnvalue_z))) "TraceTypeQuery1" false (Utilities|Array|MakeArray (Components|GetOwner))))
      (bind (_outhit_bblockinghit _outhit_binitialoverlap _outhit_time) (Collision|BreakHitResult _outhit))
      (Transformation|SetRelativeLocation 0 (Math|Vector|MakeVector 0.0 0.0 (Math|Float|SelectFloat (Math|Float|Min(Float) (Utilities|Operators|Multiply _outhit_time _minmaxsnapdistance_y -1.0) _returnvalue_x_2) _returnvalue_y_2 _returnvalue)) false false))))

(fn RefreshMaterialParams
  (bind _mid (Variables|Internal|GetMID))
  (Variables|StaticMesh|GetStaticMesh
    (:then
      (Variables|Internal|GetMID
        (:then
          _)
        (:else
          (if (Variables|Internal|GetbParameterOverrides)
            _)))
      (Material|ReattachOrCreateMID self (Variables|Internal|GetMID))
      (Rendering|Material|SetVectorParameterValue _mid "Base Color" (Variables|Color|GetBaseColor) :!class /Script/Engine.MaterialInstanceDynamic)
      (Rendering|Material|SetVectorParameterValue _mid "Beam Color" (* (Variables|Color|GetEmissiveColor) (Variables|Color|GetEmissive%20Multiplier)) :!class /Script/Engine.MaterialInstanceDynamic))
    (:else)))

(fn SetColor
  (params (BaseColor :type Linear%20Color%20Structure) (EmissiveColor :type Linear%20Color%20Structure))
  ;[1 0.6864 0 0.25] Updates the Colors of the Base and attempts to update MID (if it fails due to the mesh / materials not being loaded yet, the variables will be used later)
  (Variables|Color|SetBaseColor BaseColor)
  (Variables|Color|SetEmissiveColor EmissiveColor)
  (Variables|Internal|SetbParameterOverrides true)
  (CallFunction|UpdateMIDAndSnapToGround))

(fn SetIntensity
  (params (EmissiveIntensity :type Float%20%28double-precision%29))
  (Variables|Color|SetEmissive%20Multiplier EmissiveIntensity)
  (Variables|Internal|SetbParameterOverrides true)
  (CallFunction|UpdateMIDAndSnapToGround))
