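; Edit-only beam component for Creative devices that draws a colored beam whose length follows the height of its parent base mesh.
; On initialization or editor preview, it binds to the parent CreativeEditOnlyBase's location updates.
; UpdateScale sets Z scale from the parent's negated relative Z minus the base's minimum snap distance, hiding the beam when that length is not positive.
; A dynamic material instance receives beam color times emissive multiplier, overridable through SetColor and SetIntensity.
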(event EventBeginPlay
  ;[0.06492 1 0 0.25] When Begin Play starts, if we're in edit mode then initialize. If we're not, wait till we are then initialize (in a case where we're in Create but we're the client joining in progress).
  (CallFunction|Initialize) ; Our parent isn't the shared parent. Just initialize
)

(event Custom|Initialize
  ;[0 0.8734 1 0.25] Initialization update scale (e.g. length of beam) and listens for the parent updating location. Also waits for Mesh/Materials Loaded
  (Networking|CanExecuteCosmeticEvents
    (:True
      (CallFunction|BindtoParentBaseMesh)
      (CallFunction|UpdateScale)
      (EventDispatchers|BindEventtoOnMeshandMaterialsLoaded (delegate Custom|Update%20MID)))))

(event Custom|Update%20MID
  ;[0 0.8734 1 0.25] Create or Reattach MID, then update color from color variable and emissive multiplier
  (Utilities|IsforPreview
    (:No
      (CallFunction|RefreshMaterialParams))))

(event EventSetupForEditorPreview
  (CallFunction|BindtoParentBaseMesh)
  (CallFunction|RefreshMaterialParams))

(event EventUpdateEditorPreview
  (params (bHasMoved :type Boolean))
  (CallFunction|UpdateScale))

(event Custom|Parent%20Base%20Mesh%20Updated%20Location
  ;[0.8058 0.02494 1] When base updates location, also update scale.
  (CallFunction|UpdateScale))

(fn BindToParentBaseMesh
  (bind _outputdelegate (EventDispatchers|CreateEvent :!event Parent%20Base%20Mesh%20Updated%20Location))
  (bind _ascreative_edit_only_base (Utilities|Casting|CastToCreativeEditOnlyBase (Transformation|GetAttachParent)))
  (bind _output_get (Variables|Default|SetBase _ascreative_edit_only_base))
  (Internal|UnbindEventfromUpdatedLocation _output_get _outputdelegate)
  (Internal|BindEventtoUpdatedLocation _output_get _outputdelegate))

(fn RefreshMaterialParams
  (Variables|StaticMesh|GetStaticMesh
    (:then
      (Variables|Internal|GetMID
        (:then
          _)
        (:else
          (if (Variables|Internal|GetbParameterOverrides)
            _)))
      (Material|ReattachOrCreateMID self (Variables|Internal|GetMID))
      (Rendering|Material|SetVectorParameterValue (Variables|Internal|GetMID) "Beam Color" (* (Variables|Default|GetBeamColor) (Variables|Default|GetEmissive%20Multiplier)) :!class /Script/Engine.MaterialInstanceDynamic))
    (:else)))

(fn SetColor
  (params (Color :type Linear%20Color%20Structure))
  (Variables|Default|SetBeamColor Color)
  (Variables|Internal|SetbParameterOverrides true)
  (CallFunction|UpdateMID))

(fn SetIntensity
  (params (EmissiveIntensity :type Float%20%28double-precision%29))
  (Variables|Default|SetEmissive%20Multiplier EmissiveIntensity)
  (Variables|Internal|SetbParameterOverrides true)
  (CallFunction|UpdateMID))

(fn UpdateScale
  (bind _returnvalue (Transformation|GetAttachParent))
  (bind _gap (LocalVariables|GetGap))
  (bind (_relativelocation_x _relativelocation_y _relativelocation_z) (Math|Vector|BreakVector (Variables|GetRelativeLocation_X _returnvalue)))
  (bind _returnvalue_2 (* _relativelocation_z -1.0))
  (bind _returnvalue_3 (- _returnvalue_2 _gap))
  (bind _base (Variables|Default|GetBase)
    (:then
      (bind _minmaxsnapdistance_x (Math|Vector2D|BreakVector2D (Variables|/Game/Creative/Devices/Common/Components/CreativeEditOnlyBase.CreativeEditOnlyBase_C|GetMinMaxSnapDistance_X _base)))
      (LocalVariables|SetGap _minmaxsnapdistance_x))
    (:else
      _))
  (if (> _returnvalue_3 0)
    (if (not (Variables|Rendering|GetbHiddenInGame (Transformation|GetAttachParent)))
      (Transformation|SetRelativeScale3D 0 (Math|Vector|MakeVector 1.0 1.0 _returnvalue_3))
      (Development|SetHiddeninGame false))
    (else
      (Development|SetHiddeninGame true))))
