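; Creative device component that registers objects under integer indices and notifies listeners when they are added or removed.
; An object-to-index map moves re-registered objects out of their old index, and registered actors are dropped on End Play.
; Changes fire immediate per-object delegates, then are cached and broadcast per index after a retriggerable one-frame delay.
; Next and previous queries search for neighboring indices that contain objects.
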(event Custom|Register%20Object
  (params (Object%20to%20Register :type Object%20Reference) (Register%20Index :type Integer))
  ;[0 1 0.5195 0.25] When an object is registered, store it in our list then call a delegate to let anyone listening that a new item was added.
  (CallFunction|UpdateExistingRegister Object%20to%20Register Register%20Index)
  (bind _returnvalue (CallFunction|AddObjecttoIndex (Variables|Default|GetObjectList) Object%20to%20Register Register%20Index))
  (if _returnvalue
    (Default|CallOnObjectAdded Object%20to%20Register Register%20Index)
    (CallFunction|AddObjecttoIndex (Variables|Default|GetObjectsAddedCache) Object%20to%20Register Register%20Index)
    (Private|BindDelegates Object%20to%20Register true)
    (CallFunction|PrintRegister (Utilities|String|Append "Added " (Utilities|GetDisplayName Object%20to%20Register)))
    (Utilities|FlowControl|RetriggerableDelay 0.0) ; Wait a frame to pool all objects added..
    (Private|BroadcastObjectsAdded)))

(event Custom|Remove%20Object%20From%20Index
  (params (Object%20to%20Remove :type Object%20Reference) (Register%20Index :type Integer))
  ;[1 0.127 0 0.25] When an object is removed, store it in our list, then call a delegate to let anyone listening that an object was removed.
  (bind (_wasremoved _removedindex) (CallFunction|RemoveObjectfromIndex (Variables|Default|GetObjectList) Object%20to%20Remove Register%20Index))
  (if _wasremoved
    (Default|CallOnObjectRemoved Object%20to%20Remove _removedindex)
    (CallFunction|AddObjecttoIndex (Variables|Default|GetObjectsRemovedCache) Object%20to%20Remove Register%20Index)
    (Private|BindDelegates Object%20to%20Remove)
    (Utilities|FlowControl|RetriggerableDelay 0.0) ; Wait a frame to pool all objects added..
    (Private|BroacastObjectsRemoved)))

(fn BindDelegates
  (params (Object :type Object%20Reference) (Bind :type Boolean))
  (bind _outputdelegate (EventDispatchers|CreateEvent :!event Tracked%20Actor%20Ended%20Play))
  ;[0 0.7912 1 0.5] If we're an actor, bind to End Play so that we ensure that destroyed objects are removed from this list.
  (bind _asactor (Utilities|Casting|CastToActor Object))
  (if Bind
    (Game|BindEventtoOnEndPlay _asactor _outputdelegate)
    (else
      (Game|UnbindEventfromOnEndPlay _asactor _outputdelegate))))

(fn BroacastObjectsRemoved
  (bind _objectsremovedcache (Variables|Default|GetObjectsRemovedCache))
  (for (_array_element _array_index) _objectsremovedcache
    (bind _array_element_objects (Utilities|Struct|BreakFortObjectArray _array_element))
    (Default|CallOnObjectsRemoved _array_element_objects _array_index))
  (Utilities|Array|Clear _objectsremovedcache))

(fn BroadcastObjectsAdded
  (bind _objectsaddedcache (Variables|Default|GetObjectsAddedCache))
  (for (_array_element _array_index) _objectsaddedcache
    (bind _array_element_objects (Utilities|Struct|BreakFortObjectArray _array_element))
    (Default|CallOnObjectsAdded _array_element_objects _array_index) ; Note states that values won't be passed by ref. \nThey do get transmitted fine though as a copy.
  )
  (Utilities|Array|Clear _objectsaddedcache))

(fn GetNextRegisteredIndex
  (params (Reference%20Index :type Integer :default -1))
  (bind _objectlist (Variables|Default|GetObjectList))
  (bind _returnvalue (Utilities|Array|LastIndex _objectlist))
  (bind _output_objects (Utilities|Struct|BreakFortObjectArray (Utilities|Array|Get(aref) (Variables|Default|GetObjectList) (LocalVariables|GetCheck%20Index))))
  (bind _output_objects_2 (Utilities|Struct|BreakFortObjectArray (Utilities|Array|Get(aref) (Variables|Default|GetObjectList) (LocalVariables|GetCheck%20Index))))
  ;[0 0.7912 1 0.5] Given an input index, return the NEXT index in the object list that contains objects.
  (if (< Reference%20Index _returnvalue)
    (for _index (range (+ (Variables|GetReference%20Index) 1) (+ _returnvalue 1))
      (LocalVariables|SetCheck%20Index _index)
      (if (!= (Utilities|Array|Length _output_objects) 0)
        (return true (LocalVariables|GetCheck%20Index) _output_objects_2)))
    (else
      _))
  (return false -1 (Utilities|Array|MakeArray 0)))

(fn GetPrevRegisteredIndex
  (params (Reference%20Index :type Integer :default -1))
  (bind _output_objects (Utilities|Struct|BreakFortObjectArray (Utilities|Array|Get(aref) (Variables|Default|GetObjectList) (LocalVariables|GetCheck%20Index))))
  (bind _output_objects_2 (Utilities|Struct|BreakFortObjectArray (Utilities|Array|Get(aref) (Variables|Default|GetObjectList) (LocalVariables|GetCheck%20Index))))
  ;[0 0.7912 1 0.5] Given an input index, return the PREVIOUS index in the object list that contains objects.
  (if (> Reference%20Index 0)
    (LocalVariables|SetCheck%20Index (Math|Integer|Min(Integer) (Utilities|Array|LastIndex (Variables|Default|GetObjectList)) (- (Variables|GetReference%20Index) 1)))
    (while (> (LocalVariables|GetCheck%20Index) 0)
      (if (!= (Utilities|Array|Length _output_objects_2) 0)
        (return true (LocalVariables|GetCheck%20Index) _output_objects)
        (else
          (Math|Integer|DecrementInt (LocalVariables|GetCheck%20Index)))))
    (else
      _))
  (return false -1 (Utilities|Array|MakeArray 0)))

(fn GetRegisterTag
  ;[1 1 1] Simple getter as the tag is private (to avoid writing) but we want it to be queryable
  (return (Variables|Default|GetRegisterTag)))

(fn TrackedActorEndedPlay
  (params (Actor :type Actor%20Object%20Reference) (EndPlayReason :type EEndPlayReason%20Enum :default "Destroyed"))
  ;[0.3385 0 0.005961 0.5] A tracked actor has ended play. Update Existing Register (which will remove it from the list as we move it to an invalid index
  (CallFunction|UpdateExistingRegister Actor -1))

(fn UpdateExistingRegister
  (params (Object :type Object%20Reference) (Index :type Integer))
  (bind _objectindexmap (Variables|Default|GetObjectIndexMap))
  (bind (_value _returnvalue) (Utilities|Map|Find _objectindexmap Object))
  (bind _objectindexmap_2 (Variables|Default|GetObjectIndexMap))
  (bind _object (Variables|GetObject))
  (Utilities|FlowControl|Sequence
    (:then_0
      (if _returnvalue
        (CallFunction|RemoveObjectfromIndex (Variables|GetObject) _value)))
    (:then_1
      (if (>= Index 0)
        (Utilities|Map|Add _objectindexmap_2 _object (Variables|GetIndex))
        (else
          (Utilities|Map|Remove _objectindexmap_2 _object))))))
