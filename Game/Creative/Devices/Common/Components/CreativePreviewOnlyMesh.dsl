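; Static mesh component for Creative devices that shows a mesh only on the preview version of the device.
; On BeginPlay it destroys itself if it fails the preview validity check.
; Otherwise it async-loads soft-referenced override materials by index into a cache, then async-loads the preview mesh.
; Once the mesh is set, it applies the cached materials by slot and calls OnMeshLoaded.
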(event EventBeginPlay
  ;[0 1 0.2528 0.25] Init
  (CallFunction|LoadMeshAndMaterial))

(event Custom|Load%20Mesh%20and%20Material
  (bind _cached_material_index (Variables|Internal|GetCached%20Material%20Index))
  (bind _preview_override_materials (Variables|Default|GetPreview%20Override%20Materials))
  (bind _output (Utilities|Array|Get(acopy) _preview_override_materials _cached_material_index))
  (bind _cached_material_index_2 (Variables|Internal|GetCached%20Material%20Index))
  ;[0 0.8742 1 0.25] Async loads the mesh, and material (if set). Only does anything if the owner is a buildingprop that is the preview version of the device.
  (Utilities|IsValidforPreview
    (:True
      (if (Utilities|Array|IsValidIndex (Variables|Default|GetPreview%20Override%20Materials))
        (Variables|Internal|SetCached%20Material%20Index 0)
        (Utilities|Array|Clear (Variables|Internal|GetCached%20Materials))
        (if (<= (Variables|Internal|GetCached%20Material%20Index) (Utilities|Array|LastIndex (Variables|Default|GetPreview%20Override%20Materials)))
          (if (Utilities|IsValidSoftObjectReference _output)
            (bind _object (Utilities|AsyncLoadAsset _output)
              (:Completed
                (bind _asmaterial_interface (Utilities|Casting|CastToMaterialInterface _object)
                  (:then
                    (Utilities|Array|Insert (Variables|Internal|GetCached%20Materials) _asmaterial_interface _cached_material_index_2)
                    (Math|Integer|IncrementInt _cached_material_index_2
                      (:%20%20
                        _)))
                  (:CastFailed
                    _))))))
        (else
          _))
      (CallFunction|LoadMesh) ; No more materials to load, load the mesh.
    )
    (:False
      (Components|DestroyComponent))) ; Are we the preview version of the actor\nand on the client?
)

(event Custom|Load%20Mesh
  ;[0 0.8742 1 0.25] Soft Loading Mesh, then apply mesh and materials (if set)
  (bind _preview_mesh (Variables|Default|GetPreview%20Mesh)
    (:then
      (bind _object (Utilities|AsyncLoadAsset _preview_mesh)
        (:Completed
          (bind _asstatic_mesh (Utilities|Casting|CastToStaticMesh _object))
          (bind _returnvalue (Components|StaticMesh|SetStaticMesh _asstatic_mesh))
          (if _returnvalue
            (for (_array_element _array_index) (Variables|Internal|GetCached%20Materials)
              (Rendering|Material|SetMaterial _array_index _array_element))
            (Default|CallOnMeshLoaded)))))
    (:else)))
