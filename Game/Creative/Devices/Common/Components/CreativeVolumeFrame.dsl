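; Static mesh component that draws the visual frame of a Creative device's volume for a chosen shape.
; Auto or manual initialization looks up soft edit and preview meshes and materials in per-shape maps, then reloads them.
; On load, UpdateFrameSize sizes the frame from zone scale, per-axis multipliers, shape offset and mesh-to-meters divisor.
; Some shapes scale through material WPO via SizeBias, avoiding stretched detail, with bounds enlarged to match.
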(event EventBeginPlay
  (EventDispatchers|BindEventtoOnMeshandMaterialsLoaded (delegate Custom|On%20Mesh%20Loaded))
  ;[0 1 0.2776 0.25] Auto Initialize does the same as 'Initialize Component' without overriding variables, but forces the update
  ; so that it will always be set (normally update mesh only sets if the input shape is different from stored state)
  (if (Variables|Default|GetAutoInitialize)
    (bind _reload (CallFunction|UpdateMeshFromShape (Variables|Default|GetShape) true))
    (if _reload
      (CallFunction|ReloadMeshandMaterials))))

(event Custom|On%20Mesh%20Loaded
  ;[1 0.7548 0 0.25] Whenever we load a new mesh, we want to update the frame size.
  (CallFunction|UpdateFrameSize))

(event Custom|Initialize%20Component
  (params (FrameShape :type EGameplayVolumeShape%20Enum :default "NewEnumerator0") (ZoneScale :type Float%20%28double-precision%29 :default 1.0) (ZoneScaleMultipliers :type Vector :default "1.000000,1.000000,1.000000"))
  ;[0 1 0.2776 0.25] Init - Called when initializing manually. Will update variables then updates the mesh values from the shape input before reloading (if we have a new valid shape)
  (Variables|ZoneDetails|SetZoneScale ZoneScale)
  (Variables|ZoneDetails|SetZoneScaleMultipliers ZoneScaleMultipliers)
  (bind _reload (CallFunction|UpdateMeshFromShape FrameShape))
  (if _reload
    (if (Editor|InEditor)
      (EventDispatchers|BindEventtoOnMeshandMaterialsLoaded (delegate Custom|On%20Mesh%20Loaded)) ; Editor doesn't begin play, so must bind here
      (else
        _))
    (CallFunction|ReloadMeshandMaterials)))

(fn GetColorParam
  (bind _shape (Variables|Default|GetShape))
  (bind _ispreview (LocalVariables|GetIsPreview))
  (bind _shapepreviewmeshandmats (Variables|Default|GetShapePreviewMeshandMats))
  (bind _shapeeditmeshandmats (Variables|Default|GetShapeEditMeshAndMats))
  (bind _returnvalue (select _ispreview _shapepreviewmeshandmats _shapeeditmeshandmats))
  (bind (_value _returnvalue_2) (Utilities|Map|Find _returnvalue _shape))
  (bind (_value_mesh_2_ba91cdb64dfb95237cec639088ecafeb _value_materialoverrides_6_0ba09ebe42205d40584d3abeb77d5bdd _value_scaletometers_10_a92c273a46f5975d82a56fbc329ee84e _value_scaleinmaterial_12_3a38e9db40fb218347be679d9e425a06 _value_colorparam_15_9a09a70d429488a817bc529c1d48b7e2) (Utilities|Struct|BreakCreativeMeshDetails _value))
  ;[1 0.9002 0 0.25] Different Materials have different color parameters. This gets the correct color parameter from our 'Mesh and Mats' maps.
  (bind _asbuilding_static_mesh (Utilities|Casting|CastToBuildingSMActor (Components|GetOwner))
    (:then
      (LocalVariables|SetIsPreview (Building|IsPreviewBuildingActor _asbuilding_static_mesh)))
    (:CastFailed
      _))
  (if _returnvalue_2
    (bind _output_get (LocalVariables|SetColorParam _value_colorparam_15_9a09a70d429488a817bc529c1d48b7e2))
    (return (!= _output_get "") _output_get)
    (else
      (return false "None"))))

(fn UpdateFrameSize
  (bind _returnvalue (!= (.y (Variables|ZoneDetails|GetZoneScaleMultipliers)) 0))
  (bind (_returnvalue_x _returnvalue_y _returnvalue_z) (Math|Vector|BreakVector (Transformation|GetWorldScale)))
  (bind _localscale (LocalVariables|GetlocalScale))
  (bind _returnvalue_2 (* (.y (Variables|ZoneDetails|GetZoneScaleMultipliers)) _localscale))
  (bind _returnvalue_3 (Math|Float|SelectFloat _returnvalue_2 _returnvalue_y _returnvalue))
  (bind _shape (Variables|Default|GetShape))
  (bind _ispreview (LocalVariables|GetIsPreview))
  (bind _shapepreviewmeshandmats (Variables|Default|GetShapePreviewMeshandMats))
  (bind _shapeeditmeshandmats (Variables|Default|GetShapeEditMeshAndMats))
  (bind _returnvalue_4 (select _ispreview _shapepreviewmeshandmats _shapeeditmeshandmats))
  (bind (_value _returnvalue_5) (Utilities|Map|Find _returnvalue_4 _shape))
  (bind (_value_mesh_2_ba91cdb64dfb95237cec639088ecafeb _value_materialoverrides_6_0ba09ebe42205d40584d3abeb77d5bdd _value_scaletometers_10_a92c273a46f5975d82a56fbc329ee84e _value_scaleinmaterial_12_3a38e9db40fb218347be679d9e425a06) (Utilities|Struct|BreakCreativeMeshDetails _value))
  (bind _zonescalemultipliers (Variables|ZoneDetails|GetZoneScaleMultipliers))
  (bind _x (.x _zonescalemultipliers))
  (bind _y (.y _zonescalemultipliers))
  (bind _z (.z _zonescalemultipliers))
  (bind _returnvalue_6 (!= _z 0))
  (bind _localscale_2 (LocalVariables|GetlocalScale))
  (bind _returnvalue_7 (* _localscale_2 _z))
  (bind _returnvalue_8 (Math|Float|SelectFloat _returnvalue_7 _returnvalue_z _returnvalue_6))
  (bind _returnvalue_9 (!= _x 0))
  (bind _localscale_3 (LocalVariables|GetlocalScale))
  (bind _returnvalue_10 (* _localscale_3 _x))
  (bind _returnvalue_11 (Math|Float|SelectFloat _returnvalue_10 _returnvalue_x _returnvalue_9))
  (bind _returnvalue_12 (Math|Vector|MakeVector _returnvalue_11 _returnvalue_3 _returnvalue_8))
  (bind _scale_divisor (LocalVariables|GetScale%20Divisor))
  (bind _returnvalue_13 (/ _returnvalue_12 _scale_divisor))
  ;[1 0.9002 0 0.25] Sets Frame Size based on scale multipliers.
  (LocalVariables|SetlocalScale (+ (Utilities|Map|Find (Variables|Default|GetShapeFrameMultiplierOffset) (Variables|Default|GetShape)) (Variables|ZoneDetails|GetZoneScale)))
  (Utilities|FlowControl|Sequence
    (:then_0
      ;;[1 0.9002 0 0.25] This component's logic assumes everything is 1m in size. This bit gets the scale that can be divided by the mesh size to ensure that's the case.
      (bind _asbuilding_static_mesh (Utilities|Casting|CastToBuildingSMActor (Components|GetOwner))
        (:then
          (LocalVariables|SetIsPreview (Building|IsPreviewBuildingActor _asbuilding_static_mesh)))
        (:CastFailed
          _))
      (if _returnvalue_5
        (LocalVariables|SetScale%20Divisor _value_scaletometers_10_a92c273a46f5975d82a56fbc329ee84e)
        (else
          (LocalVariables|SetScale%20Divisor "1.000000,1.000000,1.000000"))))
    (:then_1
      ;;[1 0.9002 0 0.25] If any Zone Scale Multiplier is 0, ignore it for scaling and just use world scale.
      (if (not _value_scaleinmaterial_12_3a38e9db40fb218347be679d9e425a06)
        (Transformation|SetRelativeScale3D _returnvalue_13)
        (Rendering|SetBoundsScale 1.0)
        (else
          (UpdateFrameSize|UpdateMaterialScale _returnvalue_13))))))

(fn UpdateMaterialScale
  (params (Scale :type Vector))
  (bind (_scale_x _scale_y _scale_z) (Math|Vector|BreakVector (Variables|GetScale_X)))
  ;[1 0.9002 0 0.25] Some frame materials scale using their Material WPO rather than actually scaling the mesh. This is useful because a) it means any detail on the material doesn't get stretched, b) the size of the frame remains consistent regardless of scale. This function sets the SizeBias value which does this, along with setting world scale to 1 (as actual scaling is done in the material)
  (Transformation|SetWorldScale3D "1.000000,1.000000,1.000000")
  (Rendering|Material|SetVectorParameterValue (Material|GetMID self) "SizeBias" (Math|Conversions|ToLinearColor(Vector) (- (Math|Vector|vector*vector (Variables|Internal|GetMaterialScalar) Scale (Transformation|GetActorScale3D (Components|GetOwner))) (Variables|Internal|GetMaterialScalar))) :!class /Script/Engine.MaterialInstanceDynamic)
  (Rendering|SetBoundsScale (* (Math|Float|Max(Float) _scale_x _scale_y _scale_z) 1.5)))

(fn UpdateMeshfromShape
  (params (New%20Shape :type EGameplayVolumeShape%20Enum :default "NewEnumerator0") (Force%20Update :type Boolean))
  (bind _shape (Variables|Default|GetShape))
  (bind _shapeeditmeshandmats (Variables|Default|GetShapeEditMeshAndMats))
  (bind (_value _returnvalue) (Utilities|Map|Find _shapeeditmeshandmats _shape))
  (bind (_localdetails_mesh_2_ba91cdb64dfb95237cec639088ecafeb _localdetails_materialoverrides_6_0ba09ebe42205d40584d3abeb77d5bdd) (Utilities|Struct|BreakCreativeMeshDetails (Variables|GetLocalDetails_Mesh_2_BA91CDB64DFB95237CEC639088ECAFEB)))
  (bind (_localdetails_mesh_2_ba91cdb64dfb95237cec639088ecafeb_2 _localdetails_materialoverrides_6_0ba09ebe42205d40584d3abeb77d5bdd_2) (Utilities|Struct|BreakCreativeMeshDetails (Variables|GetLocalDetails_Mesh_2_BA91CDB64DFB95237CEC639088ECAFEB)))
  ;[1 0.9002 0 0.25] Will update the Mesh from the input shape, setting the mesh/materials to soft-load from the ShapeMeshAndMats Maps.
  ; Returns true if a new mesh has been set.
  (if (or Force%20Update (and (Utilities|Enum|NotEqual(Enum) New%20Shape (Variables|Default|GetShape)) (Utilities|Enum|NotEqual(Enum) New%20Shape "NewEnumerator0")))
    (Variables|Default|SetShape (Variables|GetNew%20Shape))
    (if _returnvalue
      (LocalVariables|SetLocalDetails _value)
      (LocalVariables|SetMesh (select (Utilities|IsValidSoftObjectReference _localdetails_mesh_2_ba91cdb64dfb95237cec639088ecafeb_2) _localdetails_mesh_2_ba91cdb64dfb95237cec639088ecafeb_2 (CallFunction|GetEditMesh)))
      (LocalVariables|SetLocalMats (select (Utilities|Array|IsValidIndex _localdetails_materialoverrides_6_0ba09ebe42205d40584d3abeb77d5bdd_2) _localdetails_materialoverrides_6_0ba09ebe42205d40584d3abeb77d5bdd_2 (CallFunction|GetEditOverrideMaterials)))
      (CallFunction|SetEditMeshandMaterials (LocalVariables|GetMesh) (LocalVariables|GetLocalMats) true)
      (LocalVariables|SetLocalDetails (Utilities|Map|Find (Variables|Default|GetShapePreviewMeshandMats) (Variables|Default|GetShape)))
      (LocalVariables|SetMesh (select (Utilities|IsValidSoftObjectReference _localdetails_mesh_2_ba91cdb64dfb95237cec639088ecafeb) _localdetails_mesh_2_ba91cdb64dfb95237cec639088ecafeb (CallFunction|GetPreviewOverrideMesh)))
      (LocalVariables|SetLocalMats (select (Utilities|Array|IsValidIndex _localdetails_materialoverrides_6_0ba09ebe42205d40584d3abeb77d5bdd) _localdetails_materialoverrides_6_0ba09ebe42205d40584d3abeb77d5bdd (CallFunction|GetPreviewOverrideMaterials)))
      (CallFunction|SetPreviewMeshandMaterials (LocalVariables|GetMesh) (LocalVariables|GetLocalMats) true)
      (return true)
      (else
        (Components|StaticMesh|SetStaticMesh self 0)
        (return false)))))
