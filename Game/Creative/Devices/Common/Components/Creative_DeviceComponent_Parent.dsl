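; Creative_DeviceComponent_Parent is a base component for Creative devices that hooks into the lifecycle of its owning device.
; On BeginPlay it casts the owner to Creative_Device_Prop_Parent and binds to that device's OnDeviceInitialized and OnDeviceActivated dispatchers.
; The bound handlers, Owning Actor Initialized and Owning Actor Activated, have empty bodies here.
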(event EventBeginPlay
  (bind _ascreative_device_prop_parent (Utilities|Casting|CastToCreative_Device_Prop_Parent (Components|GetOwner)))
  (Default|BindEventtoOnDeviceInitialized _ascreative_device_prop_parent (EventDispatchers|CreateEvent :!event Owning%20Actor%20Initialized))
  (Default|BindEventtoOnDeviceActivated _ascreative_device_prop_parent (EventDispatchers|CreateEvent :!event Owning%20Actor%20Activated)))

(fn OwningActorActivated
  (return))

(fn OwningActorInitialized
  (return))
