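; Creative device component that controls whether its owner is enabled, based on minigame phase and play/edit mode.
; Initialization binds to the minigame's state and play-mode delegates, caching it for unbinding, or simulates pre-game edit mode in editor.
; Authority filters requests through preview/edit-mode checks, then sets a replicated state struct, flushing net dormancy to push RepNotify.
; The RepNotify broadcasts state changes and runs reset-delay timers, which re-enable on authority.
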(event Custom|Initialize
  (params (Logic%20Component :type Fort%20Minigame%20Logic%20Component%20Object%20Reference) (Enabled%20Phases :type EEnabledOnPhase%20Enum :default "NewEnumerator1"))
  ;[0 1 0.3923 0.25] New Initialize function that converts Enabled Phases to take a new enum (which is index shifted by one). The old function will eventually be deprecated once all devices that use enabled components are converted.
  (CallFunction|InitializeComponent Logic%20Component (- (Math|Conversions|ToInteger(Byte) Enabled%20Phases) 1)))

(event Custom|Initialize%20Component
  (params (Logic%20Component :type Fort%20Minigame%20Logic%20Component%20Object%20Reference) (Enabled%20Index :type Integer))
  (bind _currentminigame (Variables|/Script/FortniteGame.FortMinigameLogicComponent|GetCurrentMinigame Logic%20Component))
  ;[0 0.7396 1 0.1] Initialization - Bind events to switch on Play Mode / Game State
  (Utilities|IsValid Logic%20Component
    (:Is%20Valid
      (Variables|Default|SetLogic%20Component Logic%20Component)
      (Utilities|SwitchHasAuthority
        (:Authority
          (Variables|Default|SetEnabled%20Index Enabled%20Index)
          (if (Editor|InEditor)
            (CallFunction|MinigameStateChanged 0) ; If we're the editor instance (e.g. UEFN) pretend\nwe're in Pre-Game.
            (CallFunction|MinigamePlayModeChanged 0) ; If we're the editor instance (e.g. UEFN) run as if we're not in play mode\n(which we aren't).
            (else
              (Utilities|IsValid _currentminigame
                (:Is%20Valid
                  (CallFunction|OnAddedToMinigame _currentminigame))
                (:Is%20Not%20Valid
                  (Minigame|BindEventtoOnAddedtoMinigame Logic%20Component (delegate Custom|On%20Added%20to%20Minigame)))))))))
    (:Is%20Not%20Valid
      (Development|PrintString "Enabled Component: Initialized without a valid logic component." true true "(R=1.000000,G=0.000000,B=0.000000,A=1.000000)" 1e+01))))

(event Custom|On%20Added%20to%20Minigame
  (params (Minigame :type Fort%20Minigame%20Object%20Reference))
  ;[0 0.7396 1 0.1] Initialization - Bind events to switch on Play Mode / Game State
  (CallFunction|MinigameStateChanged Minigame (Variables|/Script/FortniteGame.FortMinigame|GetCurrentState Minigame))
  ;;[0 0.8095 1 0.215] Binding to New Minigame / Unbinding from Old
  (if (not (CallFunction|IsMinigametheSameasCachedMinigame Minigame))
    (CallFunction|UnbindandResetCachedMinigame)
    (Minigame|BindEventtoOnPlayModeChangedDelegate Minigame (EventDispatchers|CreateEvent :!event Custom|Minigame%20Play%20Mode%20Changed))
    (Minigame|BindEventtoStateChangedDelegate Minigame (EventDispatchers|CreateEvent :!event Custom|Minigame%20State%20Changed))
    (Variables|Default|SetCachedMinigame (Utilities|ResolveSoftReference Minigame))))

(event Custom|Minigame%20State%20Changed
  (params (Minigame :type Fort%20Minigame%20Object%20Reference) (MinigameState :type EFortMinigameState%20Enum :default "PreGame"))
  ;[0.01178 0.2119 1] State Changes.
  (CallFunction|SetEnabledOnPhase MinigameState (Variables|Default|GetEnabled%20Index)))

(event Custom|Minigame%20Play%20Mode%20Changed
  (params (Minigame :type Fort%20Minigame%20Object%20Reference) (bIsInPlayMode :type Boolean))
  ;[0.01178 0.2119 1] State Changes.
  (CallFunction|SetEnabledOnPlayMode bIsInPlayMode)
  (if (Variables|PlayModeSwitching|GetCheck%20Phase%20Change%20During%20Play%20Switch)
    (bind _logic_component (Variables|Default|GetLogic%20Component)
      (:then
        (bind _currentminigame (Variables|/Script/FortniteGame.FortMinigameLogicComponent|GetCurrentMinigame _logic_component)
          (:then
            (CallFunction|MinigameStateChanged _currentminigame (Variables|/Script/FortniteGame.FortMinigame|GetCurrentState _currentminigame)))
          (:else)))
      (:else))))

(event Custom|UnbindAndResetCachedMinigame
  ;[0.1962 0 0.004858] Destruction
  (if (Utilities|IsValidSoftObjectReference (Variables|Default|GetCachedMinigame))
    (CallFunction|UnbindMinigameEvents (Utilities|ResolveSoftReference (Variables|Default|GetCachedMinigame)))
    (Variables|Default|SetCachedMinigame 0)))

(event EventEndPlay
  (params (EndPlayReason :type EEndPlayReason%20Enum :default "Destroyed"))
  ;[0.1962 0 0.004858] Destruction
  (|CallParentFunction EndPlayReason)
  (CallFunction|UnbindandResetCachedMinigame))

(fn CanbeEnabled
  (bind _asbuilding_static_mesh (Utilities|Casting|CastToBuildingSMActor (Components|GetOwner))
    (:then
      (if (Building|IsPreviewBuildingActor _asbuilding_static_mesh)
        (if (not (Variables|PlayModeSwitching|GetCanBeEnabledDuringPreview))
          (return false))
        (else
          _))
      (bind _logic_component (Variables|Default|GetLogic%20Component)
        (:then
          (bind _currentminigame (Variables|/Script/FortniteGame.FortMinigameLogicComponent|GetCurrentMinigame _logic_component)
            (:then
              (if (Minigame|IsinPlayMode _currentminigame :!class /Script/FortniteGame.FortMinigame)
                (return true)
                (else
                  (return (Variables|PlayModeSwitching|GetCan%20be%20Enabled%20During%20Edit%20Mode)))))
            (:else
              _)))
        (:else
          _))
      (return (not (Variables|Default|GetbRequiresInitializeToEnable))))
    (:CastFailed
      _)))

(fn IsEnabled
  (bind _enabledstate_enabledstate_5_1aa41f2a4889e79f9eb31cbea2aa0731 (Utilities|Struct|BreakCreativeEnabledState (Variables|GetEnabledState_EnabledState_5_1AA41F2A4889E79F9EB31CBEA2AA0731)))
  (return (and (Utilities|Enum|Equal(Enum) _enabledstate_enabledstate_5_1aa41f2a4889e79f9eb31cbea2aa0731 "NewEnumerator0") (CallFunction|CanBeEnabled))))

(fn IsMinigameTheSameAsCachedMinigame
  (params (Minigame :type Fort%20Minigame%20Object%20Reference))
  (return (== Minigame (Variables|Default|GetCachedMinigame))))

(fn IsResetting
  (bind _enabledstate_enabledstate_5_1aa41f2a4889e79f9eb31cbea2aa0731 (Utilities|Struct|BreakCreativeEnabledState (Variables|GetEnabledState_EnabledState_5_1AA41F2A4889E79F9EB31CBEA2AA0731)))
  (return (Utilities|Enum|Equal(Enum) _enabledstate_enabledstate_5_1aa41f2a4889e79f9eb31cbea2aa0731 "NewEnumerator1") (Utilities|Time|GetTimerRemainingTimebyHandle (Variables|Reset|GetReset%20Timer))))

(fn OnRep_EIsEnabled
  (bind (_enabledstate_enabledstate_5_1aa41f2a4889e79f9eb31cbea2aa0731 _enabledstate_resettime_4_627345954070c685370f9cb7beb6b42c) (Utilities|Struct|BreakCreativeEnabledState (Variables|GetEnabledState_EnabledState_5_1AA41F2A4889E79F9EB31CBEA2AA0731)))
  (bind _enabledstate_enabledstate_5_1aa41f2a4889e79f9eb31cbea2aa0731_2 (Utilities|Struct|BreakCreativeEnabledState (Variables|GetEnabledState_EnabledState_5_1AA41F2A4889E79F9EB31CBEA2AA0731)))
  (bind _enabledstate_enabledstate_5_1aa41f2a4889e79f9eb31cbea2aa0731_3 (Utilities|Struct|BreakCreativeEnabledState (Variables|GetEnabledState_EnabledState_5_1AA41F2A4889E79F9EB31CBEA2AA0731)))
  (Default|CallOnEnabledStateChanged (Utilities|Enum|Equal(Enum) _enabledstate_enabledstate_5_1aa41f2a4889e79f9eb31cbea2aa0731_3 "NewEnumerator0"))
  (if (Utilities|Enum|Equal(Enum) _enabledstate_enabledstate_5_1aa41f2a4889e79f9eb31cbea2aa0731_3 "NewEnumerator1")
    (bind _returnvalue (Utilities|Time|SetTimerbyEvent (EventDispatchers|CreateEvent :!event Reset%20Delay%20Complete) _enabledstate_resettime_4_627345954070c685370f9cb7beb6b42c)) ; If we're delayed, start a timer then call the Reset Delay Complete function.\nThis won't actually do anything on the client, but letting it be called\nin order to get reset times, which clients may want for Effects.
    (Variables|Reset|SetReset%20Timer _returnvalue)
    (else
      (Utilities|Time|ClearandInvalidateTimerbyHandle (Variables|Reset|GetReset%20Timer))
      (if (Variables|Default|GetShow%20Debug)
        (Development|PrintString (Utilities|String|Append "NEW ENABLED STATE: " (Utilities|String|EnumtoString _enabledstate_enabledstate_5_1aa41f2a4889e79f9eb31cbea2aa0731_2)) true true (Utilities|Select :Index _enabledstate_enabledstate_5_1aa41f2a4889e79f9eb31cbea2aa0731_2 :NewEnumerator0 "(R=0.000000,G=1.000000,B=0.277591,A=1.000000)" :NewEnumerator1 "(R=1.000000,G=0.490260,B=0.000000,A=1.000000)" :NewEnumerator2 "(R=1.000000,G=0.000000,B=0.000000,A=1.000000)" :NewEnumerator3 "(R=0.168486,G=0.249502,B=0.442708,A=1.000000)") 5.0)))))

(fn ResetDelayComplete
  (Utilities|SwitchHasAuthority
    (:Authority
      (CallFunction|SetEnabled true))))

(fn ResetDelayStart
  (params (Delay :type Float%20%28double-precision%29))
  (if (Math|Boolean|ANDBoolean (not (Reset|IsResetting)) (CallFunction|IsEnabled) (!= (Variables|GetDelay) 0.0) (Networking|HasAuthority (Components|GetOwner)))
    (if (> (Variables|GetDelay) 0.0)
      (Variables|SetEnabledState_EnabledState_5_1AA41F2A4889E79F9EB31CBEA2AA0731 (Utilities|Struct|MakeCreativeEnabledState "NewEnumerator1" (Variables|GetDelay)))
      (Networking|FlushNetDormancy (Components|GetOwner)) ; Force Replication for push the notify through \n(needed for Repnotifies on components)
      (else
        (CallFunction|SetEnabled false)))))

(fn SetEnabled
  (params (bEnabled :type Boolean))
  (bind _true (CallFunction|CanBeEnabled))
  (bind _returnvalue (and bEnabled _true))
  (bind _returnvalue_2 (select _returnvalue "NewEnumerator0" "NewEnumerator2"))
  (bind _enabledstate_enabledstate_5_1aa41f2a4889e79f9eb31cbea2aa0731 (Utilities|Struct|BreakCreativeEnabledState (Variables|GetEnabledState_EnabledState_5_1AA41F2A4889E79F9EB31CBEA2AA0731)))
  (Utilities|SwitchHasAuthority
    (:Authority
      (if (Utilities|Enum|NotEqual(Enum) _returnvalue_2 _enabledstate_enabledstate_5_1aa41f2a4889e79f9eb31cbea2aa0731)
        (Default|CallOnEnabledStatePreChanged (Variables|GetbEnabled))
        (Variables|SetEnabledState_EnabledState_5_1AA41F2A4889E79F9EB31CBEA2AA0731 (Utilities|Struct|MakeCreativeEnabledState _returnvalue_2))
        (Networking|FlushNetDormancy (Components|GetOwner)) ; Force Replication for push the notify through \n(needed for Repnotifies on components)
      ))))

(fn SetEnabledonPhase
  (params (State :type EFortMinigameState%20Enum :default "PreGame") (EnabledIndex :type Integer))
  (if (Variables|PlayModeSwitching|GetSwitch%20On%20Phase%20Change)
    (if (not (Utilities|Set|ContainsItem (Variables|Default|GetIgnorePhases) State))
      (CallFunction|SetEnabled (Class|CreativeCommonDeviceFunctionLibrary|IsEnabledPhaseMatch State EnabledIndex (Components|GetOwner))))))

(fn SetEnabledonPlayMode
  (params (Play%20Mode :type Boolean))
  (if (Variables|PlayModeSwitching|GetSwitch%20On%20Play%20Change)
    (if Play%20Mode
      (CallFunction|SetEnabled (Variables|PlayModeSwitching|GetEnabledDuringPlayMode))
      (else
        (CallFunction|SetEnabled (Variables|PlayModeSwitching|GetEnabledDuringEditMode))))))

(fn UnbindMinigameEvents
  (params (Minigame :type Fort%20Minigame%20Object%20Reference))
  (Utilities|IsValid Minigame
    (:Is%20Valid
      (Minigame|UnbindEventfromOnPlayModeChangedDelegate Minigame (EventDispatchers|CreateEvent :!event Minigame%20Play%20Mode%20Changed))
      (Minigame|UnbindEventfromStateChangedDelegate Minigame (EventDispatchers|CreateEvent :!event Minigame%20State%20Changed)))))
