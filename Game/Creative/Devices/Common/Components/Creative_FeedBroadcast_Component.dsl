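; Creative device component that broadcasts message feed events to minigame players.
; ApplyHUDGA calls its parent, then grants and activates GA_MessageFeed on the controller's ability system if the MessageFeed tag is absent, storing itself as that ability's required optional object.
; DoFeedEvent, when feedback is enabled, sends the MessageFeed activate gameplay event to every participating player, carrying the owning device and this component in the payload.
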(fn ApplyHUDGA
  (params (PlayerController :type Player%20Controller%20Object%20Reference))
  (bind _returnvalue (Ability|GetAbilitySystemComponent PlayerController))
  (|CallParentFunction PlayerController)
  (if (not (GameplayTags|HasMatchingGameplayTag _returnvalue "(TagName=\"Ability.Creative.MessageFeed\")"))
    (bind _returnvalue_2 (Ability|Advanced|GrantandActivateAbilityonSelf _returnvalue "/Game/Creative/Devices/Common/GameplayAbilities/GA_MessageFeed.GA_MessageFeed_C"))
    (bind _asga_message_feed (Utilities|Casting|CastToGA_MessageFeed (.ReturnValue (Ability|GameplayAbility|GetGameplayAbilityfromSpecHandle _returnvalue _returnvalue_2))))
    (Variables|/Game/Creative/Devices/Common/GameplayAbilities/GA_MessageFeed.GA_MessageFeed_C|SetOptionalObjectRequired self _asga_message_feed)))

(fn DoFeedEvent
  (params (Should%20Show%20Feedback :type Boolean :default true))
  (if Should%20Show%20Feedback
    (bind _outplayers (Minigame|GetParticipatingPlayers_NotPure (Variables|Default|GetFort%20Minigame)))
    (for _array_element _outplayers
      (Ability|SendGameplayEventtoActor _array_element "(TagName=\"Event.Creative.Device.MessageFeed.Activate\")" (Utilities|Struct|MakeGameplayEventData "(TagName=\"Event.Creative.Device.MessageFeed.Activate\")" (Components|GetOwner) 0 self)))))
