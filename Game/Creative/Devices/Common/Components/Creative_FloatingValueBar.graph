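# Floating world-space widget for Creative devices that shows an icon fill bar with value, label or distance text, driven by its Creative value bar component.
# Component and team-change events refresh text, font, scale and affiliation-based colors, overrides and visibility.
# A looping timer checks allowance, range and optional throttled async line-of-sight traces, playing Appear forward or reversed.
# Other timers update distance in meters and rotate an edge arrow when the target leaves the viewport.
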WidgetBlueprint /Game/Creative/Devices/Common/Components/Creative_FloatingValueBar
animations [
 Appear {
  binding_count=3
  length_seconds=0.7
  object=&self:Appear
  sequence {bindings [
    ProgressIcon {
     id=#1
     possessed_class=Image
     tracks [
      {
       type=MovieSceneFloatTrack
       display_name=Render Opacity
       sections [{
         type=MovieSceneFloatSection
         end_frame=10
         start_frame=0
         channels [None {type=MovieSceneScriptingFloatChannel, keys [[0,0],[9.789,1],[10.789,1]]}]
        }]
      }
      {
       type=MovieScene2DTransformTrack
       sections [{
         type=MovieScene2DTransformSection
         end_frame=10
         start_frame=0
         channels [
          Translation.X {type=MovieSceneScriptingFloatChannel}
          Translation.Y {type=MovieSceneScriptingFloatChannel}
          Angle {type=MovieSceneScriptingFloatChannel}
          Scale.X {type=MovieSceneScriptingFloatChannel, keys [[2.6977,-0.5],[7.789,1.05],[10.789,1]]}
          Scale.Y {type=MovieSceneScriptingFloatChannel, keys [[2.6977,0.5],[7.789,1.05],[10.789,1]]}
          Shear.X {type=MovieSceneScriptingFloatChannel}
          Shear.Y {type=MovieSceneScriptingFloatChannel}
         ]
        }]
      }
      {
       type=MovieSceneWidgetMaterialTrack
       sections [{
         type=MovieSceneParameterSection
         end_frame=10
         start_frame=0
         channels [OverrideFillColor {type=MovieSceneScriptingFloatChannel, keys [[0,1],[4.8253,0.9763],[5.327,-0.0237],[9.199,0],[10.789,0]]}]
        }]
      }
     ]
    }
    ValueBox {
     id=#2
     possessed_class=HorizontalBox
     tracks [
      {
       type=MovieSceneFloatTrack
       display_name=Render Opacity
       sections [{
         type=MovieSceneFloatSection
         end_frame=14
         start_frame=0
         channels [None {type=MovieSceneScriptingFloatChannel, keys [[0,0],[10.789,0],[14,1]]}]
        }]
      }
      {
       type=MovieScene2DTransformTrack
       sections [{
         type=MovieScene2DTransformSection
         end_frame=14
         start_frame=10
         channels [
          Translation.X {type=MovieSceneScriptingFloatChannel, const [0,2]}
          Translation.Y {type=MovieSceneScriptingFloatChannel, keys [[10,7],[14,0]]}
          Angle {type=MovieSceneScriptingFloatChannel}
          Scale.X {type=MovieSceneScriptingFloatChannel}
          Scale.Y {type=MovieSceneScriptingFloatChannel}
          Shear.X {type=MovieSceneScriptingFloatChannel}
          Shear.Y {type=MovieSceneScriptingFloatChannel}
         ]
        }]
      }
     ]
    }
    Text_IndicatorLabel {
     id=#3
     display_name=IndicatorSwitcher
     possessed_class=TextBlock
     tracks [
      {
       type=MovieSceneFloatTrack
       display_name=Render Opacity
       sections [{
         type=MovieSceneFloatSection
         end_frame=10
         start_frame=0
         channels [None {type=MovieSceneScriptingFloatChannel, keys [[0,0],[5,1],[10.789,1]]}]
        }]
      }
      {
       type=MovieScene2DTransformTrack
       sections [{
         type=MovieScene2DTransformSection
         end_frame=10
         start_frame=0
         channels [
          Translation.X {type=MovieSceneScriptingFloatChannel}
          Translation.Y {type=MovieSceneScriptingFloatChannel}
          Angle {type=MovieSceneScriptingFloatChannel}
          Scale.X {type=MovieSceneScriptingFloatChannel, keys [[0,0.3023],[3,0.8665],[4,1.4397],[6.139,1],[10.789,1]]}
          Scale.Y {type=MovieSceneScriptingFloatChannel, keys [[0,0.6],[3,1.5185],[4,0.8081],[6.139,1],[10.789,1]]}
          Shear.X {type=MovieSceneScriptingFloatChannel}
          Shear.Y {type=MovieSceneScriptingFloatChannel}
         ]
        }]
      }
     ]
    }
   ]}
 }
 ShowDamage {
  binding_count=1
  length_seconds=0.2
  object=&self:ShowDamage
  sequence {bindings [ProgressIcon {
     id=#4
     possessed_class=Image
     tracks [{
       type=MovieSceneWidgetMaterialTrack
       sections [{
         type=MovieSceneParameterSection
         end_frame=4
         start_frame=0
         channels [ShowDamage {type=MovieSceneScriptingFloatChannel, keys [[0,1],[4,0]]}]
        }]
      }]
    }]}
 }
]
graphs [
 Bind Owning Player {nodes [
   Bind Owning Player() {is=title K2Node_FunctionEntry}
   n1 = GetOwningPlayer()
   > Is Valid(InputObject=@n1)
   > n2 = Cast To FortPlayerControllerAthena(Object=@n1)
   > Bind Event to On Local Player Changed Teams(self:Fort Player Controller Athena Object Reference=@n2, Delegate=Create Event()) {is=title K2Node_AddDelegate}
  ]}
 Bind Widget Component {nodes [
   Bind Widget Component() {is=title K2Node_FunctionEntry}
   > n1 = Get!(execute:Exec) -> (then:Exec, else:Exec, Value Bar Component:Creative Value Bar Component Object Reference)
   > Bind Event to On Value Bar Component Updated(self:Creative Value Bar Component Object Reference=@n1, Delegate=Create Event()) {is=title K2Node_AddDelegate}
   > Bind Event to Custom Color Changed(self:Creative Value Bar Component Object Reference=@n1, Delegate=Create Event()) {is=title K2Node_AddDelegate}
  ]}
 CheckTeamUpdates {nodes [
   CheckTeamUpdates() {is=title K2Node_FunctionEntry}
   > Get()
   > n1 = GetLocalPlayerAffiliation!(execute, self) -> (then, LocalTeamAffiliation:EFortTeamAffiliation Enum=Friendly)
   > n2 = Branch(Condition=NOT Boolean(A=Get UseCustomColor() -> (UseCustomColor:bool)))
   > n3 = UpdateIconColorByAffiliation!(execute, self, TeamAffiliation:EFortTeamAffiliation Enum=@n1) -> (then)
   UpdateTextByAffiliation!(execute=[@n3.then, @n2.else], self, TeamAffiliation:EFortTeamAffiliation Enum=@n1) -> (then)
   > UpdateVisibilityByAffiliation!(execute, self, TeamAffiliation:EFortTeamAffiliation Enum=@n1) -> (then)
  ]}
 EventGraph {nodes [
   On Widget Component Updated!() -> (OutputDelegate, then, ValueBarComp:Creative Value Bar Component Object Reference) {is=title K2Node_CustomEvent}
   > n1 = Get()
   > UpdateVisibilityCheck()
   > Show Value!(execute, self, On:bool=Get Has Dynamic Value!(self:Creative Value Bar Component Object Reference=@n1) -> (Has Dynamic Value:bool)) -> (then)
   > CheckTeamUpdates() {is=title K2Node_CallFunction}
   > "Update Text && Indicator"()
   > n2 = Sequence!(execute) -> (then_0, then_1, then_2:Exec)
   > n3 = Branch(Condition=Get bDisplayDistanceText!(self:Creative Value Bar Component Object Reference=Get Value Bar Component() -> (Value Bar Component:Creative Value Bar Component Object Reference)) -> (bDisplayDistanceText:bool))
   > Show Distance Text()
   Hide Distance Text(execute=@n3.else)
   n4 = Branch(
    execute=@n2.then_1
    Condition=Get bDisplayOffscreenArrow!(self:Creative Value Bar Component Object Reference=Get Value Bar Component()) -> (bDisplayOffscreenArrow:bool)
   )
   > Show Offscreen Arrow()
   Hide Offscreen Arrow(execute=@n4.else)
   n5 = Sequence!(execute=@n2.then_2) -> (then_0, then_1)
   n6 = Get Value Bar Component()
   > Set Progress Icon Visibility!(
    execute
    self
    bVisible:bool=Get bDisplayProgressIcon!(self:Creative Value Bar Component Object Reference=@n6) -> (bDisplayProgressIcon:bool)
   ) -> (then) {is=title K2Node_CallFunction}
   SetRenderScale!(
    execute=@n5.then_1
    self:Widget Object Reference=Get Overlay() -> (Overlay:Overlay Object Reference)
    Scale:Vector 2D Structure=Get ScaleOverride!(self:Creative Value Bar Component Object Reference=@n6) -> (ScaleOverride:Vector 2D Structure)
   ) -> (then)
   UpdateVisibilityCheck() {is=title K2Node_CustomEvent}
   > n7 = Get()
   n8 = Get Visible Distance!(self:Creative Value Bar Component Object Reference=@n7) -> (Visible Distance:double)
   > n9 = Branch(Condition=AND Boolean(A="float >= float"(A=@n8, B=0) {is=title K2Node_CallFunction}, B=ShouldBeVisible!() -> (bShouldBeVisible:bool)))
   > n10 = Range Check!(execute, self) -> (then, ReturnValue:bool)
   > n11 = Set CurrentVisibility(CurrentVisibility:bool=@n10) -> (Output_Get:bool)
   Branch(
    execute=[@n23.then, @n11.then]
    Condition="Not Equal (Boolean)"!(A:bool=Get CurrentVisibility() -> (CurrentVisibility:bool), B:bool=Get LastVisibility() -> (LastVisibility:bool)) -> (ReturnValue:bool) {is=title K2Node_CallFunction}
   )
   > PlayAnimation(
    InAnimation=Get Appear() -> (Appear:Widget Animation Object Reference)
    PlayMode=Select(Option 0:EUMGSequencePlayMode Enum=Reverse, Option 1:EUMGSequencePlayMode Enum, Index=Get CurrentVisibility()) -> (ReturnValue:EUMGSequencePlayMode Enum)
    PlaybackSpeed=1.5
   )
   > n12 = Set LastVisibility(LastVisibility:bool=Get CurrentVisibility()) -> (Output_Get:bool)
   > n13 = Branch(Condition=@n12)
   > SetVisibility(InVisibility=SelfHitTestInvisible)
   > Clear and Invalidate Timer by Handle(Handle=Get HideTimer() -> (HideTimer:Timer Handle Structure))
   n14 = HideIndicator() {is=title K2Node_CustomEvent}
   n15 = Set Timer by Event(execute=@n13.else, Delegate=@n14, Time=GetEndTime(self=Get Appear()), bLooping=false)
   > Set HideTimer(HideTimer:Timer Handle Structure=@n15) -> (Output_Get:Timer Handle Structure)
   Disable HUD(execute=@n9.else)
   Clear and Invalidate Timer by Handle(execute=@n7.else, Handle=Get VisCheckTimerHandle() -> (VisCheckTimerHandle:Timer Handle Structure))
   Show Offscreen Arrow() {is=title K2Node_CustomEvent}
   > n16 = Set Timer by Event(Delegate=Create Event(), Time=Get OffscreenArrowTickTime() -> (OffscreenArrowTickTime:double), bLooping=true)
   > Set OffscreenArrowTimerHandle(OffscreenArrowTimerHandle:Timer Handle Structure=@n16) -> (Output_Get:Timer Handle Structure)
   Hide Offscreen Arrow() {is=title K2Node_CustomEvent}
   > Clear and Invalidate Timer by Handle(Handle=Get OffscreenArrowTimerHandle() -> (OffscreenArrowTimerHandle:Timer Handle Structure))
   > SetVisibility(self=Get ArrowSizeBox() -> (ArrowSizeBox:Size Box Object Reference), InVisibility=Hidden)
   n17 = Get ObjectiveVerticalBox() -> (ObjectiveVerticalBox:Vertical Box Object Reference)
   > SetDrawSize!(
    execute
    self:Widget Component Object Reference=Get Value Bar Component()
    Size:Vector 2D Structure=GetLocalSize!(Geometry:"Geometry Structure (by ref)"=GetCachedGeometry!(self:Widget Object Reference=@n17) -> (ReturnValue:"Geometry Structure (by ref)")) -> (ReturnValue:Vector 2D Structure)
   ) -> (then)
   Update Distance Text Timer() {is=title K2Node_CustomEvent}
   > n18 = Get!(execute) -> (then, else, My Actor:Actor Object Reference)
   n19 = GetOwningPlayerPawn!(self:User Widget Object Reference) -> (ReturnValue:Pawn Object Reference)
   > n20 = Is Valid(InputObject=@n19)
   n21 = float / float(A="Distance (Vector)"!(V1:Vector=Get Actor Location(self=@n19), V2:Vector=Get Actor Location(self=@n18)) -> (ReturnValue:double), B=100) {is=title K2Node_CallFunction}
   > "SetText (Text)"(
    self=Get DistanceText() -> (DistanceText:Text Object Reference)
    InText=Format Text!(Format="LOCTEXT(\"{Distance}m\")", Distance:int=Floor(A=@n21) {is=title K2Node_CallFunction}) -> (Result)
   )
   "SetText (Text)"(execute=[@n20.Is Not Valid, @n18.else], self=Get DistanceText(), InText="LOCTEXT(\"?\")")
   Show Distance Text() {is=title K2Node_CustomEvent}
   > SetActiveWidgetIndex(self=Get TextSwitch() -> (TextSwitch:Common Widget Switcher Legacy Object Reference), Index=2)
   > n22 = Set Timer by Event(Delegate=Create Event(), Time=Get DistanceTickTime() -> (DistanceTickTime:double), bLooping=true)
   > Set DistanceUpdateTimerHandle(DistanceUpdateTimerHandle:Timer Handle Structure=@n22) -> (Output_Get:Timer Handle Structure)
   Disable HUD() {is=title K2Node_CustomEvent}
   > n23 = Set CurrentVisibility(CurrentVisibility=false)
   Show Damage for Duration() {is=title K2Node_CustomEvent}
   > Branch(Condition=Get HasText() -> (HasText:bool))
   > n24 = Show Health()
   n25 = Show Objective Text() {is=title K2Node_CustomEvent}
   Set Timer by Event(execute=@n24.then, Delegate=@n25, Time=1, bLooping=false)
   n26 = UpdateVisibilityTimer!() -> (OutputDelegate, then, Visible:bool=~) {is=title K2Node_CustomEvent}
   > n27 = Branch(Condition=@n26.Visible)
   n28 = Get VisCheckTimerHandle()
   > Unpause Timer by Handle!(execute, Handle:Timer Handle Structure=@n28) -> (then)
   Pause Timer by Handle!(execute=@n27.else, Handle:Timer Handle Structure=@n28) -> (then)
   Show Health() {is=title K2Node_CustomEvent}
   > SetActiveWidgetIndex(self=Get TextSwitch(), Index=0)
   Hide Distance Text() {is=title K2Node_CustomEvent}
   > Update Text Data()
   > Clear and Invalidate Timer by Handle(Handle=Get DistanceUpdateTimerHandle() -> (DistanceUpdateTimerHandle:Timer Handle Structure))
   n29 = Event InitFromObject!() -> (OutputDelegate, then, InitObject:Object Reference) {is=title K2Node_Event}
   > n30 = Cast To Creative_ValueBar_Component!(execute, Object=@n29.InitObject) -> (then, CastFailed, AsCreative Value Bar Component:Creative Value Bar Component Object Reference) {is=title K2Node_DynamicCast}
   > n31 = Set Value Bar Component(Value Bar Component:Creative Value Bar Component Object Reference=@n30) -> (Output_Get:Creative Value Bar Component Object Reference)
   > Bind Widget Component() {is=title K2Node_CallFunction}
   > n32 = Set My Actor(My Actor:Actor Object Reference=Get Actor Ref!(self:Creative Value Bar Component Object Reference=@n31) -> (Actor Ref:Actor Object Reference)) -> (Output_Get:Actor Object Reference)
   > n33 = Sequence!(execute) -> (then_0, then_1, then_2:Exec, then_3:Exec)
   n34 = GetComponentByClass!(self:Actor Object Reference=@n32, ComponentClass:Actor Component Class Reference=/Script/FortniteGame.FortMinigameObjectiveComponent) -> (ReturnValue:Fort Minigame Objective Component Object Reference)
   > n35 = Is Valid(InputObject=@n34)
   Set FortMinigameLogic(
    execute=@n35.Is Not Valid
    FortMinigameLogic:Fort Minigame Progress Component Object Reference=GetComponentByClass(self=@n32, ComponentClass=/Script/FortniteGame.FortMinigameProgressComponent) -> (ReturnValue:Fort Minigame Progress Component Object Reference)
   ) -> (Output_Get:Fort Minigame Progress Component Object Reference)
   Set Fort Minigame Objective(execute=@n35.Is Valid, Fort Minigame Objective:Fort Minigame Objective Component Object Reference=@n34) -> (Output_Get:Fort Minigame Objective Component Object Reference)
   Bind Owning Player(execute=@n33.then_1) {is=title K2Node_CallFunction}
   > n36 = Get!(execute, self:Creative Value Bar Component Object Reference=Get Value Bar Component()) -> (then, else, Override Icon:Material Interface Object Reference)
   > SetBrushFromMaterial(self=Get ProgressIcon() -> (ProgressIcon:Image Object Reference), Material=@n36)
   "Update Text && Indicator"(execute=@n33.then_2)
   n37 = float / float(A=Get Visibility Initial Tick Max() -> (Visibility Initial Tick Max:double), B=2)
   > n38 = Set Timer by Event(
    Delegate=Create Event()
    Time=Get Visibility Tick Time() -> (Visibility Tick Time:double)
    bLooping=true
    InitialStartDelay=@n37
    InitialStartDelayVariance=@n37
   )
   > Set VisCheckTimerHandle(VisCheckTimerHandle:Timer Handle Structure=@n38) -> (Output_Get:Timer Handle Structure)
   > CheckTeamUpdates()
   n39 = Get ObjectiveVerticalBox()
   SetDrawSize(execute=@n33.then_3, self=Get Value Bar Component(), Size=GetLocalSize(Geometry=GetCachedGeometry(self=@n39)))
   > SetVisibility(self=Get ArrowSizeBox(), InVisibility=Collapsed)
   SetVisibility(execute=@n14.then, InVisibility=Collapsed)
   SetActiveWidgetIndex(execute=@n25.then, self=Get TextSwitch(), Index=1)
  ]}
 Get Minigame State {nodes [
   Get Minigame State() {is=title K2Node_FunctionEntry}
   > n1 = Sequence!(execute) -> (then_0, then_1, then_2)
   > n2 = Get!(execute) -> (then, else, Fort Minigame Objective:Fort Minigame Objective Component Object Reference)
   > Set local Current Minigame(local Current Minigame:Fort Minigame Object Reference=Get CurrentMinigame!(self:Fort Minigame Logic Component Object Reference=@n2) -> (CurrentMinigame:Fort Minigame Object Reference)) -> (Output_Get:Fort Minigame Object Reference)
   n3 = Get!(execute=@n2.else) -> (then, else, FortMinigameLogic:Fort Minigame Progress Component Object Reference)
   > Set local Current Minigame(local Current Minigame=Get CurrentMinigame(self=@n3))
   n4 = Get!(execute=@n1.then_1) -> (then, else, local Current Minigame:Fort Minigame Object Reference)
   > Set local State(local State:EFortMinigameState Enum=Get CurrentState!(self:Fort Minigame Object Reference=@n4) -> (CurrentState:EFortMinigameState Enum=PreGame)) -> (Output_Get:EFortMinigameState Enum=PreGame)
   Return Node!(execute=@n1.then_2, State:EFortMinigameState Enum=Get local State() -> (local State:EFortMinigameState Enum=PreGame))
  ]}
 GetLocalPlayerAffiliation {nodes [
   GetLocalPlayerAffiliation() {is=title K2Node_FunctionEntry}
   > n1 = Branch(Condition=Get bUseOwnerAffiliation() -> (bUseOwnerAffiliation:bool))
   > n2 = Set localTeamIndex(localTeamIndex:Byte=GetActorPvPTeamIndex!(Actor:Actor Object Reference=GetValidPlayerPawn!(self) -> (FoundPawn:Pawn Object Reference)) -> (TeamIndex:Byte=0, ReturnValue:bool)) -> (Output_Get:Byte=0)
   n3 = Get localTeamIndex() -> (localTeamIndex:Byte=0)
   n4 = Get Value Bar Component()
   n5 = Branch(
    execute=[@n2.then, @n12.then]
    Condition=Contains Item(
     TargetArray:Array of Integers=Get Friendly Teams!(self:Creative Value Bar Component Object Reference=@n4) -> (Friendly Teams:Array of Integers)
     ItemToFind:"Integer (by ref)"="To Integer (Byte)"!(InByte:Byte=@n3) -> (ReturnValue:int=0)
    )
   )
   > Return Node!(execute, LocalTeamAffiliation:EFortTeamAffiliation Enum=Friendly)
   n6 = Get localTeamIndex() -> (localTeamIndex=0)
   n7 = Get Value Bar Component()
   n8 = Branch(
    execute=@n5.else
    Condition=Contains Item(
     TargetArray=Get Hostile Teams!(self:Creative Value Bar Component Object Reference=@n7) -> (Hostile Teams:Array of Integers)
     ItemToFind="To Integer (Byte)"(InByte=@n6) -> (ReturnValue=0)
    )
   )
   > Return Node(LocalTeamAffiliation=Hostile)
   n9 = Get My Actor() -> (My Actor:Actor Object Reference)
   n10 = GetValidPlayerPawn()
   n11 = Branch(execute=@n8.else, Condition=AND Boolean(A=IsValid(Object=@n10), B=IsValid(Object=@n9)))
   > Return Node(LocalTeamAffiliation=GetTeamAffiliation!(ActorA:Actor Object Reference=GetValidPlayerPawn(), ActorB:Actor Object Reference=Get My Actor()) -> (ReturnValue:EFortTeamAffiliation Enum=Friendly))
   Return Node(execute=@n11.else, LocalTeamAffiliation=Neutral)
   n12 = Set localTeamIndex(
    execute=@n1.else
    localTeamIndex=GetActorPvPTeamIndex(Actor=GetFirstLocalFortPlayerController!(bIncludeNonSpectators:bool=true, bIncludeSpectators:bool=false) -> (ReturnValue:Fort Player Controller Object Reference)) -> (TeamIndex=0)
   ) -> (Output_Get=0)
  ]}
 GetValidPlayerPawn {nodes [
   GetValidPlayerPawn() {is=title K2Node_FunctionEntry}
   > n1 = Cast To Pawn!(execute, Object=GetViewTarget!(self:Controller Object Reference=GetOwningPlayer()) -> (ReturnValue:Actor Object Reference)) -> (then, CastFailed, AsPawn:Pawn Object Reference) {is=title K2Node_DynamicCast}
   > Return Node!(execute, FoundPawn:Pawn Object Reference=@n1)
   Return Node(execute=@n1.CastFailed)
  ]}
 IsShowingValueBarAllowedBasedOnMinigameState {nodes [
   IsShowingValueBarAllowedBasedOnMinigameState() {is=title K2Node_FunctionEntry}
   n1 = Get Value Bar Component()
   > Return Node!(
    execute
    bAllowShowingValueBar:bool=Contains Item!(
     TargetSet:Set of EFortMinigameState Enums=Get ShowDuringPhases!(self:Creative Value Bar Component Object Reference=@n1) -> (ShowDuringPhases:Set of EFortMinigameState Enums=PreGame)
     ItemToFind:"EFortMinigameState Enum (by ref)"=Get Minigame State!() -> (State:EFortMinigameState Enum=PreGame) {is=title K2Node_CallFunction}
    ) -> (ReturnValue:bool) {is=title K2Node_CallFunction}
   )
  ]}
 Range Check {nodes [
   Range Check() {is=title K2Node_FunctionEntry}
   > n1 = Sequence!(execute) -> (then_0, then_1)
   > n2 = Get!(execute) -> (then, else, Value Bar Component:Creative Value Bar Component Object Reference)
   > n3 = Get!(execute) -> (then, else, My Actor:Actor Object Reference)
   n4 = GetValidPlayerPawn() {is=title K2Node_CallFunction}
   n5 = Get Visible Distance(self=@n2)
   > n6 = Branch(Condition=AND Boolean!(A=IsValid(Object=@n3), B="float > float"(A=@n5, B=0) {is=title K2Node_CallFunction}, C:bool=IsValid(Object=@n4)) -> (ReturnValue))
   n7 = Get Visible Distance(self=Get Value Bar Component())
   n8 = vector - vector!(A:Vector=Get Actor Location(self=Get My Actor()), B:Vector=Get Actor Location(self=GetValidPlayerPawn())) -> (ReturnValue:Vector) {is=title K2Node_CallFunction}
   > n9 = Set local In Range(local In Range:bool="float >= float"(A=Square!(A:double=@n7) -> (ReturnValue:double), B=Vector Length Squared!(A:Vector=@n8) -> (ReturnValue:double))) -> (Output_Get:bool)
   n10 = Get Value Bar Component()
   > Branch(Condition=AND Boolean!(A=@n9, B=Get Require LOS!(self:Creative Value Bar Component Object Reference=@n10) -> (Require LOS:bool)) -> (ReturnValue))
   > Request LOS Update()
   > Set local In Range(local In Range=Get isInLineOfSight() -> (isInLineOfSight:bool))
   SetVisibility(execute=@n6.else, InVisibility=Collapsed)
   > Set local In Range(local In Range=false)
   Return Node!(execute=@n1.then_1, ReturnValue:bool=Get local In Range() -> (local In Range:bool))
  ]}
 Request LOS Update {nodes [
   Request LOS Update() {is=title K2Node_FunctionEntry}
   > n1 = GetAsyncTraceQueue!(execute, ContextActor:Actor Object Reference=GetOwningPlayer()) -> (then, ReturnValue:Async Trace Queue Object Reference)
   > n2 = Get!(execute) -> (then, else, Value Bar Component:Creative Value Bar Component Object Reference)
   > n3 = Get!(execute) -> (then, else, My Actor:Actor Object Reference)
   n4 = Get Attached Mesh!(self:Creative Value Bar Component Object Reference=@n2) -> (Attached Mesh:Static Mesh Component Object Reference)
   n5 = GetBoundingBox!(self:Static Mesh Object Reference=Get StaticMesh(self=@n4)) -> (ReturnValue_Min:Vector=~, ReturnValue_Max:Vector=~, ReturnValue_IsValid:bool=~)
   n6 = vector + vector(A:Vector=@n5, B:Vector=@n5.ReturnValue_Max) -> (ReturnValue:Vector) {is=title K2Node_CommutativeAssociativeBinaryOperator}
   n7 = Get World Scale!(self:Scene Component Object Reference=@n4) -> (ReturnValue:Vector) {is=title K2Node_CallFunction}
   > ThrottledAsyncTrace_SourceTarget!(
    execute
    self:Async Trace Queue Object Reference=@n1
    bUseLocalCameraAsSource:bool=true
    SourceActor:Actor Object Reference
    SourceOffset_X:double=0
    SourceOffset_Y:double=0
    SourceOffset_Z:double=0
    TargetActor:Actor Object Reference=@n3
    TargetOffset:"Vector (by ref)"=vector / float!(
     A:Vector=vector * vector(A:Vector=@n6, B:Vector=@n7) -> (ReturnValue:Vector) {is=title K2Node_CommutativeAssociativeBinaryOperator}
     B:double=2
    ) -> (ReturnValue:Vector)
    ActorsToIgnore:Array of Actor Object References=Make Array!("[0]":Actor Object Reference=GetValidPlayerPawn(), "[1]":Actor Object Reference=Get My Actor()) -> (Array:Array of Actor Object References)
    CollisionChannel:ECollisionChannel Enum=ECC_Camera
    CompletionFunction:Delegate=Create Event()
   ) -> (then)
  ]}
 Set Progress Icon Visibility {nodes [
   n1 = Set Progress Icon Visibility!() -> (then, bVisible:bool=~) {is=title K2Node_FunctionEntry}
   > SetVisibility(
    self=Get ProgressIcon()
    InVisibility=Select(Option 0:ESlateVisibility Enum=Hidden, Option 1:ESlateVisibility Enum=Visible, Index=@n1) -> (ReturnValue:ESlateVisibility Enum)
   )
  ]}
 Set Text Override {nodes [Set Text Override() {is=title K2Node_FunctionEntry}]}
 SetCustomColor {nodes [
   SetCustomColor() {is=title K2Node_FunctionEntry}
   > n1 = Branch(Condition=Get Use Custom Color!(self:Creative Value Bar Component Object Reference=Get Value Bar Component()) -> (Use Custom Color:bool))
   n2 = Get Value Bar Component()
   > Set DistanceToWhite(DistanceToWhite:float="Distance (LinearColor)"!(
     C1:Linear Color Structure=Get Custom Color!(self:Creative Value Bar Component Object Reference=@n2) -> (Custom Color:Linear Color Structure)
     C2:Linear Color Structure=(R=1,G=1,B=1,A=1)
    ) -> (ReturnValue:float)) -> (Output_Get:float)
   > n3 = GetDynamicMaterial(self=Get ProgressIcon())
   n4 = Get Value Bar Component()
   > SetVectorParameterValue(
    self=@n3
    ParameterName=BackingColor
    Value=LinearColor - LinearColor(A:Linear Color Structure=Get Custom Color(self=@n4), B=Get CustomColor_BackingDifference() -> (CustomColor_BackingDifference:double)) -> (ReturnValue:Linear Color Structure) {is=title K2Node_PromotableOperator}
   )
   n5 = Get CustomColor_BorderSwitchAmount() -> (CustomColor_BorderSwitchAmount:double)
   n6 = Get DistanceToWhite() -> (DistanceToWhite:float)
   n7 = Get CustomColor_BorderDifference() -> (CustomColor_BorderDifference:double)
   n8 = Get Custom Color(self=Get Value Bar Component())
   > SetVectorParameterValue(
    self=@n3
    ParameterName=BorderColor
    Value=SelectColor(
     A=LinearColor - LinearColor(A=@n8, B=@n7)
     B=LinearColor + LinearColor(A:Linear Color Structure=@n8, B=@n7) -> (ReturnValue:Linear Color Structure) {is=title K2Node_PromotableOperator}
     bPickA="float < float"(A=@n6, B=@n5)
    )
   )
   n9 = Get Custom Color(self=Get Value Bar Component())
   > SetVectorParameterValue(self=@n3, ParameterName=FillColor, Value=@n9)
   n10 = "float < float"(A=Get DistanceToWhite(), B=Get CustomColor_IndicatorSwitchAmount() -> (CustomColor_IndicatorSwitchAmount:double))
   > SetContentColorAndOpacity!(
    execute
    self:Border Object Reference=Get IndicatorColor() -> (IndicatorColor:Border Object Reference)
    InContentColorAndOpacity:Linear Color Structure="Lerp (LinearColor)"(A=@n9, B=SelectColor(A=(R=0.1441,G=0.1441,B=0.1441,A=1), B=(R=1,G=1,B=1,A=1), bPickA=@n10), Alpha=0.9)
   ) -> (then)
   n11 = GetLocalPlayerAffiliation(execute=@n1.else) -> (LocalTeamAffiliation=Friendly) {is=title K2Node_CallFunction}
   > UpdateIconColorByAffiliation(TeamAffiliation=@n11)
  ]}
 ShouldBeVisible {nodes [
   ShouldBeVisible() {is=title K2Node_FunctionEntry}
   > n1 = Get!(execute) -> (then, else, Value Bar Component:Creative Value Bar Component Object Reference)
   > Return Node!(execute, bShouldBeVisible:bool=IsShowingValueBarAllowed!(self:Creative Value Bar Component Object Reference=@n1) -> (IsAllowed:bool))
   Return Node(execute=@n1.else, bShouldBeVisible=false)
  ]}
 Show Value {nodes [
   n1 = Show Value!() -> (then, On:bool=~) {is=title K2Node_FunctionEntry}
   > SetVisibility(
    self=[
     Get ValueBoxSpacer() -> (ValueBoxSpacer:Image Object Reference)
     Get CurrentValue() -> (CurrentValue:Text Object Reference)
     Get MaxValue() -> (MaxValue:Text Object Reference)
    ]
    InVisibility=Select(Option 0=Hidden, Option 1=SelfHitTestInvisible, Index=@n1)
   )
   > n2 = Branch(Condition=Get On() -> (On:bool))
   n3 = GetDynamicMaterial(execute=@n2.else, self=Get ProgressIcon())
   > SetScalarParameterValue(self=@n3, ParameterName=PercentFill, Value=1)
  ]}
 TraceComplete {nodes [
   n1 = TraceComplete!() -> (then, Success:bool=~, ActorHit:Actor Object Reference, HitResult:Hit Result Structure) {is=title K2Node_FunctionEntry}
   > Branch(Condition=@n1)
   > n2 = Branch(Condition=IsValid(Object=@n1.ActorHit))
   > n3 = Get!(execute) -> (then, else, My Actor:Actor Object Reference)
   > n4 = Branch(Condition="Equal (Object)"(A=@n3, B=@n1.ActorHit) {is=title K2Node_CallFunction})
   Set isInLineOfSight(execute=[@n2.else, @n4.then], isInLineOfSight:bool=true) -> (Output_Get:bool)
   Set isInLineOfSight(execute=@n4.else, isInLineOfSight=false)
  ]}
 Update {nodes [
   n1 = Update!() -> (then, CurrentValue:int, MaxValue:int, ForcedValuePercent:double=-1) {is=title K2Node_FunctionEntry}
   n2 = "Not Equal (Float)"!(A:double=@n1.ForcedValuePercent, B:double=-1) -> (ReturnValue:bool) {is=title K2Node_CallFunction}
   n3 = Truncate(A=int * float!(A:int=@n1.MaxValue, B:double=@n1.ForcedValuePercent) -> (ReturnValue:double))
   > "SetText (Text)"(self=Get CurrentValue(), InText="To Text (Integer)"(Value=Select(Option 0:int=@n1, Option 1:int=@n3, Index=@n2) -> (ReturnValue:int)))
   > "SetText (Text)"(self=Get MaxValue(), InText="To Text (Integer)"(Value=Get MaxValue() -> (MaxValue:int=0)))
   n4 = "To Float (Integer)"(InInt=Get MaxValue() -> (MaxValue=0))
   n5 = "To Float (Integer)"(InInt=Get CurrentValue() -> (CurrentValue:int=0))
   > Set LocalValuePercent(LocalValuePercent:double=Select(Option 0:double=float / float(A=@n5, B=@n4), Option 1:double=Get ForcedValuePercent() -> (ForcedValuePercent:double), Index=@n2) -> (ReturnValue:double)) -> (Output_Get:double)
   > n6 = GetDynamicMaterial(self=Get ProgressIcon())
   > SetScalarParameterValue(self=@n6, ParameterName=PercentFill, Value=Get LocalValuePercent() -> (LocalValuePercent:double))
   > Set CurrentValuePercent(CurrentValuePercent:double=Get LocalValuePercent()) -> (Output_Get:double)
   > Branch(Condition="float < float"!(A:double=Get LocalValuePercent(), B:double=Get CurrentValuePercent() -> (CurrentValuePercent:double)) -> (ReturnValue:bool) {is=title K2Node_CallFunction})
   > PlayAnimation(InAnimation=Get ShowDamage() -> (ShowDamage:Widget Animation Object Reference))
  ]}
 Update Offscreen Arrow {nodes [
   Update Offscreen Arrow() {is=title K2Node_FunctionEntry}
   > n1 = Get!(execute) -> (then, else, Value Bar Component:Creative Value Bar Component Object Reference)
   > Branch(Condition=Get bClampProjectedLocalPositionToViewport!(self:Fort Widget Component Object Reference=@n1) -> (bClampProjectedLocalPositionToViewport:bool))
   > Set ViewportSize(ViewportSize:Vector 2D Structure=GetViewportSize!() -> (ReturnValue:Vector 2D Structure)) -> (Output_Get:Vector 2D Structure)
   n2 = Get!(self:Fort Widget Component Object Reference=Get Value Bar Component()) -> (PreClampedLocalPosition_X:double, PreClampedLocalPosition_Y:double)
   n3 = GetViewportScale!() -> (ReturnValue:float)
   n4 = Get!() -> (ViewportSize_X:double, ViewportSize_Y:double)
   n5 = float / float(A=@n4, B=@n3) {is=title K2Node_PromotableOperator}
   n6 = Get!(self:Fort Widget Component Object Reference=Get Value Bar Component()) -> (PreClampedLocalPosition_X:double, PreClampedLocalPosition_Y:double)
   n7 = float / float(A=@n4.ViewportSize_Y, B=@n3)
   > Set bOutOfViewportBounds(bOutOfViewportBounds:bool=OR Boolean!(
     A="float < float"(A=@n2, B=0)
     B="float < float"(A=@n2.PreClampedLocalPosition_Y, B=0)
     C:bool="float > float"(A=@n6, B=@n5)
     D:bool="float > float"(A=@n6.PreClampedLocalPosition_Y, B=@n7)
    ) -> (ReturnValue)) -> (Output_Get:bool)
   n8 = Get bDisplayOffscreenArrow(self=Get Value Bar Component())
   n9 = Get bOutOfViewportBounds() -> (bOutOfViewportBounds:bool)
   > SetVisibility(
    self=Get ArrowSizeBox()
    InVisibility=Select(Option 0:ESlateVisibility Enum=Hidden, Option 1:ESlateVisibility Enum=SelfHitTestInvisible, Index=AND Boolean(A=@n9, B=@n8)) -> (ReturnValue:ESlateVisibility Enum)
   )
   > Branch(Condition=IsVisible(self=Get ArrowSizeBox()))
   n10 = GetCachedGeometry(self=Get ObjectiveVerticalBox())
   n11 = Get ArrowSizeBox()
   n12 = Get WidthOverride!(self:Size Box Object Reference=@n11) -> (WidthOverride:float)
   n13 = Get HeightOverride!(self:Size Box Object Reference=@n11) -> (HeightOverride:float)
   > SetDrawSize(
    self=Get Value Bar Component()
    Size=Select(
     Option 0:Vector 2D Structure=GetLocalSize(Geometry=@n10)
     Option 1:Vector 2D Structure=MakeVector2D(X=@n12, Y=@n13)
     Index=Get bOutOfViewportBounds()
    ) -> (ReturnValue:Vector 2D Structure)
   )
   > Branch(Condition=Get bOutOfViewportBounds())
   n14 = Get(self=Get Value Bar Component())
   > n15 = Branch(Condition="float < float"(A=@n14, B=0))
   > SetRenderTransformAngle!(execute, self:Widget Object Reference=Get ArrowSpinnerSizeBox() -> (ArrowSpinnerSizeBox:Size Box Object Reference), Angle:float=-90) -> (then)
   n16 = Branch(execute=@n15.else, Condition="float < float"(A=@n14.PreClampedLocalPosition_Y, B=0))
   > SetRenderTransformAngle(self=Get ArrowSpinnerSizeBox(), Angle=0)
   n17 = Get(self=Get Value Bar Component())
   n18 = Get!() -> (ViewportSize_X:double, ViewportSize_Y:double)
   n19 = Branch(execute=@n16.else, Condition="float > float"(A=@n17, B=@n18))
   > SetRenderTransformAngle(self=Get ArrowSpinnerSizeBox(), Angle=90)
   Branch(execute=@n19.else, Condition="float > float"(A=@n17.PreClampedLocalPosition_Y, B=@n18.ViewportSize_Y))
   > SetRenderTransformAngle(self=Get ArrowSpinnerSizeBox(), Angle=180)
  ]}
 "Update Text && Indicator" {nodes [
   "Update Text && Indicator"() {is=title K2Node_FunctionEntry}
   > n1 = Sequence()
   n2 = GetIndicator!(self:Creative Value Bar Component Object Reference=Get Value Bar Component()) -> (isText:bool, Text:Text, Texture:Texture 2D Soft Object Reference)
   n3 = Get Text_IndicatorLabel() -> (Text_IndicatorLabel:Text Object Reference)
   > "SetText (Text)"(self=@n3, InText=@n2.Text)
   n4 = Get Icon_IndicatorLabel() -> (Icon_IndicatorLabel:Image Object Reference)
   > SetBrushFromSoftTexture!(execute, self:Image Object Reference=@n4, SoftTexture:Texture 2D Soft Object Reference=@n2.Texture, bMatchSize:bool=false) -> (then)
   > SetActiveWidget(
    self:Widget Switcher Object Reference=Get IndicatorSwitcher() -> (IndicatorSwitcher:Fort Prioritized Widget Switcher Object Reference)
    Widget=Select(Option 0:Widget Object Reference=@n4, Option 1:Widget Object Reference=@n3, Index=@n2) -> (ReturnValue:Widget Object Reference)
   )
   n5 = Get Value Bar Component()
   n6 = Get TextScale!(self:Creative Value Bar Component Object Reference=@n5) -> (TextScale:double)
   SetRenderTransform!(
    execute=@n1.then_1
    self:Widget Object Reference=Get ValueBoxSpacer()
    InTransform_Translation:Vector 2D Structure=(X=0,Y=0)
    InTransform_Scale_X:double=@n6
    InTransform_Scale_Y:double=1
    InTransform_Shear:Vector 2D Structure=(X=0,Y=0)
    InTransform_Angle:float=0
   ) -> (then)
   n7 = Get FontStyle!(self:Creative Value Bar Component Object Reference=@n5) -> (FontStyle:EBeaconFontStyle Enum=NewEnumerator0)
   n8 = Find!(
    TargetMap:Map of EBeaconFontStyle Enums to Slate Font Info Structures=Get FontStyleMap() -> (FontStyleMap:Map of EBeaconFontStyle Enums to Slate Font Info Structures=NewEnumerator0)
    Key:"EBeaconFontStyle Enum (by ref)"=@n7
   ) -> (
    Value_FontObject:Object Reference
    Value_FontMaterial:Object Reference
    Value_OutlineSettings:Font Outline Settings Structure
    Value_TypefaceFontName:Name
    Value_Size:float=~
    Value_LetterSpacing:int
    Value_SkewAmount:float=~
    Value_bForceMonospaced:bool=~
    Value_bMaterialIsStencil:bool=~
    Value_MonospacedWidth:float=~
    ReturnValue
   )
   n9 = Truncate(A=float * float(A=@n6, B=10))
   n10 = Get CurrentValue() -> (CurrentValue:Text Object Reference)
   n11 = Get Text!() -> (Text:Text Object Reference)
   n12 = Get MaxValue() -> (MaxValue:Text Object Reference)
   > SetFont(
    self=[@n10, @n11, @n12]
    InFontInfo=MakeSlateFontInfo!(
     FontObject:Object Reference=@n8
     FontMaterial:Object Reference=@n8.Value_FontMaterial
     OutlineSettings:Font Outline Settings Structure=@n8.Value_OutlineSettings
     TypefaceFontName:Name=@n8.Value_TypefaceFontName
     Size:float="To Float (Integer)"(InInt=@n9)
     LetterSpacing:int=@n8.Value_LetterSpacing
     SkewAmount:float=@n8.Value_SkewAmount
     bForceMonospaced:bool=false
     bMaterialIsStencil:bool=false
     MonospacedWidth:float=1
    ) -> (ReturnValue:Slate Font Info Structure)
   )
   n13 = Find!(
    TargetMap:Map of EBeaconFontStyle Enums to Slate Color Structures=Get FontColorMap() -> (FontColorMap:Map of EBeaconFontStyle Enums to Slate Color Structures=NewEnumerator0)
    Key=@n7
   ) -> (Value:Slate Color Structure, ReturnValue)
   n14 = Break Slate Color!(SlateColor:"Slate Color Structure (by ref)"=@n13) -> (
    SpecifiedColor_R:float=~
    SpecifiedColor_G:float=~
    SpecifiedColor_B:float=~
    SpecifiedColor_A:float=~
    ColorUseRule:ESlateColorStylingMode Enum
   ) {is=title K2Node_BreakStruct}
   n15 = Get!(self:Border Object Reference=Get IndicatorColor()) -> (ContentColorAndOpacity_R:float=~, ContentColorAndOpacity_G:float=~, ContentColorAndOpacity_B:float=~, ContentColorAndOpacity_A:float=~)
   > SetColorAndOpacity(
    self=[@n10, @n11, @n12]
    InColorAndOpacity=Select!(
     Option 0:Slate Color Structure=@n13
     Option 1_SpecifiedColor_R:float=@n15
     Option 1_SpecifiedColor_G:float=@n15.ContentColorAndOpacity_G
     Option 1_SpecifiedColor_B:float=@n15.ContentColorAndOpacity_B
     Option 1_SpecifiedColor_A:float=@n14.SpecifiedColor_A
     Option 1_ColorUseRule:ESlateColorStylingMode Enum=@n14.ColorUseRule
     Index=Get UseCustomColor()
    ) -> (ReturnValue:Slate Color Structure)
   )
  ]}
 Update Text Data {nodes [
   Update Text Data() {is=title K2Node_FunctionEntry}
   > n1 = Branch(Condition=Get HasText())
   > Show Objective Text() {is=title K2Node_CallFunction}
   Show Health(execute=@n1.else) {is=title K2Node_CallFunction}
  ]}
 UpdateIconColorByAffiliation {nodes [
   n1 = UpdateIconColorByAffiliation!() -> (then, TeamAffiliation:EFortTeamAffiliation Enum=Friendly) {is=title K2Node_FunctionEntry}
   > n2 = GetDynamicMaterial(self=Get ProgressIcon())
   n3 = Break Creative Floating Health Color!(Creative_FloatingHealthColor:"Creative Floating Health Color Structure (by ref)"=Find(
     TargetMap:Map of EFortTeamAffiliation Enums to Creative Floating Health Color Structures=Get TeamColors() -> (TeamColors:Map of EFortTeamAffiliation Enums to Creative Floating Health Color Structures=Friendly)
     Key:"EFortTeamAffiliation Enum (by ref)"=@n1
    ) -> (Value:Creative Floating Health Color Structure)) -> (
    BackingColor_2_#5:Linear Color Structure
    BorderColor_4_#6:Linear Color Structure
    FillColor_6_#7:Linear Color Structure
    ContentColorAndOpacity_8_#8:Linear Color Structure
   ) {is=title K2Node_BreakStruct}
   > SetVectorParameterValue(self=@n2, ParameterName=BackingColor, Value=@n3)
   > SetVectorParameterValue(self=@n2, ParameterName=BorderColor, Value=@n3.BorderColor_4_#6)
   > SetVectorParameterValue(self=@n2, ParameterName=FillColor, Value=@n3.FillColor_6_#7)
   > SetContentColorAndOpacity(self=Get IndicatorColor(), InContentColorAndOpacity=@n3.ContentColorAndOpacity_8_#8)
  ]}
 UpdateTextByAffiliation {nodes [
   n1 = UpdateTextByAffiliation!() -> (then, TeamAffiliation:EFortTeamAffiliation Enum=Friendly) {is=title K2Node_FunctionEntry}
   > n2 = Get!(execute:Exec) -> (then:Exec, else:Exec, Value Bar Component:Creative Value Bar Component Object Reference)
   n3 = "Get (a copy)"(
    Array=Get Text Overrides!(self:Creative Value Bar Component Object Reference=@n2) -> (Text Overrides:Array of Strings)
    Dimension 1=Select!(Friendly:int=0, Neutral:int=1, Hostile:int=2, Index:EFortTeamAffiliation Enum=@n1) -> (ReturnValue:int)
   )
   > "SetText (Text)"(self=Get Text(), InText="To Text (String)"(InString=@n3))
   > Set HasText(HasText:bool="To Boolean (Integer)"!(InInt:int=Len(S=@n3)) -> (ReturnValue:bool)) -> (Output_Get:bool)
   > Update Text Data() {is=title K2Node_CallFunction}
  ]}
 UpdateVisibilityByAffiliation {nodes [
   n1 = UpdateVisibilityByAffiliation!() -> (then, TeamAffiliation:EFortTeamAffiliation Enum=Friendly) {is=title K2Node_FunctionEntry}
   > n2 = Get()
   n3 = AND Boolean(
    A=IsShowTeamNeutral!(self:Creative Value Bar Component Object Reference=@n2) -> (ReturnValue:bool)
    B="Equal (Enum)"(A:EFortTeamAffiliation Enum=@n1, B:EFortTeamAffiliation Enum=Neutral)
   )
   n4 = AND Boolean(A=IsShowTeamFriendlies!(self:Creative Value Bar Component Object Reference=@n2) -> (ReturnValue:bool), B="Equal (Enum)"(A=@n1, B=Friendly))
   n5 = AND Boolean(A=IsShowTeamHostiles!(self:Creative Value Bar Component Object Reference=@n2) -> (ReturnValue:bool), B="Equal (Enum)"(A=@n1, B=Hostile))
   n6 = GetOwningPlayer()
   n7 = "Equal (Integer)"(
    A=GetShowTeamAsIndex!(self:Creative Value Bar Component Object Reference=@n2) -> (ReturnValue:int=0)
    B="To Integer (Byte)"(InByte=GetActorPvPTeamIndex(Actor=@n6) -> (TeamIndex=0)) -> (ReturnValue=0)
   ) {is=title K2Node_CallFunction}
   n8 = IsShowTeamValid!(self:Creative Value Bar Component Object Reference=@n2) -> (ReturnValue:bool)
   > n9 = Branch(Condition=OR Boolean(
     A=IsLocalPlayerRegistered!(self:Creative Value Bar Component Object Reference=@n2) -> (Valid:bool)
     B=IsShowTeamAny!(self:Creative Value Bar Component Object Reference=@n2) -> (ReturnValue:bool)
     C=AND Boolean(A=@n8, B=@n7)
     D=OR Boolean!(A=@n4, B=@n5, C=@n3) -> (ReturnValue)
    ))
   > UpdateVisibilityTimer!(execute, self, Visible:bool=true) -> (then) {is=title K2Node_CallFunction}
   UpdateVisibilityTimer(execute=@n9.else, Visible=false)
   > Disable HUD() {is=title K2Node_CallFunction}
  ]}
]
variables [
 bUseOwnerAffiliation {type=boolean, default=true}
 CurrentValuePercent {type=number}
 CurrentVisibility {type=boolean}
 CustomColor_BackingDifference {type=number, default=0.3}
 CustomColor_BorderDifference {type=number, default=0.2}
 CustomColor_BorderSwitchAmount {type=number, default=0.2}
 CustomColor_IndicatorSwitchAmount {type=number, default=0.35}
 DistanceTickTime {type=number, default=0.1}
 DistanceUpdateTimerHandle {type=TimerHandle}
 FontColorMap {default {
   Bold {colorUseRule=UseColor_Specified, specifiedColor=<rgba 1 1 1 0.9>}
   Subtle {colorUseRule=UseColor_Specified, specifiedColor=<rgba 1 1 1 0.75>}
  }}
 FontStyleMap {default {
   Bold {
    bForceMonospaced=false
    bMaterialIsStencil=false
    fontMaterial=null
    fontObject=&Burbank
    letterSpacing=1
    monospacedWidth=1
    size=20
    skewAmount=0
    typefaceFontName=Burbank-PlayerPower
    outlineSettings {
     bApplyOutlineToDropShadows=false
     bMiteredCorners=false
     bSeparateFillAlpha=false
     outlineMaterial=null
     outlineSize=1
     outlineColor=<rgba 0 0 0 1>
    }
   }
   Subtle {
    bForceMonospaced=false
    bMaterialIsStencil=false
    fontMaterial=null
    fontObject=&NotoSans
    letterSpacing=0
    monospacedWidth=1
    size=24
    skewAmount=0
    typefaceFontName=NotoSans-Bold
    outlineSettings {
     bApplyOutlineToDropShadows=false
     bMiteredCorners=false
     bSeparateFillAlpha=false
     outlineMaterial=null
     outlineSize=0
     outlineColor=<rgba 0 0 0 1>
    }
   }
  }}
 Fort Minigame Objective {type=FortMinigameObjectiveComponent, default=null}
 FortMinigameLogic {type=FortMinigameProgressComponent, default=null}
 HasText {type=boolean}
 HideTimer {type=TimerHandle}
 isInLineOfSight {type=boolean}
 LastVisibility {type=boolean, default=true}
 My Actor {type=Actor, default=null}
 OffscreenArrowTickTime {type=number, default=0.1}
 OffscreenArrowTimerHandle {type=TimerHandle}
 TeamColors {default {
   Friendly {
    backingColor=<rgba 0.004 0.0232 0.2159 1>
    borderColor=<rgba 0.0513 0.7379 1 1>
    contentColorAndOpacity=<rgba 0.5029 1 1 1>
    fillColor=<rgba 0 0.402 1 1>
   }
   Hostile {
    backingColor=<rgba 0.1845 0 0 1>
    borderColor=<rgba 1 0.4 0.1 1>
    contentColorAndOpacity=<rgba 1 0.7913 0.5029 1>
    fillColor=<rgba 0.8796 0.107 0.0212 1>
   }
   Neutral {
    backingColor=<rgba 0.0243 0.0243 0.0243 1>
    borderColor=<rgba 1 1 1 1>
    contentColorAndOpacity=<rgba 0.7913 1 0.7529 1>
    fillColor=<rgba 0.1746 0.7305 0.0782 1>
   }
  }}
 TeamIcons {default {
   Friendly=/Game/Athena/UI/InGame/Creative/PropertyEditor/Icon_Picker/Textures_V2/128x128/T_UI_IconLibrary_Gravestone_128
   Hostile=/Game/Athena/UI/InGame/Creative/PropertyEditor/Icon_Picker/Textures_V2/128x128/T_UI_IconLibrary_DogTags_128
   Neutral=/Game/Athena/UI/InGame/Creative/PropertyEditor/Icon_Picker/Textures/128x128/T_UI_IconLibrary_2_128
  }}
 UseCustomColor {type=boolean}
 Value Bar Component {type=Creative_ValueBar_Component_C, default=null}
 VisCheckTimerHandle {type=TimerHandle}
 Visibility Initial Tick Max {type=number, default=1}
 Visibility Tick Time {type=number, default=0.5}
]
widgets {
 info {inheritedWidgetCount=0, parentClass=&UserWidget, rootWidgetClass=&SizeBox, widgetCount=19}
 widgets [
  SizeBox_3 {widgetClassPath=&SizeBox, widget=SizeBox_3, properties {heightOverride=21.3612}}
  Overlay {widgetClassPath=&Overlay, bIsVariable=true, parent=SizeBox_3, widget=Overlay, slot {class=SizeBoxSlot}}
  ArrowSizeBox {
   widgetClassPath=&SizeBox
   bIsVariable=true
   parent=Overlay
   widget=ArrowSizeBox
   properties {bOverride_HeightOverride=true, bOverride_WidthOverride=true, heightOverride=160, widthOverride=160}
   slot {class=OverlaySlot}
  }
  ArrowSpinnerSizeBox {
   widgetClassPath=&SizeBox
   bIsVariable=true
   parent=ArrowSizeBox
   widget=ArrowSpinnerSizeBox
   properties {bOverride_WidthOverride=true, widthOverride=32}
   slot {class=SizeBoxSlot, properties {horizontalAlignment=HAlign_Center}}
  }
  Image_61 {
   widgetClassPath=&Image
   parent=ArrowSpinnerSizeBox
   widget=Image_61
   properties {brush {
     bIsDynamicallyLoaded=false
     drawAs=Image
     imageType=NoImage
     mirroring=NoMirror
     resourceName=None
     resourceObject=/Game/Widgets/SActorCanvas/DirectionArrow
     tiling=NoTile
     imageSize=<32 32>
     margin=<ltrb 0 0 0 0>
     outlineSettings {
      bUseBrushTransparency=false
      roundingType=HalfHeightRadius
      width=0
      color {colorUseRule=UseColor_Specified, specifiedColor=<rgba 0 0 0 0>}
      cornerRadii=<0 0 0 0>
     }
     tintColor {colorUseRule=UseColor_Specified, specifiedColor=<rgba 1 1 1 1>}
     uVRegion {bIsValid=false, max=<0 0>, min=<0 0>}
    }}
   slot {class=SizeBoxSlot, properties {horizontalAlignment=HAlign_Center, verticalAlignment=VAlign_Top}}
  }
  ObjectiveVerticalBox {
   widgetClassPath=&VerticalBox
   bIsVariable=true
   parent=Overlay
   widget=ObjectiveVerticalBox
   slot {class=OverlaySlot, properties {horizontalAlignment=HAlign_Center, verticalAlignment=VAlign_Center}}
  }
  TextSwitch {
   widgetClassPath=&CommonWidgetSwitcherLegacy
   bIsVariable=true
   parent=ObjectiveVerticalBox
   widget=TextSwitch
   slot {class=VerticalBoxSlot}
  }
  ValueBox {
   widgetClassPath=&HorizontalBox
   bIsVariable=true
   parent=TextSwitch
   widget=ValueBox
   properties {visibility=HitTestInvisible}
   slot {class=WidgetSwitcherSlot, properties {horizontalAlignment=HAlign_Center, padding=<ltrb 0 0 0 -2>}}
  }
  CurrentValue {
   widgetClassPath=&TextBlock
   bIsVariable=true
   parent=ValueBox
   widget=CurrentValue
   properties {
    justification=Right
    minDesiredWidth=20
    text=75
    colorAndOpacity {colorUseRule=UseColor_Specified, specifiedColor=<rgba 1 1 1 0.75>}
    font {
     bForceMonospaced=false
     bMaterialIsStencil=false
     fontMaterial=null
     fontObject=&NotoSans
     letterSpacing=0
     monospacedWidth=1
     size=10
     skewAmount=0
     typefaceFontName=NotoSans-Bold
     outlineSettings {
      bApplyOutlineToDropShadows=false
      bMiteredCorners=false
      bSeparateFillAlpha=false
      outlineMaterial=null
      outlineSize=0
      outlineColor=<rgba 0 0 0 1>
     }
    }
   }
   slot {class=HorizontalBoxSlot, properties {padding=<ltrb 0 0 2 0>}}
  }
  ValueBoxSpacer {
   widgetClassPath=&Image
   bIsVariable=true
   parent=ValueBox
   widget=ValueBoxSpacer
   properties {
    renderOpacity=0.5
    visibility=SelfHitTestInvisible
    brush {
     bIsDynamicallyLoaded=false
     drawAs=Image
     imageType=NoImage
     mirroring=NoMirror
     resourceName=None
     resourceObject=null
     tiling=NoTile
     imageSize=<2 6>
     margin=<ltrb 0 0 0 0>
     outlineSettings {
      bUseBrushTransparency=false
      roundingType=HalfHeightRadius
      width=0
      color {colorUseRule=UseColor_Specified, specifiedColor=<rgba 0 0 0 0>}
      cornerRadii=<0 0 0 0>
     }
     tintColor {colorUseRule=UseColor_Specified, specifiedColor=<rgba 1 1 1 1>}
     uVRegion {bIsValid=false, max=<0 0>, min=<0 0>}
    }
   }
   slot {class=HorizontalBoxSlot, properties {padding=<ltrb 0 1 0 2>}}
  }
  MaxValue {
   widgetClassPath=&TextBlock
   bIsVariable=true
   parent=ValueBox
   widget=MaxValue
   properties {
    minDesiredWidth=20
    text=75
    colorAndOpacity {colorUseRule=UseColor_Specified, specifiedColor=<rgba 1 1 1 0.75>}
    font {
     bForceMonospaced=false
     bMaterialIsStencil=false
     fontMaterial=null
     fontObject=&NotoSans
     letterSpacing=0
     monospacedWidth=1
     size=10
     skewAmount=0
     typefaceFontName=NotoSans-Bold
     outlineSettings {
      bApplyOutlineToDropShadows=false
      bMiteredCorners=false
      bSeparateFillAlpha=false
      outlineMaterial=null
      outlineSize=0
      outlineColor=<rgba 0 0 0 1>
     }
    }
   }
   slot {class=HorizontalBoxSlot, properties {padding=<ltrb 2 0 0 0>}}
  }
  Text {
   widgetClassPath=&TextBlock
   bIsVariable=true
   parent=TextSwitch
   widget=Text
   properties {
    justification=Center
    minDesiredWidth=20
    text=Objective
    colorAndOpacity {colorUseRule=UseColor_Specified, specifiedColor=<rgba 1 1 1 0.75>}
    font {
     bForceMonospaced=false
     bMaterialIsStencil=false
     fontMaterial=null
     fontObject=&NotoSans
     letterSpacing=0
     monospacedWidth=1
     size=10
     skewAmount=0
     typefaceFontName=NotoSans-Bold
     outlineSettings {
      bApplyOutlineToDropShadows=false
      bMiteredCorners=false
      bSeparateFillAlpha=false
      outlineMaterial=null
      outlineSize=0
      outlineColor=<rgba 0 0 0 1>
     }
    }
   }
   slot {class=WidgetSwitcherSlot, properties {horizontalAlignment=HAlign_Center, verticalAlignment=VAlign_Center}}
  }
  DistanceText {
   widgetClassPath=&TextBlock
   bIsVariable=true
   parent=TextSwitch
   widget=DistanceText
   properties {
    justification=Center
    minDesiredWidth=20
    text=0m
    font {
     bForceMonospaced=false
     bMaterialIsStencil=false
     fontMaterial=null
     fontObject=&Burbank
     letterSpacing=0
     monospacedWidth=1
     size=20
     skewAmount=0
     typefaceFontName=Burbank-PlayerPower
     outlineSettings {
      bApplyOutlineToDropShadows=false
      bMiteredCorners=false
      bSeparateFillAlpha=false
      outlineMaterial=null
      outlineSize=1
      outlineColor=<rgba 0 0 0 1>
     }
    }
   }
   slot {class=WidgetSwitcherSlot, properties {horizontalAlignment=HAlign_Center, verticalAlignment=VAlign_Center}}
  }
  Icon {widgetClassPath=&Overlay, parent=ObjectiveVerticalBox, widget=Icon, slot {class=VerticalBoxSlot}}
  ProgressIcon {
   widgetClassPath=&Image
   bIsVariable=true
   parent=Icon
   widget=ProgressIcon
   properties {
    brush {
     bIsDynamicallyLoaded=false
     drawAs=Image
     imageType=NoImage
     mirroring=NoMirror
     resourceName=None
     resourceObject=/Game/Athena/HUD/LTM/Playground/Materials/M_UI_HUD_Objective_Progress
     tiling=NoTile
     imageSize=<58 58>
     margin=<ltrb 0 0 0 0>
     outlineSettings {
      bUseBrushTransparency=false
      roundingType=HalfHeightRadius
      width=0
      color {colorUseRule=UseColor_Specified, specifiedColor=<rgba 0 0 0 0>}
      cornerRadii=<0 0 0 0>
     }
     tintColor {colorUseRule=UseColor_Specified, specifiedColor=<rgba 1 1 1 1>}
     uVRegion {bIsValid=false, max=<0 0>, min=<0 0>}
    }
    renderTransformPivot=<0.5 1>
   }
   slot {class=OverlaySlot, properties {horizontalAlignment=HAlign_Center, verticalAlignment=VAlign_Center}}
  }
  IndicatorColor {
   widgetClassPath=&Border
   bIsVariable=true
   parent=Icon
   widget=IndicatorColor
   properties {
    horizontalAlignment=HAlign_Center
    verticalAlignment=VAlign_Center
    background {
     bIsDynamicallyLoaded=false
     drawAs=NoDrawType
     imageType=NoImage
     mirroring=NoMirror
     resourceName=None
     resourceObject=null
     tiling=NoTile
     imageSize=<32 32>
     margin=<ltrb 0 0 0 0>
     outlineSettings {
      bUseBrushTransparency=false
      roundingType=HalfHeightRadius
      width=0
      color {colorUseRule=UseColor_Specified, specifiedColor=<rgba 0 0 0 0>}
      cornerRadii=<0 0 0 0>
     }
     tintColor {colorUseRule=UseColor_Specified, specifiedColor=<rgba 1 1 1 1>}
     uVRegion {bIsValid=false, max=<0 0>, min=<0 0>}
    }
    padding=<ltrb 0 0 0 0>
   }
   slot {class=OverlaySlot, properties {horizontalAlignment=HAlign_Center, verticalAlignment=VAlign_Center, padding=<ltrb 0 0 0 2>}}
  }
  IndicatorSwitcher {
   widgetClassPath=/Script/FortniteUI.FortPrioritizedWidgetSwitcher
   bIsVariable=true
   parent=IndicatorColor
   widget=IndicatorSwitcher
   slot {class=BorderSlot, properties {horizontalAlignment=HAlign_Center, verticalAlignment=VAlign_Center, padding=<ltrb 0 0 0 0>}}
  }
  Text_IndicatorLabel {
   widgetClassPath=&TextBlock
   bIsVariable=true
   parent=IndicatorSwitcher
   widget=Text_IndicatorLabel
   properties {
    text=A
    visibility=SelfHitTestInvisible
    font {
     bForceMonospaced=false
     bMaterialIsStencil=false
     fontMaterial=null
     fontObject=&Burbank
     letterSpacing=0
     monospacedWidth=1
     size=24
     skewAmount=0
     typefaceFontName=Burbank-PlayerPower
     outlineSettings {
      bApplyOutlineToDropShadows=false
      bMiteredCorners=false
      bSeparateFillAlpha=false
      outlineMaterial=null
      outlineSize=1
      outlineColor=<rgba 0 0 0 0>
     }
    }
    shadowOffset=<0 0>
   }
   slot {class=WidgetSwitcherSlot, properties {padding=<ltrb 4 2 4 2>}}
  }
  Icon_IndicatorLabel {
   widgetClassPath=&Image
   bIsVariable=true
   parent=IndicatorSwitcher
   widget=Icon_IndicatorLabel
   properties {renderTransformPivot=<0.5 1>}
   slot {class=WidgetSwitcherSlot, properties {horizontalAlignment=HAlign_Center, verticalAlignment=VAlign_Center, padding=<ltrb 0 0 0 5>}}
  }
 ]
}
