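# Floating health bar widget for Creative devices, a Creative_FloatingValueBar child that tracks a BuildingActor's health.
# Initialization casts the bound actor to BuildingActor and binds its health-changed event, which pushes truncated health and max into Update and briefly shows damage.
# With no manual max on the Value Bar Component, the value box hides and Update forces a full bar unless given a percent.
# ShouldBeVisible optionally shows the bar only below max.
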WidgetBlueprint /Game/Creative/Devices/Common/Components/Creative_FloatingHealthBar
graphs [
 EventGraph {nodes [
   OnHealthChanged() {is=title K2Node_CustomEvent}
   n1 = Get MaxValueManualSet!(self:Creative Value Bar Component Object Reference=Get Value Bar Component() -> (Value Bar Component:Creative Value Bar Component Object Reference)) -> (MaxValueManualSet:double)
   > SetVisibility(
    self=Get ValueBox() -> (ValueBox:Horizontal Box Object Reference)
    InVisibility=Select(
     Option 0=HitTestInvisible
     Option 1=Hidden
     Index="Equal (Float)"!(A:double=@n1, B:double=-1) -> (ReturnValue:bool) {is=title K2Node_CallFunction}
    )
   )
   > n2 = Get!(execute:Exec) -> (then:Exec, else:Exec, My Building:Building Actor Object Reference)
   n3 = Truncate(A=GetMaxHealth!(self:Building Actor Object Reference=@n2) -> (ReturnValue:float))
   > Branch(Condition="Not Equal (Integer)"!(A:int=@n3, B:int=0) -> (ReturnValue:bool) {is=title K2Node_CallFunction})
   > Update!(
    execute
    self:Creative Floating Value Bar Object Reference
    CurrentValue:int=Truncate(A=GetHealth!(self:Building Actor Object Reference=@n2) -> (ReturnValue:float))
    MaxValue:int=@n3
    ForcedValuePercent:double=-1
   ) -> (then)
   > Show Damage for Duration(self:Creative Floating Value Bar Object Reference)
   n4 = Event InitFromObject!() -> (OutputDelegate, then, InitObject:Object Reference) {is=title K2Node_Event}
   > Parent: InitFromObject!(execute, InitObject:Object Reference=@n4.InitObject) -> (then) {is=title K2Node_CallParentFunction}
   > n5 = Get!(execute) -> (then, else, My Actor:Actor Object Reference)
   > n6 = Cast To BuildingActor!(execute, Object=@n5) -> (then, CastFailed, AsBuilding Actor:Building Actor Object Reference) {is=title K2Node_DynamicCast}
   > n7 = Set My Building(My Building:Building Actor Object Reference=@n6) -> (Output_Get:Building Actor Object Reference)
   > Bind Event to On Building Health Changed(self:Building Actor Object Reference=@n7, Delegate=Create Event()) {is=title K2Node_AddDelegate}
   > OnHealthChanged() {is=title K2Node_CallFunction}
   Event TryUpdate() {is=title K2Node_Event}
   > OnHealthChanged()
   n8 = Event Show Value!() -> (OutputDelegate, then, On:bool) {is=title K2Node_Event}
   > Parent: Show Value!(execute, On:bool=@n8.On) -> (then) {is=title K2Node_CallParentFunction}
   > Branch(Condition=@n8.On)
   > OnHealthChanged()
   n9 = Event Update!() -> (OutputDelegate, then, CurrentValue:int=0, MaxValue:int=0, ForcedValuePercent:double) {is=title K2Node_Event}
   > n10 = Branch(Condition="Equal (Float)"(A=@n9.ForcedValuePercent, B=-1))
   n11 = Get Value Bar Component()
   > n12 = Branch(Condition="Equal (Float)"(A=Get MaxValueManualSet(self=@n11), B=-1))
   > Parent: Update!(execute, CurrentValue:int=@n9.CurrentValue, MaxValue:int=@n9.MaxValue, ForcedValuePercent:double=1) -> (then) {is=title K2Node_CallParentFunction}
   Parent: Update(execute=[@n12.else, @n10.else], CurrentValue=@n9.CurrentValue, MaxValue=@n9.MaxValue, ForcedValuePercent=@n9.ForcedValuePercent)
  ]}
 ShouldBeVisible {nodes [
   ShouldBeVisible() {is=title K2Node_FunctionEntry}
   > n1 = Get!(execute) -> (then, else, Value Bar Component:Creative Value Bar Component Object Reference)
   > n2 = Branch(Condition=Get OnlyShowWhenNotMax!(self:Creative Value Bar Component Object Reference=@n1) -> (OnlyShowWhenNotMax:bool))
   n3 = Get CurrentValuePercent() -> (CurrentValuePercent:double)
   n4 = Parent: ShouldBeVisible!() -> (bShouldBeVisible:bool) {is=title K2Node_CallParentFunction}
   > Return Node!(
    execute
    bShouldBeVisible:bool=AND Boolean(A="float < float"!(A:double=@n3, B:double=1) -> (ReturnValue:bool) {is=title K2Node_CallFunction}, B=@n4)
   )
   Return Node(execute=[@n2.else, @n1.else], bShouldBeVisible=@n4)
  ]}
]
variables [My Building {type=BuildingActor, default=null}]
widgets {
 info {
  inheritedWidgetCount=0
  parentClass=&/Creative_FloatingValueBar.Creative_FloatingValueBar_C
  rootWidgetClass=null
  widgetCount=0
 }
 widgets []
}
