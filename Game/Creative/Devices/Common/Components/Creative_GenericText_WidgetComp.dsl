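; Widget component for Creative devices that shows generic text with a color, acting only where cosmetic events can run.
; Update Text doubles as initialization: empty text discards the widget to free render target space, otherwise it loads the class (blocking in editor, async otherwise), creates and caches it.
; Update Widget waits a tick outside the editor, then calls InitFromObject, which pulls GetDetails, and requests a render update.
; Hiding clears the text to release the widget.
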(event Custom|Update%20Text
  (params (Text :type Text))
  (bind _returnvalue (Editor|InEditor))
  (bind _widget_class (Variables|Default|GetWidget%20Class))
  ;[0 1 0.1429 0.25] Update Text Will update the text on the widget, and also Create or Destroy it if it needs to (e.g. you want text or not).
  ; This also functions as the devices de-facto initialization function (since this will be used on UEFN objects and they can't
  ; handle Begin Play)
  (Networking|CanExecuteCosmeticEvents
    (:True
      (bind _output_get (Variables|Default|SetText Text))
      (if (Utilities|Text|TextIsEmpty _output_get)
        ;[1 0 0.03977 0.25] If we don't want any text, discard the widget to free up render target
        ; allocation space.
        (UserInterface|SetWidget 0)
        (Variables|Internal|SetInitialized false)
        (elif (or (not (Variables|Internal|GetInitialized)) (!= (UserInterface|GetUserWidgetObject) (Variables|Internal|GetLoadedUserWidget)))
          ;[0 1 0.06881 0.25] Create the widget and cache it.
          (bind _loadeduserwidget (Variables|Internal|GetLoadedUserWidget)
            (:then
              (if (!= (UserInterface|GetUserWidgetObject) _loadeduserwidget)
                _))
            (:else
              (if _returnvalue
                (bind _returnvalue_2 (Utilities|LoadClassAssetBlocking _widget_class)) ; If we're editor we can't asnyc load, so block load\n(we're in a component here, so can't use common macro)
                (else
                  (bind _class (Utilities|AsyncLoadClassAsset _widget_class)
                    (:Completed
                      _))))
              (bind _asuser_widget (Utilities|Casting|CastToUserWidgetClass (Utilities|SelectClass _returnvalue_2 _class _returnvalue)))
              (bind _returnvalue_3 (UserInterface|CreateWidget _asuser_widget (Fortnite|GetFirstLocalFortPlayerController true)))
              (Variables|Internal|SetLoadedUserWidget _returnvalue_3)))
          (UserInterface|SetWidget (Variables|Internal|GetLoadedUserWidget))
          (Variables|Internal|SetInitialized true)
          (CallFunction|UpdateWidget)
          (else
            (CallFunction|UpdateWidget) ; We already have a valid widget, update immediately.
          ))))))

(event Custom|Update%20Color
  (params (Color :type Linear%20Color%20Structure))
  ;[0 0.4964 1 0.25] Updates the Color when asks the widget to update.
  (Networking|CanExecuteCosmeticEvents
    (:True
      (Variables|Default|SetColor (Math|Conversions|ToColor(LinearColor) Color))
      (CallFunction|UpdateWidget))))

(event Custom|Update%20Widget
  (bind _returnvalue (UserInterface|GetUserWidgetObject))
  ;[0 0.5023 1 0.25] Calls InitFromObject on the widget, which will then update itself based on this component (it grabs details from GetDetails). Includes an enforced one frame delay to avoid unnecessary refreshes when it's being initialized (e.g. color and text getting updated in same frame).
  (Utilities|IsValid _returnvalue
    (:Is%20Valid
      (if (not (Editor|InEditor))
        (Utilities|FlowControl|DelayUntilNextTick))
      (Init|InitfromObject(Message) _returnvalue self)
      (UserInterface|RequestRenderUpdate))))

(fn GetDetails
  (return (Variables|Default|GetText) (Variables|Default|GetText%20Size) (Math|Conversions|ToLinearColor(Color) (Variables|Default|GetColor))))

(fn UpdateVisibilityAndUpdateWidget
  (params (Visibility :type Boolean))
  ;[0 0.7788 1 0.25] Render target space stays allocated for invisible widgets, so when hiding we also set the text to null, which for this widget will release the widget.
  (Networking|CanExecuteCosmeticEvents
    (:True
      (Rendering|SetVisibility Visibility)
      (CallFunction|UpdateText (select Visibility (Variables|Default|GetText) "")))))
