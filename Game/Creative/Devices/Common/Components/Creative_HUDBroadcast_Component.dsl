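; Creative device component that gives players the score manager HUD and broadcasts score events for the owning device.
; ApplyHUDGA calls its parent, then grants and activates GA_ScoreManagerHUD_Passive on the player controller's ability system unless the Ability.Creative.ScoreManager.HUD tag is present.
; DoScoreEvent and DoScoreEventForParticipant send Minigame.Stat.Score gameplay events to a player or any controller, carrying the score as magnitude.
; Both skip sending when Should Score is false.
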(fn ApplyHUDGA
  (params (PlayerController :type Player%20Controller%20Object%20Reference))
  (bind _returnvalue (Ability|GetAbilitySystemComponent PlayerController))
  (|CallParentFunction PlayerController)
  (if (not (GameplayTags|HasMatchingGameplayTag _returnvalue "(TagName=\"Ability.Creative.ScoreManager.HUD\")"))
    (Ability|Advanced|GrantandActivateAbilityonSelf _returnvalue "/Game/Creative/Devices/Common/GameplayAbilities/GA_ScoreManagerHUD_Passive.GA_ScoreManagerHUD_Passive_C")))

(fn DoScoreEvent
  (params (Instigating%20Player :type Fort%20Player%20Controller%20Object%20Reference) (Score :type Integer) (Should%20Score :type Boolean :default true))
  (if Should%20Score
    (Ability|SendGameplayEventtoActor Instigating%20Player "(TagName=\"Minigame.Stat.Score\")" (Utilities|Struct|MakeGameplayEventData "(TagName=\"Minigame.Stat.Score\")" (Components|GetOwner) Instigating%20Player 0 0 "()" "(GameplayTags=)" "(GameplayTags=)" (Math|Conversions|ToFloat(Integer) Score)))))

(fn DoScoreEventForParticipant
  (params (Instigating%20Controller :type Controller%20Object%20Reference) (Score :type Integer) (Should%20Score :type Boolean :default true))
  (if Should%20Score
    (Ability|SendGameplayEventtoActor Instigating%20Controller "(TagName=\"Minigame.Stat.Score\")" (Utilities|Struct|MakeGameplayEventData "(TagName=\"Minigame.Stat.Score\")" (Components|GetOwner) Instigating%20Controller 0 0 "()" "(GameplayTags=)" "(GameplayTags=)" (Math|Conversions|ToFloat(Integer) Score)))))
