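; Creative device component that manages a header/description HUD widget, showing either the owner's description or island portal details.
; Spawning async-loads or reuses the widget, then fills it through the owner's header-details interface; without cosmetic event support, the component destroys itself.
; A new-matchmaking-portal-UI check switches between event dispatchers and direct widget calls.
; Show requests apply only for the local player pawn, with a redundancy check; forced shows bypass both.
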(event Custom|Force%20Show%20Widget
  (params (Show :type Boolean))
  (Variables|HeaderDescriptionHUDComponent|SetbDesiredVisibility Show)
  (if (CallFunction|IsNewMatchmakingPortalUIEnabled)
    (EventDispatchers|CallOnHUDShow (Variables|HeaderDescriptionHUDComponent|GetbDesiredVisibility) true)
    (else
      (bind _hudwidget (Variables|Default|GetHUDWidget)
        (:then
          (Class|CreativeHeaderDescriptionHUD|UpdateVisibility _hudwidget (Variables|HeaderDescriptionHUDComponent|GetbDesiredVisibility)))
        (:else)) ; HUD Widget may not have finished loading yet.
    )))

(event EventUpdateHUDWidget
  (Init|InitfromObject(Message) (Variables|HeaderDescriptionHUDComponent|GetLoadedHUDWidget) self)
  (CallFunction|UpdateWidget (Variables|Default|GetWidgetType)))

(event Custom|Show%20Widget
  (params (Requesting%20Object :type Object%20Reference) (Show :type Boolean))
  (if (and (Utilities|Equal(Object) (Game|GetPlayerPawn 0) Requesting%20Object) (CallFunction|ShouldApplyVisibility Show))
    (Variables|HeaderDescriptionHUDComponent|SetbDesiredVisibility Show)
    (if (CallFunction|IsNewMatchmakingPortalUIEnabled)
      (EventDispatchers|CallOnHUDShow (Variables|HeaderDescriptionHUDComponent|GetbDesiredVisibility) true)
      (else
        (bind _hudwidget (Variables|Default|GetHUDWidget)
          (:then
            (Class|CreativeHeaderDescriptionHUD|UpdateVisibility _hudwidget (Variables|HeaderDescriptionHUDComponent|GetbDesiredVisibility)))
          (:else)) ; HUD Widget may not have finished loading yet.
      ))))

(event Custom|Spawn%20or%20Update%20HUD
  (params (Widget%20Type :type ECreativeHUDWidgets%20Enum :default "NewEnumerator0"))
  (Variables|Default|SetWidgetType Widget%20Type)
  (Networking|CanExecuteCosmeticEvents
    (:True
      (if (CallFunction|IsNewMatchmakingPortalUIEnabled)
        (Variables|HeaderDescriptionHUDComponent|GetLoadedHUDWidget
          (:then
            (Variables|HeaderDescriptionHUDComponent|SetbHUDAlreadyExists true)
            (CallFunction|UpdateHUDWidget))
          (:else
            (Variables|HeaderDescriptionHUDComponent|SetbHUDAlreadyExists false)
            (CallFunction|LoadandCreateHUDWidget)))
        (else
          (Variables|Default|GetHUDWidget
            (:then
              (Variables|HeaderDescriptionHUDComponent|SetbHUDAlreadyExists true))
            (:else
              (Variables|HeaderDescriptionHUDComponent|SetbHUDAlreadyExists false)
              (bind _class (Utilities|AsyncLoadClassAsset (Variables|Default|GetHUD%20Class))
                (:Completed
                  (CallFunction|CreateWidget _class)))))
          (Init|InitfromObject(Message) (Variables|Default|GetHUDWidget) self)
          (CallFunction|UpdateWidget (Variables|Default|GetWidgetType))
          (Default|CallWidgetReady))))
    (:False
      (Components|DestroyComponent) ; If we can't execute cosmetic effects, it's a bit pointless \nspawning a UI. Destroy self instead.
    )))

(event EventEndPlay
  (params (EndPlayReason :type EEndPlayReason%20Enum :default "Destroyed"))
  (Utilities|IsValid (Variables|Default|GetHUDWidget)
    (:Is%20Valid
      (Default|UnbindallEventsfromOnHeaderHidden (Variables|Default|GetHUDWidget))
      (Default|UnbindallEventsfromOnHeaderAbouttobeDisplayed (Variables|Default|GetHUDWidget)))
    (:Is%20Not%20Valid
      _))
  (Utilities|IsValid (Variables|HeaderDescriptionHUDComponent|GetLoadedHUDWidget)
    (:Is%20Valid
      (EventDispatchers|UnbindallEventsfromOnHUDShow)
      (EventDispatchers|UnbindallEventsfromOnWidgetUpdate))
    (:Is%20Not%20Valid
      _))
  (|CallParentFunction EndPlayReason))

(event Custom|Update%20)

(fn CreateWidget
  (params (Class :type Object%20Class%20Reference))
  (bind _asuser_widget (Utilities|Casting|CastToUserWidgetClass Class))
  (bind _returnvalue (UserInterface|CreateWidget _asuser_widget (Game|GetPlayerController 0)))
  (UserInterface|SetWidget _returnvalue)
  (bind _ascreative_header_description_hud (Utilities|Casting|CastToCreative_HeaderDescriptionHUD _returnvalue))
  (Variables|Default|SetHUDWidget _ascreative_header_description_hud))

(fn ShouldApplyVisibility
  (params (DesiredVisibility :type Boolean))
  (return (or (Math|Boolean|NotEqual(Boolean) DesiredVisibility (Variables|HeaderDescriptionHUDComponent|GetbDesiredVisibility)) (Variables|Default|GetIsValidEqualVisibility))))

(fn UpdateWidget
  (params (Widget%20Type :type ECreativeHUDWidgets%20Enum :default "NewEnumerator0"))
  (bind _returnvalue (Components|GetOwner))
  (bind (_header _world_location _linkid) (Class|CreativeHeaderDescriptionHUDInterface|GetHeaderDetails(Message) (Components|GetOwner)))
  (LocalVariables|SetHeader _header)
  (LocalVariables|SetWorldLocation _world_location)
  (LocalVariables|SetLinkId _linkid)
  (if (CallFunction|IsNewMatchmakingPortalUIEnabled)
    (EventDispatchers|CallOnWidgetUpdate)
    (else
      (Class|CreativeHeaderDescriptionHUD|SetHeaderDetails (Variables|Default|GetHUDWidget) (LocalVariables|GetHeader) (LocalVariables|GetWorldLocation) (LocalVariables|GetLinkId))
      (switch Utilities|FlowControl|Switch|SwitchonECreativeHUDWidgets (Variables|GetWidget%20Type)
        (:NewEnumerator0
          (Class|CreativeHeaderDescriptionHUD|SetDescription (Variables|Default|GetHUDWidget) _returnvalue))
        (:NewEnumerator1
          (IslandPortal|UpdateIslandPortalDetails (Variables|Default|GetHUDWidget) _returnvalue))))))
