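; Blueprint interface that Creative devices implement to supply text and state for a header and description HUD.
; It declares four query functions covering header details, description, per-player-controller interaction details and island portal details.
; The bodies return only empty or default values, so implementing devices provide the actual content.
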(fn GetDescriptionDetails
  (return ""))

(fn GetHeaderDetails
  (return "" "0, 0, 0" 0))

(fn GetInteractionDetails
  (params (Controller :type Fort%20Player%20Controller%20Object%20Reference))
  (return "NewEnumerator0" 0))

(fn GetIslandPortalDetails
  (return "NewEnumerator0" "" "" "" 0 0))
