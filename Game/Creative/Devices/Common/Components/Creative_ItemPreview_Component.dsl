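; Creative device component that shows a rotating preview of the items a device holds, mainly outside play mode.
; Item list, spawn params and variant changes feed a queue that async-loads meshes one by one into static or skeletal components with rarity overlays and weapon mods.
; Tick scales previews in and out to cycle them and pulls newly queued pickups in with a black-hole material effect.
; Low-spec platforms skip ticking and use a timer.
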(event Custom|Play%20Mode%20Changed
  (params (Minigame :type Fort%20Minigame%20Object%20Reference) (bIsInPlayMode :type Boolean))
  ;[0 0.3279 1 0.2] Play Mode Switch - If we're in edit, load our overlay material. Then update preview.
  (if (not bIsInPlayMode)
    (bind _object (Utilities|AsyncLoadAsset (Variables|Rarity|GetRarity%20Override%20Material))
      (:Completed
        (bind _asmaterial_interface (Utilities|Casting|CastToMaterialInterface _object))
        (bind _returnvalue (Rendering|Material|CreateDynamicMaterialInstance _asmaterial_interface "None" "None" :!class /Script/Engine.KismetMaterialLibrary))
        (Variables|Rarity|SetRarity%20Override%20MID _returnvalue))))
  (CallFunction|UpdatePreviews))

(event EventEndPlay
  (params (EndPlayReason :type EEndPlayReason%20Enum :default "Destroyed"))
  ;[1 0 0.08598 0.5] Destruction
  (|CallParentFunction EndPlayReason)
  (CallFunction|BindOrUnbindEvents false)
  (CallFunction|ClearCurrentPickup))

(event EventTick
  (params (DeltaSeconds :type Float%20%28single-precision%29))
  ;[1 1 1] Tick - Rotate the mesh, scale it in and out during transitions, and pulled queued objects to the preview location.
  (Utilities|IsLowSpec
    (:No
      (bind _previewmeshscene (Variables|Default|GetPreviewMeshScene)
        (:then
          (Transformation|AddRelativeRotation _previewmeshscene (Math|Rotator|MakeRotator 0.0 0.0 (* DeltaSeconds (Variables|Default|GetRotation%20Rate))) false false)
          (CallFunction|ScaleAndSwapPreview)
          (bind _currently_pulling_item (CallFunction|PullItemToDevice))
          (if _currently_pulling_item
            (goto _label Reset))
          ;;[1 0.7441 0] Spawn Sound at Location once item has been pulled.
          (Utilities|FlowControl|DoOnce true :!label _label
            (:Completed
              (Audio|PlaySoundatLocation (Variables|Audio|GetObjectAddedToSpawnerSound) (Transformation|GetWorldLocation (Variables|Default|GetPreviewMeshScene))))))
        (:else
          (Components|Tick|SetComponentTickEnabled false) ; Don't need to tick if there is no PreviewMeshScene
        )))
    (:Yes
      (Components|Tick|SetComponentTickEnabled false) ; Mobile should never tick.
    )))

(event Custom|Item%20Added%20to%20Queue%20FX
  (params (Item :type Fort%20Pickup%20Athena%20Object%20Reference))
  ;[0 1 0] Called when an object has been added to the item queue
  (Networking|CanExecuteCosmeticEvents
    (:True
      (Utilities|IsLowSpec
        (:No
          (Utilities|IsValid Item
            (:Is%20Valid
              (Audio|PlaySoundatLocation (Variables|Audio|GetObjectAddedToQueueSound) (Transformation|GetActorLocation Item))
              (bind _output_get (Variables|MeshData|SetCurrentPickup Item))
              (Variables|Default|SetPickedUpWorldLocation (Transformation|GetActorLocation _output_get))
              (bind _asb_pickups_parent (Utilities|Casting|CastToB_Pickups_Parent (Utilities|Array|Get(acopy) (Actor|GetAttachedActors _output_get true false)))
                (:then
                  (Class|BPickupsParent|SetActiveBackgroundParticleSystem _asb_pickups_parent false false)
                  (bind _asb_pickups_default (Utilities|Casting|CastToB_Pickups_Default _asb_pickups_parent))
                  (Class|BPickupsParent|HideBackgroundandSpotlight _asb_pickups_default)
                  (Variables|MeshData|SetPickupMIDArray (Variables|/Script/FortniteGame.FortPickupsParent|GetMIDs _asb_pickups_parent)))
                (:CastFailed
                  ;;[0.8818 1 0.3073 0.5] Sometimes the cast to B_Pickups_Parent will fail if pickups are added to this event soon after spawning. If this happens, we should allow the event to continue so the pickup mesh will still be moved to the preview location just without the black hole effect.
                  (Utilities|Array|Clear (Variables|MeshData|GetPickupMIDArray))))
              (CallFunction|StartSuckToBackpack))))))))

(event Custom|LoadItemPreviewAsync
  (bind (_itemdefintion _returnvalue) (Creative|ItemData|GetItemDefinition (Variables|Default|GetNext%20Item) "/Script/FortniteGame.FortWorldItemDefinition"))
  ;[0 1 1 0.5] Load the next item to our 'Next Mesh Preview' struct and then show it if we're not already playing previews.
  (if (and (not _returnvalue) (Math|Conversions|ToBoolean(Integer) (Utilities|Array|Length (Variables|MeshData|GetItems%20to%20Load))))
    (Utilities|FlowControl|Delay 0.0) ; adding a 0.0s delay to ensure this only gets called once per frame.
    (bind _nextitem (CallFunction|SetNextItem (Utilities|Array|Get(acopy) (Variables|MeshData|GetItems%20to%20Load) (Variables|MeshData|GetNext%20Mesh%20Load%20Index))))
    (CallFunction|DebugMobileLogging (Utilities|String|ToString(Text) (Creative|ItemData|GetDisplayName _nextitem)) "Attempting to Load")
    (if (CallFunction|GetMeshObjectPathforItem (Creative|ItemData|GetItemDefinition _nextitem "/Script/FortniteGame.FortWorldItemDefinition"))
      (Utilities|SaveWeaponWrapforPreview (Creative|ItemData|GetItemDefinition _nextitem "/Script/FortniteGame.FortWorldItemDefinition"))
      (bind _object (Utilities|AsyncLoadAsset (.SoftMeshObjectPath (CallFunction|GetMeshObjectPathforItem (Creative|ItemData|GetItemDefinition _nextitem "/Script/FortniteGame.FortWorldItemDefinition"))))
        (:Completed
          (CallFunction|LoadItemPreviewPostAsync _object)))
      (else
        (Utilities|Array|RemoveIndex (Variables|MeshData|GetItems%20to%20Load) (Variables|MeshData|GetNext%20Mesh%20Load%20Index)) ; Current Index  is invalid, strip it out of our list.
        (CallFunction|ModMeshLoadIndextoArraySize) ; ensure load index is within bounds.
        (CallFunction|ItemLoadedAndDisplayed false)))))

(event Custom|LoadItemPreview_PostAsync
  (params (Mesh%20Object :type Object%20Reference))
  (bind (_bisvalid _newitem) (CallFunction|CreatePickupItemPreview Mesh%20Object))
  ;[0 1 1 0.5] Load the next item to our 'Next Mesh Preview' struct and then show it if we're not already playing previews.
  (if _bisvalid
    (Variables|Default|SetNextMeshPreview _newitem)
    (CallFunction|IncrementMeshLoadIndex 1)
    (PreviewTicking|ShowPreviewAndStartTick)))

(event Custom|Item%20Loaded%20and%20Displayed
  (params (Successfully%20Loaded :type Boolean))
  ;[0 1 1 0.5] Load the next item to our 'Next Mesh Preview' struct and then show it if we're not already playing previews.
  ;;[1 1 1] Clear the Loading Item and attempt to load the next preview.
  (Variables|Default|SetNext%20Item)
  (Utilities|ClearMeshPreview)
  (Variables|MeshData|SetCurrently%20Previewing%20Items Successfully%20Loaded)
  (if (> (Utilities|Array|Length (Variables|MeshData|GetItems%20to%20Load)) 1)
    (CallFunction|LoadItemPreviewAsync)))

(event Custom|Update%20Item%20List%20from%20Spawn%20Params
  (params (Spawn%20List :type Array%20of%20Minigame%20Spawner%20Spawn%20Params%20Structures))
  ;[0 0.6006 1 0.5] Called from Owning Actor. If we have a list of Spawn Item Params (i.e. likely a device using MinigameSpawnComponent, or an older trap), load the assets and once everything is loaded update the item list.
  (Utilities|Array|Clear (Variables|Default|GetPrimaryAssetsToLoad))
  (Utilities|Array|Clear (Variables|MeshData|GetItems%20to%20Load))
  (Utilities|FlowControl|Sequence
    (:then_0
      (bind _output_get (Variables|Default|SetSpawnParamsToLoad Spawn%20List))
      (for _array_element _output_get
        (bind _array_element_pickuptospawn (Utilities|Struct|BreakMinigameSpawnerSpawnParams _array_element))
        (if (AssetManager|IsValidPrimaryAssetId (AssetManager|GetPrimaryAssetIdfromSoftObjectReference _array_element_pickuptospawn))
          (Utilities|Array|AddUnique (Variables|Default|GetPrimaryAssetsToLoad) (AssetManager|GetPrimaryAssetIdfromSoftObjectReference _array_element_pickuptospawn)))))
    (:then_1
      (if (> (Utilities|Array|Length (Variables|Default|GetPrimaryAssetsToLoad)) 0)
        (Math|Integer|IncrementInt (Variables|Default|GetPendingLoadCount)
          (:%20%20
            (AssetManager|AsyncLoadPrimaryAssetList (Variables|Default|GetPrimaryAssetsToLoad)
              (:then
                (Utilities|Array|Clear (Variables|Default|GetPrimaryAssetsToLoad)))
              (:Completed
                (bind _result (Math|Integer|DecrementInt (Variables|Default|GetPendingLoadCount))
                  (:%20%20
                    (if (== _result 0)
                      _))))))))
      (CallFunction|ParseandConsumeSpawnParams)
      (CallFunction|UpdateItemList))))

(event Custom|OnItemTemplateChanged
  (params (ItemTemplate :type Fort%20Creative%20Item%20Template%20Object%20Reference))
  (CallFunction|UpdateItemList))

(fn AddModsToWeaponPreview
  (params (ModWeaponPreview :type Fort%20Weapon%20Object%20Reference))
  ;[0.4644 1 0.8795 0.5] Set up the mod mesh components
  (for (_array_element _array_index) (Class|FortWeapon|GetWeaponModSlots ModWeaponPreview)
    (bind _array_element_weaponmod (Utilities|Struct|BreakFortWeaponModSlot _array_element))
    (Utilities|IsValid _array_element_weaponmod
      (:Is%20Valid
        (bind _spawnedcomponent (Fortnite|Utilities|SpawnTransientSceneComponent "/Script/Engine.StaticMeshComponent" (Components|GetOwner) (Variables|Default|GetPreviewSkeletalMesh)))
        (Utilities|Array|SetArrayElem (Variables|Default|GetModMeshComps) _array_index _spawnedcomponent true)
        (bind _returnvalue (Class|FortWeaponModFunctionLibrary|FindPreviewModMeshComponent _array_element_weaponmod (Variables|GetModWeaponPreview)))
        (Utilities|IsValid _returnvalue
          (:Is%20Valid
            (Components|StaticMesh|SetStaticMesh _spawnedcomponent (Variables|/Script/Engine.StaticMeshComponent|GetStaticMesh _returnvalue))
            (Transformation|SetWorldTransform _spawnedcomponent (Transformation|GetWorldTransform _returnvalue) false false)))))))

(fn ApplyRarityOverrideMID
  (params (MeshComp :type Mesh%20Component%20Object%20Reference) (RarityColor :type Linear%20Color%20Structure))
  (bind _rarity_override_mid (Variables|Rarity|GetRarity%20Override%20MID)
    (:then
      (Rendering|Material|SetOverlayMaterial MeshComp _rarity_override_mid false 0 :!class /Script/Engine.MeshComponent)
      (Rendering|Material|SetOverlayMaterialMaxDrawDistance MeshComp (CallFunction|GetOverlayDistance) :!class /Script/Engine.MeshComponent)
      (Rendering|Material|SetVectorParameterValue _rarity_override_mid (Variables|Rarity|GetRarity%20Color%20Param%20Name) RarityColor :!class /Script/Engine.MaterialInstanceDynamic))
    (:else
      (Rendering|Material|SetOverlayMaterial MeshComp 0 false 0 :!class /Script/Engine.MeshComponent) ; Clear the overlay material
    )))

(fn ApplyWeaponWrapMaterial
  (Variables|Default|GetCurrentWrapMaterial
    (:then
      (if (> (Variables|Default|GetCurrentWrapSectionMask) -1)
        (for (_array_element _array_index) (Rendering|Material|GetMaterials (Variables|Default|GetPreviewSkeletalMesh))
          ;[0.3065 0.7188 0.566 0.5] Shift and bitwise operation taken from AWrapPreviewGridActor::ApplyWrapToActor() to check the section mask
          ; which ensures we only apply the wrap to specific material indexes.
          (if (!= (Math|Integer|BitwiseAND (Variables|Default|GetCurrentWrapSectionMask) (Math|Float|Truncate (Math|Float|Power 2.0 (Math|Conversions|ToFloat(Integer) _array_index)))) 0)
            (bind _output_get (LocalVariables|SetBaseWeaponMat _array_element))
            (bind _output_get_2 (LocalVariables|SetOverrideWeaponMat (Material|GetMIDfromMaterial (Variables|Default|GetPreviewSkeletalMesh) _array_index (Variables|Default|GetCurrentWrapMaterial))))
            (Rendering|Material|CopyMaterialInstanceParameters _output_get_2 _output_get true)))
        (Variables|Default|SetCurrentWrapMaterial 0)
        (Variables|Default|SetCurrentWrapSectionMask -1)))
    (:else)))

(fn BindorUnbindEvents
  (params (Bind :type Boolean))
  (bind _returnvalue (Components|GetOwner))
  (bind _returnvalue_2 (Minigame|GetMinigame _returnvalue :!class /Script/FortniteGame.MinigameBlueprintLibrary))
  (bind _outputdelegate (EventDispatchers|CreateEvent :!event Update%20Item%20List))
  (bind _outputdelegate_2 (EventDispatchers|CreateEvent :!event Play%20Mode%20Changed))
  (bind _outputdelegate_3 (EventDispatchers|CreateEvent :!event OnItemTemplateChanged))
  (Utilities|IsValid _returnvalue_2
    (:Is%20Valid
      (if (Variables|GetBind)
        (Minigame|BindEventtoOnPlayModeChangedDelegate _returnvalue_2 _outputdelegate_2)
        (CallFunction|PlayModeChanged _returnvalue_2 (Minigame|IsinPlayMode _returnvalue_2 :!class /Script/FortniteGame.FortMinigame))
        (bind _pickupitemlist (Variables|Default|GetPickupItemList)
          (:then
            (Default|BindEventtoOnItemListUpdated _pickupitemlist _outputdelegate)
            (CallFunction|UpdateItemList))
          (:else
            _))
        (bind _variantnotifier (Variables|Default|GetVariantNotifier)
          (:then
            (ItemTemplate|BindEventtoOnItemTemplateChanged _variantnotifier _outputdelegate_3))
          (:else))
        (else
          (Minigame|UnbindEventfromOnPlayModeChangedDelegate _returnvalue_2 _outputdelegate_2)
          (bind _pickupitemlist_2 (Variables|Default|GetPickupItemList)
            (:then
              (Default|UnbindEventfromOnItemListUpdated _pickupitemlist_2 _outputdelegate))
            (:else
              _))
          (bind _variantnotifier_2 (Variables|Default|GetVariantNotifier)
            (:then
              (ItemTemplate|UnbindEventfromOnItemTemplateChanged _variantnotifier_2 _outputdelegate_3))
            (:else)))))))

(fn ClearCurrentPickup
  ;[1 1 1] Clears the current pickup, if any, resetting any currently playing 'sucking' material effects (to avoid the material being stuck halfway through an effect)
  (Variables|MeshData|GetCurrentPickup
    (:then
      (CallFunction|ClearWeaponMods true)
      (for _array_element (Variables|MeshData|GetPickupMIDArray)
        (Utilities|IsValid _array_element
          (:Is%20Valid
            (Rendering|Material|SetVectorParameterValue _array_element "WorldPostionOfBlackHole" "(R=0.000000,G=0.000000,B=0.000000,A=0.000000)" :!class /Script/Engine.MaterialInstanceDynamic) ; return 0,0,0,0
          )))
      (Variables|MeshData|SetCurrentPickup 0))
    (:else)))

(fn ClearPreviews
  (Utilities|ClearMeshPreview)
  (CallFunction|StopPreviews))

(fn ClearWeaponMods
  (params (bClearModdableWID :type Boolean :default true))
  (for _array_element (Variables|Default|GetModMeshComps)
    (Components|StaticMesh|SetStaticMesh _array_element 0))
  (if bClearModdableWID
    (Variables|Default|SetNextModdableWID 0)))

(fn CreatePickupItemPreview
  (params (Mesh%20Object :type Object%20Reference))
  (bind _next_item (Variables|Default|GetNext%20Item))
  (bind _rarity (Creative|ItemData|GetRarity _next_item))
  (bind _returnvalue (Math|Conversions|ToInteger(Byte) _rarity))
  (Utilities|FlowControl|Sequence
    (:then_0
      (Utilities|Casting|CastToStaticMesh Mesh%20Object
        (:then
          _)
        (:CastFailed
          (Utilities|Casting|CastToSkeletalMesh Mesh%20Object
            (:then
              _)
            (:CastFailed
              (CallFunction|DebugMobileLogging (Utilities|GetDisplayName Mesh%20Object) " Create Pickup Item Preview. Mesh was not a Static or Skeletal" true)))))
      (bind _output_get (LocalVariables|SetWorldItemDef (Creative|ItemData|GetItemDefinition (Variables|Default|GetNext%20Item) "/Script/FortniteGame.FortWorldItemDefinition")))
      (Utilities|IsValid _output_get
        (:Is%20Valid
          (if (CallFunction|GetMeshObjectPathforItem _output_get)
            (LocalVariables|SetMeshObjectPath (.SoftMeshObjectPath (CallFunction|GetMeshObjectPathforItem _output_get)))
            (bind _meshtransform (Class|FortItemFunctionLibrary|GetPickupMeshTransform _output_get)
              (:IsIdentity
                _)
              (:IsOverridden
                (LocalVariables|SetTransform _meshtransform)))
            (Utilities|FlowControl|Sequence
              (:then_0
                ;[1 0.7005 0.2541] Keep track of the item definition if it has any mods
                (Variables|Default|SetNextModdableWID 0)
                (bind (_bvaliddefinition _fortweaponitemdefinition) (CallFunction|GetFortWeaponItemDef (LocalVariables|GetWorldItemDef)))
                (if _bvaliddefinition
                  (bind _array_element (Utilities|Array|ForEachLoopwithBreak (WeaponItemDefinition|GetWeaponModSlots _fortweaponitemdefinition) :!label _label)
                    (:LoopBody
                      (bind _array_element_weaponmod (Utilities|Struct|BreakFortWeaponModSlot _array_element))
                      (Utilities|IsValid _array_element_weaponmod
                        (:Is%20Valid
                          (Variables|Default|SetNextModdableWID _fortweaponitemdefinition)
                          (goto _label Break))))
                    (:Completed
                      _)))
                (if (Utilities|Array|IsValidIndex (Variables|Default|GetRaritySourceColors) _returnvalue)
                  (LocalVariables|SetRarityColor (Utilities|Array|Get(aref) (Variables|Default|GetRaritySourceColors) _returnvalue))
                  (else
                    (LocalVariables|SetRarityColor (Utilities|Array|Get(aref) (Variables|Default|GetRaritySourceColors))))))
              (:then_1
                (LocalVariables|SetbSuccess true)
                (Variables|SetOutputItem_MeshAssetPath_8_6B40C41C47213E2D7D041A876713C4F5 (Utilities|Struct|MakePickupItemPreview (LocalVariables|GetMeshObjectPath) (Variables|GetMesh%20Object) (LocalVariables|GetRarityColor) (LocalVariables|GetTransform)))))))
        (:Is%20Not%20Valid
          (CallFunction|DebugMobileLogging (Utilities|GetDisplayName _output_get) " Create Pickup Item Preview: Item Object was Invalid" true))))
    (:then_1
      (return (LocalVariables|GetbSuccess) (LocalVariables|GetOutputItem)))))

(fn DebugMobileLogging
  (params (Dynamic%20String :type String) (Suffix :type String) (Warning :type Boolean) (Also%20Show%20on%20High%20Spec :type Boolean))
  (bind _returnvalue (Utilities|String|Append "Creative Item Preview - " Dynamic%20String " : " Suffix))
  (Utilities|IsLowSpec
    (:No
      (if Also%20Show%20on%20High%20Spec
        (Development|PrintString _returnvalue true true "(R=1.000000,G=0.000000,B=0.452495,A=1.000000)" 15.0)))
    (:Yes
      (Fortnite|FortShippingLog (Components|GetOwner) _returnvalue Warning))))

(fn EditorDisplayPreview
  (params (Preview%20Static%20Mesh%20Comp :type Static%20Mesh%20Component%20Object%20Reference))
  (bind _returnvalue (Components|GetOwner))
  (bind _returnvalue_2 (Actor|GetComponentbyClass _returnvalue "/Script/FortniteGame.FortMinigameSpawnerComponent"))
  (bind _tospawnlist (Variables|/Script/FortniteGame.FortMinigameSpawnerComponent|GetToSpawnList _returnvalue_2))
  (bind _output_pickuptospawn (Utilities|Struct|BreakMinigameSpawnerSpawnParams (Utilities|Array|Get(acopy) _tospawnlist)))
  (bind _previewstaticmesh (Variables|Default|GetPreviewStaticMesh))
  (bind _item_def (LocalVariables|GetItem%20Def))
  (bind _validitem (LocalVariables|GetValidItem))
  (bind _output_get (Variables|Default|SetPreviewStaticMesh (Variables|GetPreview%20Static%20Mesh%20Comp)))
  (Utilities|IsValid _output_get
    (:Is%20Valid
      (Utilities|FlowControl|Sequence
        (:then_0
          ;[1 0.9002 0 0.25] Override the primitive's material with a 2d facing icon.
          (bind _returnvalue_3 (Utilities|LoadAssetBlocking "/Game/Creative/Devices/Generic/Meshes/CP_Device_Generic_Icon_01.CP_Device_Generic_Icon_01"))
          (bind _asstatic_mesh (Utilities|Casting|CastToStaticMesh _returnvalue_3))
          (Components|StaticMesh|SetStaticMesh (Variables|Default|GetPreviewStaticMesh) _asstatic_mesh)
          ;[0 0.7015 1 0.25] Get Appropriate item definition from the pickup item list or minigame spawner list (item list checked first, since in the case where a device contains both, initially they'll be stored in the pickup list)
          (bind _output_get_2 (LocalVariables|SetLocal%20Item%20List%20 (Actor|GetComponentbyClass (Components|GetOwner) "/Game/Creative/Devices/CaptureArea/PickupItemListComponent.PickupItemListComponent_C")))
          (Utilities|IsValid _output_get_2
            (:Is%20Valid
              (if (Utilities|Array|IsValidIndex (ItemData|GetContainedItems _output_get_2))
                (LocalVariables|SetItem%20Def (Utilities|Array|Get(acopy) (ItemData|GetContainedItems _output_get_2)))))
            (:Is%20Not%20Valid
              (bind _output_get_3 (LocalVariables|SetLocal%20Spawner%20Comp _returnvalue_2))
              (Utilities|IsValid _output_get_3
                (:Is%20Valid
                  (if (Utilities|Array|IsValidIndex _tospawnlist)
                    (bind _returnvalue_4 (Utilities|LoadAssetBlocking _output_pickuptospawn))
                    (bind _asfort_world_item_definition (Utilities|Casting|CastToFortWorldItemDefinition _returnvalue_4))
                    (LocalVariables|SetItem%20Def _asfort_world_item_definition)))
                (:Is%20Not%20Valid
                  (bind _output_get_4 (LocalVariables|SetLocalKeyLock (Actor|GetComponentbyClass (Components|GetOwner) "/Script/FortniteGame.FortCreativeKeyLockComponent")))
                  (Utilities|IsValid _output_get_4
                    (:Is%20Valid
                      (bind _asfort_world_item_definition_2 (Utilities|Casting|CastToFortWorldItemDefinition (Class|FortCreativeKeyLockComponent|GetItemDefinitionForItem _output_get_4)))
                      (LocalVariables|SetItem%20Def _asfort_world_item_definition_2)))))))
          ;[1 0.9002 0 0.25] Load Texture from Item Def and Set it on the mesh.
          (Utilities|IsValid _item_def
            (:Is%20Valid
              (bind _returnvalue_5 (Utilities|LoadAssetBlocking (Item|GetLargePreviewImage _item_def)))
              (bind _astexture (Utilities|Casting|CastToTexture _returnvalue_5))
              (Rendering|Material|SetTextureParameterValue (CallFunction|GetStaticMeshPreviewTransientMID) "Texture" _astexture)
              (LocalVariables|SetValidItem true))))
        (:then_1
          ;[1 0.9002 0 0.25] Display if valid, else hide.
          (Rendering|SetVisibility _previewstaticmesh _validitem false)
          (Development|SetHiddeninGame _previewstaticmesh (not _validitem) false)
          (Default|CallItemsChanged _validitem))))))

(fn FindStaticMeshComponent
  (CallFunction|DebugMobileLogging "Find Static Mesh Component" "Static Mesh Component is Invalid. Searching for New One" true)
  (bind _previewmeshscene (Variables|Default|GetPreviewMeshScene)
    (:then
      (for _array_element (Components|GetChildrenComponents _previewmeshscene false)
        (bind _asstatic_mesh_component (Utilities|Casting|CastToStaticMeshComponent _array_element))
        (Variables|Default|SetPreviewStaticMesh _asstatic_mesh_component)
        (return true))
      (CallFunction|DebugMobileLogging "Find Static Mesh Component" "Static Mesh Component has been Destroyed" true)
      (return false))
    (:else
      (CallFunction|DebugMobileLogging "Find Static Mesh Component" "Preview Mesh Scene is also Invalid" true)
      (return false))))

(fn GetOverlayDistance
  (bind (_returnvalue_x _returnvalue_y _returnvalue_z) (Math|Vector|BreakVector (Transformation|GetActorScale3D (Components|GetOwner))))
  (return (* (* (Math|Float|Max(Float) (Math|Float|Absolute(Float) _returnvalue_x) (Math|Float|Absolute(Float) _returnvalue_y) (Math|Float|Absolute(Float) _returnvalue_z)) (Variables|Default|GetBase%20Scale)) (Variables|Rarity|GetRarityOverlayDistance))))

(fn GetPreviewScale
  (Utilities|IsValid (Components|GetOwner)
    (:Is%20Valid
      (return (* (Transformation|GetActorScale3D (Components|GetOwner)) (Math|Float|Max(Float) (Variables|Default|GetItem%20Preview%20Scale%20Multiplier) 0.01))))
    (:Is%20Not%20Valid
      (return (Math|Float|Max(Float) (Variables|Default|GetItem%20Preview%20Scale%20Multiplier) 0.01) (Math|Float|Max(Float) (Variables|Default|GetItem%20Preview%20Scale%20Multiplier) 0.01) (Math|Float|Max(Float) (Variables|Default|GetItem%20Preview%20Scale%20Multiplier) 0.01)))))

(fn GetStaticMeshPreviewTransientMID
  (bind _previewstaticmesh (Variables|Default|GetPreviewStaticMesh))
  (bind _staticmesh (Variables|/Script/Engine.StaticMeshComponent|GetStaticMesh _previewstaticmesh))
  (bind _returnvalue (StaticMesh|GetMaterial _staticmesh 0))
  (bind _uniquename (Material|GenerateUniqueMaterialName _previewstaticmesh (Utilities|GetDisplayName _returnvalue)))
  (bind _returnvalue_2 (Rendering|Material|CreateDynamicMaterialInstance _returnvalue _uniquename "Transient" :!class /Script/Engine.KismetMaterialLibrary))
  (Rendering|Material|SetMaterial (Variables|Default|GetPreviewStaticMesh) 0 _returnvalue_2)
  (return _returnvalue_2))

(fn GetFortWeaponItemDef
  (params (FortWorldItemDefinition :type Fort%20World%20Item%20Definition%20Object%20Reference))
  (bind _asfort_weapon_item_definition (Utilities|Casting|CastToFortWeaponItemDefinition FortWorldItemDefinition)
    (:then
      (return true _asfort_weapon_item_definition))
    (:CastFailed
      (bind _asfort_gadget_item_definition (Utilities|Casting|CastToFortGadgetItemDefinition FortWorldItemDefinition)
        (:then
          (Utilities|IsValid (GadgetItemDefinition|GetWeaponItemDefinition _asfort_gadget_item_definition)
            (:Is%20Valid
              (return true (GadgetItemDefinition|GetWeaponItemDefinition _asfort_gadget_item_definition)))
            (:Is%20Not%20Valid
              _)))
        (:CastFailed
          _))
      (return false 0))))

(fn GetMeshObjectPathForItem
  (params (World%20Item%20Definition :type Fort%20World%20Item%20Definition%20Object%20Reference))
  (Utilities|IsValid World%20Item%20Definition
    (:Is%20Valid
      (bind _itemdata (Class|ItemDefinitionBase|GetItemData World%20Item%20Definition)
        (:Found
          (if (Utilities|IsValidSoftObjectReference (Utilities|Struct|BreakFortnitePickupItemData _itemdata))
            (return true (Utilities|Struct|BreakFortnitePickupItemData _itemdata))
            (elif (Utilities|IsValidSoftObjectReference (.PickupSkeletalMesh (Utilities|Struct|BreakFortnitePickupItemData _itemdata)))
              (return true (.PickupSkeletalMesh (Utilities|Struct|BreakFortnitePickupItemData _itemdata)))
              (else
                (return false 0)))))))))

(fn IncrementMeshLoadIndex
  (params (Increment%20Amount :type Integer :default 1))
  (if (Math|Conversions|ToBoolean(Integer) (Utilities|Array|Length (Variables|MeshData|GetItems%20to%20Load)))
    (Variables|MeshData|SetNext%20Mesh%20Load%20Index (Math|Integer|%25%28Integer%29 (+ (Variables|MeshData|GetNext%20Mesh%20Load%20Index) Increment%20Amount) (Utilities|Array|Length (Variables|MeshData|GetItems%20to%20Load))))
    (else
      (Variables|MeshData|SetNext%20Mesh%20Load%20Index 0))))

(fn InitializeComponent
  (params (PreviewMeshScene :type Scene%20Component%20Object%20Reference) (PreviewStaticMesh :type Static%20Mesh%20Component%20Object%20Reference) (PreviewSkeletalMesh :type Skeletal%20Mesh%20Component%20Object%20Reference) (Preview%20Audio%20Cue :type Audio%20Component%20Object%20Reference) (PickupItemList :type Pickup%20Item%20List%20Component%20Object%20Reference) (VairantNotifier :type Fort%20Creative%20Item%20Variant%20Notifier%20Component%20Object%20Reference))
  (Networking|CanExecuteCosmeticEvents
    (:True
      (Variables|Default|SetPreviewMeshScene PreviewMeshScene)
      (Variables|Default|SetPreviewStaticMesh PreviewStaticMesh)
      (Variables|MeshData|SetDefault%20Static%20Mesh (Variables|/Script/Engine.StaticMeshComponent|GetStaticMesh PreviewStaticMesh))
      (Variables|Default|SetPreviewSkeletalMesh PreviewSkeletalMesh)
      (Variables|Default|SetPreview%20Audio%20Cue Preview%20Audio%20Cue)
      (Variables|Default|SetPickupItemList PickupItemList)
      (Variables|Default|SetVariantNotifier VairantNotifier)
      (CallFunction|BindOrUnbindEvents true)
      (CallFunction|UpdatePreviews)
      (bind _asbuilding_actor (Utilities|Casting|CastToBuildingActor (Components|GetOwner)))
      (Class|BuildingActor|AddUniqueTransientOverride _asbuilding_actor (Utilities|String|StringToName (Utilities|GetObjectName (Variables|Default|GetPreviewStaticMesh))) "StaticMesh" true))))

(fn ItemSwitchTime
  (bind (_mintime _maxtime) (Math|Curves|GetTimeRange (Variables|Curves|GetItem%20Scale%20Curve)))
  ;[1 1 1] Switch when we reach the end of our scale curve.
  (return (Math|Float|Max(Float) _maxtime 1.0)))

(fn LoadandStartPreviews
  ;[0 0.6006 1 0.5] Load Meshes as required, fade in the audio cue. Starting the tick is done from the Load Items function once all of the items are loaded.
  (Networking|CanExecuteCosmeticEvents
    (:True
      (CallFunction|LoadItemPreviewAsync)
      (bind _preview_audio_cue (Variables|Default|GetPreview%20Audio%20Cue)
        (:then
          (Audio|Components|Audio|FadeIn _preview_audio_cue 0.0 1.0 0.0 "Linear" :!class /Script/Engine.AudioComponent))
        (:else)))))

(fn ModMeshLoadIndexToArraySize
  (bind _items_to_load (Variables|MeshData|GetItems%20to%20Load))
  (bind _returnvalue (Utilities|Array|Length _items_to_load))
  ;[1 1 1] Modulo the Mesh Load Index to the array length to clamp it within bounds
  ; (A check whenever we remove an object from the list / clear it to ensure
  (if (Math|Conversions|ToBoolean(Integer) _returnvalue)
    (Variables|MeshData|SetNext%20Mesh%20Load%20Index (Math|Integer|%25%28Integer%29 (Variables|MeshData|GetNext%20Mesh%20Load%20Index) _returnvalue))
    (else
      (Variables|MeshData|SetNext%20Mesh%20Load%20Index 0))))

(fn ParseAndConsumeSpawnParams
  (for _array_element (Variables|Default|GetSpawnParamsToLoad)
    (bind (_array_element_pickuptospawn _array_element_pickupquantity _array_element_pickupinstigatorhandle _array_element_itemvariantguid) (Utilities|Struct|BreakMinigameSpawnerSpawnParams _array_element))
    (Utilities|IsValid (Utilities|ResolveSoftReference _array_element_pickuptospawn)
      (:Is%20Valid
        (Utilities|Array|Add (Variables|MeshData|GetItems%20to%20Load) (Utilities|Struct|MakeItemVariantHandle (Utilities|ResolveSoftReference _array_element_pickuptospawn) _array_element_itemvariantguid)))))
  (Utilities|Array|Clear (Variables|Default|GetSpawnParamsToLoad)))

(fn PullItemtoDevice
  (bind (_mintime _maxtime) (Math|Curves|GetTimeRange (Variables|Curves|GetMove%20Mesh%20to%20Center%20Curve)))
  (bind (_mintime_2 _maxtime_2) (Math|Curves|GetTimeRange (Variables|Curves|GetSuck%20to%20Backpack%20Curve)))
  (bind (_returnvalue_x _returnvalue_y _returnvalue_z) (Math|Vector|BreakVector (+ (Transformation|GetWorldLocation (Variables|Default|GetPreviewMeshScene)) (Variables|Default|GetBackpackLocationOffset))))
  (LocalVariables|SetTime (- (Utilities|Time|GetGameTimeinSeconds) (Variables|Timekeeping|GetSuckToBackpackTime)))
  (if (<= (LocalVariables|GetTime) (Math|Float|Max(Float) _maxtime_2 _maxtime))
    (bind _currentpickup (Variables|MeshData|GetCurrentPickup)
      (:then
        (Transformation|SetActorLocation _currentpickup (Math|Vector|Lerp(Vector) (Variables|Default|GetPickedUpWorldLocation) (+ (Transformation|GetWorldLocation (Variables|Default|GetPreviewMeshScene)) (Variables|Default|GetHoverLocationOffset)) (Math|Float|Min(Float) (Math|Curves|GetFloatValue (Variables|Curves|GetMove%20Mesh%20to%20Center%20Curve) (LocalVariables|GetTime)) 1.0)) false false)
        (for _array_element (Variables|MeshData|GetPickupMIDArray)
          (Rendering|Material|SetVectorParameterValue _array_element "WorldPostionOfBlackHole" (Math|Color|MakeColor _returnvalue_x _returnvalue_y _returnvalue_z (Math|Curves|GetFloatValue (Variables|Curves|GetSuck%20to%20Backpack%20Curve) (LocalVariables|GetTime))) :!class /Script/Engine.MaterialInstanceDynamic))
        (return true))
      (:else
        (return false)))
    (else
      (return false))))

(fn RecenterSkeletalMesh
  (params (OptionalAttachComponent :type Scene%20Component%20Object%20Reference))
  (bind _returnvalue_origin (Math|BoxSphereBounds|BreakBoxSphereBounds (Fortnite|Utility|GetLocalBounds (Variables|Default|GetPreviewSkeletalMesh))))
  (bind _previewskeletalmesh (Variables|Default|GetPreviewSkeletalMesh))
  (bind _returnvalue (Transformation|GetAllSocketNames _previewskeletalmesh))
  ;[1 1 1] Many weapons have offset pivots. Get center mass throws NaN errors, so this is my workaround to getting the offet to re-center skel meshes
  (Variables|Default|GetPreviewSkeletalMesh
    (:then
      (for _array_element _returnvalue
        (LocalVariables|SetSkeletalMeshCenter (+ (Transformation|GetSocketLocation (Variables|Default|GetPreviewSkeletalMesh) _array_element) (LocalVariables|GetSkeletalMeshCenter))))
      (Transformation|SetRelativeLocation (Variables|Default|GetPreviewSkeletalMesh) (Math|Vector|MakeVector (.x (* _returnvalue_origin -1.0)) (.y (* _returnvalue_origin -1.0)) (Math|Float|SelectFloat (.z (* _returnvalue_origin -1.0)) (.z (/ (LocalVariables|GetSkeletalMeshCenter) (Utilities|Array|Length _returnvalue))) (>= (.z (* _returnvalue_origin -1.0)) (.z (/ (LocalVariables|GetSkeletalMeshCenter) (Utilities|Array|Length _returnvalue)))))) false false)
      (bind _optionalattachcomponent (Variables|GetOptionalAttachComponent)
        (:then
          (Transformation|AttachComponentToComponent (Variables|Default|GetPreviewSkeletalMesh) _optionalattachcomponent "None" "KeepRelative" "KeepRelative" "KeepRelative" true))
        (:else)))
    (:else)))

(fn ResetMaterials
  (params (Mesh :type Primitive%20Component%20Object%20Reference))
  (for _index (range (Rendering|Material|GetNumMaterials Mesh))
    (Rendering|Material|SetMaterial Mesh _index 0) ; Reset Material
  ))

(fn ScaleandSwapPreview
  (bind (_mintime _maxtime) (Math|Curves|GetTimeRange (Variables|Curves|GetItem%20Scale%20Curve)))
  (Variables|Default|SetCurrent%20Scale%20Time (Math|Float|Min(Float) (Math|Float|SelectFloat _maxtime (/ _maxtime 2.0) (and (Utilities|IsValid (Utilities|Struct|BreakPickupItemPreview (Variables|Default|GetNextMeshPreview)) :!class /Script/Engine.KismetSystemLibrary) (> (Utilities|Array|Length (Variables|MeshData|GetItems%20to%20Load)) 1))) (+ (Variables|Default|GetCurrent%20Scale%20Time) (Utilities|Time|GetWorldDeltaSeconds))))
  (Transformation|SetRelativeScale3D (Variables|Default|GetPreviewMeshScene) (Math|Conversions|ToVector(Float) (Math|Float|Max(Float) (Math|Float|float*float (Variables|Default|GetItem%20Preview%20Scale%20Multiplier) (Math|Curves|GetFloatValue (Variables|Curves|GetItem%20Scale%20Curve) (Variables|Default|GetCurrent%20Scale%20Time)) (Variables|Default|GetBase%20Scale)) 0.01)))
  (if (>= (Variables|Default|GetCurrent%20Scale%20Time) (CallFunction|ItemSwitchTime))
    (CallFunction|ShowNextPreviewMesh)))

(fn SetEnabled
  (params (Enable :type Boolean))
  (Variables|Default|SetEnabled Enable)
  (CallFunction|UpdatePreviews))

(fn SetNextItem
  (params (ItemHandle :type Item%20Variant%20Handle%20Structure))
  (bind _output_get (Variables|Default|SetNext%20Item ItemHandle))
  (Utilities|FlowControl|Sequence
    (:then_0
      ;[0 0.6299 1] Bind to item vairant change events (or unbind entirely if Next Item is invalid)
      (bind _variantnotifier (Variables|Default|GetVariantNotifier)
        (:then
          (ItemTemplate|BindtoItemVariant _variantnotifier (Variables|Default|GetNext%20Item)))
        (:else)))
    (:then_1
      (return _output_get))))

(fn SetSkeletalMeshHidden
  (params (Hidden :type Boolean))
  (bind _previewskeletalmesh (Variables|Default|GetPreviewSkeletalMesh))
  (Development|SetHiddeninGame _previewskeletalmesh Hidden false)
  (if Hidden
    (Components|Activation|Deactivate _previewskeletalmesh)
    (else
      (Components|Activation|Activate _previewskeletalmesh false))))

(fn ShouldShowPreviews
  (return (Math|Boolean|ANDBoolean (not (Class|CreativeCommonDeviceFunctionLibrary|IsinPlayMode (Components|GetOwner))) (Math|Conversions|ToBoolean(Integer) (Utilities|Array|Length (Variables|MeshData|GetItems%20to%20Load))) (Variables|Default|GetEnabled))))

(fn ShowPreviewandStartTick
  (Utilities|IsLowSpec
    (:No
      (Components|Tick|SetComponentTickEnabled true))
    (:Yes
      (CallFunction|UpdateMobileTimer)))
  (if (and (or (== (Utilities|Array|Length (Variables|MeshData|GetItems%20to%20Load)) 1) (not (Variables|MeshData|GetCurrently%20Previewing%20Items))) (CallFunction|ShouldShowPreviews))
    (CallFunction|ShowNextPreviewMesh) ; As long as one item is in the preview list, \nwe're good to start previewing.
    (else
      (Variables|Default|SetNext%20Item) ; Item has been loaded, allow Asnyc to load the next one if required.
    )))

(fn ShowNextPreviewMesh
  (bind (_meshassetpath_8_6b40c41c47213e2d7d041a876713c4f5 _meshobject_5_319aaacd4f43372817f7429c492d1b93) (Utilities|Struct|BreakPickupItemPreview (Variables|Default|GetNextMeshPreview)))
  (bind _previewstaticmesh (Variables|Default|GetPreviewStaticMesh))
  (bind (_nextmeshpreview_meshassetpath_8_6b40c41c47213e2d7d041a876713c4f5 _nextmeshpreview_meshobject_5_319aaacd4f43372817f7429c492d1b93 _nextmeshpreview_raritycolor_2_e689f1664314461dba24e7b7f58d6edc _nextmeshpreview_pickuptransform_15_7d98aff949c4025f30a8f4b107b5ccc8) (Utilities|Struct|BreakPickupItemPreview (Variables|GetNextMeshPreview_MeshAssetPath_8_6B40C41C47213E2D7D041A876713C4F5)))
  (bind _previewstaticmesh_2 (Variables|Default|GetPreviewStaticMesh))
  (bind _previewskeletalmesh (Variables|Default|GetPreviewSkeletalMesh))
  (Utilities|IsValid _meshobject_5_319aaacd4f43372817f7429c492d1b93
    (:Is%20Valid
      (Variables|Default|SetCurrent%20Scale%20Time 0.0)
      (Utilities|FlowControl|Sequence
        (:then_0
          (CallFunction|ClearWeaponMods false) ; Don't clear the WID yet
          (LocalVariables|SetRarityColor _nextmeshpreview_raritycolor_2_e689f1664314461dba24e7b7f58d6edc)
          (LocalVariables|SetTransform _nextmeshpreview_pickuptransform_15_7d98aff949c4025f30a8f4b107b5ccc8)
          (bind _asstatic_mesh (Utilities|Casting|CastToStaticMesh _nextmeshpreview_meshobject_5_319aaacd4f43372817f7429c492d1b93)
            (:then
              (LocalVariables|SetStatic%20Mesh _asstatic_mesh)
              (Variables|Default|GetPreviewStaticMesh
                (:then
                  _)
                (:else
                  (bind _found (CallFunction|FindStaticMeshComponent))
                  (if _found
                    _)))
              (CallFunction|SetSkeletalMeshHidden true) ; We're using the Static Preview\nHide the Skeletal Preview
              (Development|SetHiddeninGame (Variables|Default|GetPreviewStaticMesh) (not (CallFunction|ShouldShowPreviews)) false)
              (Components|StaticMesh|SetStaticMesh _previewstaticmesh (LocalVariables|GetStatic%20Mesh))
              (CallFunction|ResetMaterials _previewstaticmesh)
              (CallFunction|UpdateStaticMeshTransform (LocalVariables|GetTransform))
              (bind _rarity_override_mid (Variables|Rarity|GetRarity%20Override%20MID)
                (:then
                  (Rendering|Material|SetOverlayMaterial _previewstaticmesh_2 _rarity_override_mid false 0 :!class /Script/Engine.MeshComponent)
                  (Rendering|Material|SetOverlayMaterialMaxDrawDistance _previewstaticmesh_2 (CallFunction|GetOverlayDistance) :!class /Script/Engine.MeshComponent)
                  (Rendering|Material|SetVectorParameterValue _rarity_override_mid (Variables|Rarity|GetRarity%20Color%20Param%20Name) (LocalVariables|GetRarityColor) :!class /Script/Engine.MaterialInstanceDynamic))
                (:else
                  (Rendering|Material|SetOverlayMaterial (Variables|Default|GetPreviewStaticMesh) 0 false 0 :!class /Script/Engine.MeshComponent))))
            (:CastFailed
              (bind _asskeletal_mesh (Utilities|Casting|CastToSkeletalMesh _nextmeshpreview_meshobject_5_319aaacd4f43372817f7429c492d1b93)
                (:then
                  (LocalVariables|SetSkeletal%20Mesh _asskeletal_mesh)
                  ;[1 0.2891 0.7518 0.264] Skeletal Mesh Path
                  (bind _previewskeletalmesh_2 (Variables|Default|GetPreviewSkeletalMesh)
                    (:then
                      (Transformation|DetachFromComponent _previewskeletalmesh_2 "KeepRelative" "KeepRelative" "KeepRelative" true)
                      (Transformation|SetRelativeTransform _previewskeletalmesh_2 (LocalVariables|GetTransform) false true)
                      (CallFunction|SetSkeletalMeshHidden (not (CallFunction|ShouldShowPreviews)))
                      (Components|SkinnedMesh|SetSkinnedAssetandUpdate _previewskeletalmesh_2 (LocalVariables|GetSkeletal%20Mesh) true)
                      (Development|SetHiddeninGame (Variables|Default|GetPreviewStaticMesh) true false) ; We're using the Skeletal Preview.\nHide the Static Preview
                      (CallFunction|ResetMaterials _previewskeletalmesh)
                      (CallFunction|ApplyRarityOverrideMID _previewskeletalmesh (LocalVariables|GetRarityColor))
                      (CallFunction|ApplyWeaponWrapMaterial)
                      (if (Utilities|IsValid (Variables|Default|GetNextModdableWID) :!class /Script/Engine.KismetSystemLibrary)
                        (CallFunction|UpdateModdableWeaponPreviewMesh (LocalVariables|GetRarityColor)))
                      (CallFunction|RecenterSkeletalMesh (Variables|Default|GetPreviewMeshScene)))
                    (:else)))
                (:CastFailed
                  (LocalVariables|SetLoad%20Failed true))))))
        (:then_1
          (if (not (LocalVariables|GetLoad%20Failed))
            (Default|CallPreviewMeshUpdated)
            ;[1 1 1] If load has passed, start loading the next object immediately.
            (CallFunction|ItemLoadedAndDisplayed true)
            (bind _itemswapped_sound (Variables|Audio|GetItemSwapped%20Sound)
              (:then
                (Audio|PlaySoundatLocation _itemswapped_sound (Transformation|GetWorldLocation (Variables|Default|GetPreviewMeshScene))) ; Play Switch Audio
              )
              (:else))
            (else
              ;[1 0.2062 0] Debug - Somehow an item is attempting to get loaded that isn't a skeletal or static mesh. Print a warning to the log.
              (CallFunction|DebugMobileLogging (Utilities|GetDisplayName (Utilities|Struct|BreakPickupItemPreview (Variables|Default|GetNextMeshPreview))) " Show Next Preview Mesh: Mesh Object was not a Static or Skeletal Mesh." true))))))
    (:Is%20Not%20Valid
      (CallFunction|StopPreviews))))

(fn StartSucktoBackpack
  (Variables|Timekeeping|SetSuckToBackpackTime (Utilities|Time|GetGameTimeinSeconds))
  (Components|Tick|SetComponentTickEnabled true) ; Directly enabling tick (rather than setting Start Preview) as if this is \nthe first item we won't have any items so load async items will fail.
)

(fn StopPreviews
  ;[1 0.07009 0 0.5] Clear the Meshes, hide the components, then stop the component ticking. The ticking component is what rotates/switches out preview meshes, so disabling that is essentially telling the component to stop.
  (bind _previewskeletalmesh (Variables|Default|GetPreviewSkeletalMesh)
    (:then
      (Components|SkinnedMesh|SetSkinnedAssetandUpdate _previewskeletalmesh 0 true))
    (:else
      _))
  (bind _previewstaticmesh (Variables|Default|GetPreviewStaticMesh)
    (:then
      (Components|StaticMesh|SetStaticMesh _previewstaticmesh (Variables|MeshData|GetDefault%20Static%20Mesh)))
    (:else
      _))
  (Development|SetHiddeninGame (Variables|Default|GetPreviewStaticMesh) true false)
  (CallFunction|SetSkeletalMeshHidden true)
  (CallFunction|ClearWeaponMods true)
  (bind _preview_audio_cue (Variables|Default|GetPreview%20Audio%20Cue)
    (:then
      (Audio|Components|Audio|Stop _preview_audio_cue))
    (:else
      _))
  (Components|Tick|SetComponentTickEnabled false)
  (Utilities|Time|ClearandInvalidateTimerbyHandle (Variables|Timekeeping|GetMobile%20Preview%20Timer))
  (Variables|Default|SetNext%20Item)
  (Variables|MeshData|SetCurrently%20Previewing%20Items false))

(fn UpdateItemList
  (Utilities|ClearMeshPreview)
  (bind _pickupitemlist (Variables|Default|GetPickupItemList)
    (:then
      (Variables|MeshData|SetItems%20to%20Load (ItemData|GetContainedItemVariants _pickupitemlist)))
    (:else
      _))
  (if (== (Variables|Default|GetPreviewMeshIndex) 0)
    (CallFunction|IncrementMeshLoadIndex (Utilities|Array|LastIndex (Variables|MeshData|GetItems%20to%20Load))) ; we were about to loop back to the start (were on the last item)\nJump back to the last item instead.
    (else
      (CallFunction|ModMeshLoadIndextoArraySize) ; ensure load index is within bounds.
    ))
  (Default|CallItemsChanged (Math|Conversions|ToBoolean(Integer) (Utilities|Array|Length (Variables|MeshData|GetItems%20to%20Load))))
  (CallFunction|UpdatePreviews))

(fn UpdateMobileTimer
  (if (and (> (Utilities|Array|Length (Variables|MeshData|GetItems%20to%20Load)) 1) (CallFunction|ShouldShowPreviews))
    (if (not (Utilities|Time|IsTimerActivebyHandle (Variables|Timekeeping|GetMobile%20Preview%20Timer)))
      (bind _returnvalue (Utilities|Time|SetTimerbyEvent (EventDispatchers|CreateEvent :!event ShowNextPreviewMesh) (CallFunction|ItemSwitchTime) true)) ; on Mobile we want to simply switch out the mesh every X seconds, rather than rotate and scale.
      (Variables|Timekeeping|SetMobile%20Preview%20Timer _returnvalue))
    (else
      (Utilities|Time|ClearandInvalidateTimerbyHandle (Variables|Timekeeping|GetMobile%20Preview%20Timer)))))

(fn UpdatePreviews
  ;[0 0.6006 1 0.5] Previews should always play in Edit Mode if there are items to preview and the component isn't disabled. Otherwise they should never play.
  (if (Editor|InEditor)
    (if (Variables|Default|GetEnabled)
      (CallFunction|EditorDisplayPreview (Variables|Default|GetPreviewStaticMesh)))
    (elif (CallFunction|ShouldShowPreviews)
      (CallFunction|LoadAndStartPreviews)
      (else
        (CallFunction|StopPreviews)))))

(fn UpdateScale
  (params (Item%20Preview%20Scale :type Float%20%28double-precision%29 :default 1.0))
  ;[1 1 1] Called from Owning Actor
  (Variables|Default|SetItem%20Preview%20Scale%20Multiplier Item%20Preview%20Scale))

(fn UpdateStaticMeshTransform
  (params (InTransform :type Transform))
  (bind _previewstaticmesh (Variables|Default|GetPreviewStaticMesh))
  (bind _staticmesh (Variables|/Script/Engine.StaticMeshComponent|GetStaticMesh _previewstaticmesh))
  (bind _output (Utilities|ResolveSoftReference _staticmesh))
  (bind _specifytransformforthesestaticmeshes (Variables|Default|GetSpecifyTransformForTheseStaticMeshes))
  (bind (_value _returnvalue) (Utilities|Map|Find _specifytransformforthesestaticmeshes _output))
  ;[1 1 1] Allow user-specified transforms for specific objects
  (if _returnvalue
    (Transformation|SetRelativeTransform (Variables|Default|GetPreviewStaticMesh) _value false true)
    (else
      (Transformation|SetRelativeTransform (Variables|Default|GetPreviewStaticMesh) InTransform false true))))

(fn UpdateModdableWeaponPreviewMesh
  (params (RarityColor :type Linear%20Color%20Structure))
  (bind _returnvalue (Class|FortWeaponModFunctionLibrary|SpawnWeaponModPreviewActor (Variables|Default|GetNextModdableWID) (Transformation|GetWorldTransform (Variables|Default|GetPreviewSkeletalMesh)))) ; This spawns the weapon mesh with its default mods attached, so that we can reference it.\nIf we just attach this directly to the device, selecting it with the move tool won't update the weapon's material.\nAs a result, we have to spawn and attach each individual mod mesh.
  (LocalVariables|SetModWeaponRef _returnvalue)
  (Utilities|FlowControl|Sequence
    (:then_0
      ;[1 0.5393 0.2532 0.5] Add the mod meshes and apply the overlay material to them
      (bind _modweaponref (LocalVariables|GetModWeaponRef)
        (:then
          (CallFunction|AddModstoWeaponPreview _modweaponref))
        (:else
          _))
      (Variables|Rarity|GetRarity%20Override%20MID
        (:then
          (for _array_element (Variables|Default|GetModMeshComps)
            (Utilities|IsValid _array_element
              (:Is%20Valid
                (CallFunction|ApplyRarityOverrideMID _array_element (Variables|GetRarityColor))))))
        (:else
          (Rendering|Material|SetOverlayMaterial (Variables|Default|GetPreviewSkeletalMesh) 0 false 0 :!class /Script/Engine.MeshComponent))))
    (:then_1
      ;[0.4167 0.1037 0.09537 0.5] Destroy the reference actor and clear the WID
      (bind _modweaponref_2 (LocalVariables|GetModWeaponRef)
        (:then
          (Actor|DestroyActor _modweaponref_2))
        (:else
          _))
      (Variables|Default|SetNextModdableWID 0))))
