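; Creative device mesh zone component: a scalable overlap volume that reports participants entering, leaving and activating it.
; The server toggles collision, filters overlaps by use restrictions, rechecks on team or class change, and maps pawns to controllers to resolve dead pawns.
; A required dwell time uses synced controller/time arrays and one timer aimed at the oldest entry.
; Clients animate curve-driven scaling and entry pulses, while shape meshes and materials load asynchronously.
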(event Custom|On%20Player%20Class%20Changed
  (params (PlayerState :type Fort%20Player%20State%20Object%20Reference) (ClassSlotIndex :type Byte))
  ;[0 1 0.8098 0.25] Class / Team Changes - If this happens to one of our overlapped players, reassess whether they should be inside.
  (CallFunction|ReassessPlayer (ParticipantInteraction|GetLastParticipantController PlayerState)))

(event Custom|Overlapped%20Player%20Pawn%20Died
  (params (DamagedActor :type Actor%20Object%20Reference) (Damage :type Float%20%28single-precision%29) (InstigatedBy :type Controller%20Object%20Reference) (DamageCauser :type Actor%20Object%20Reference) (HitLocation :type Vector) (FHitComponent :type Primitive%20Component%20Object%20Reference) (BoneName :type Name) (Momentum :type Vector))
  ;[1 0.2607 0 0.25] A tracked player pawn has died, we'll want to treat them as leaving.
  (CallFunction|OnEndOverlap 0 DamagedActor))

(event Custom|On%20Player%20Team%20Changed
  (params (PlayerState :type Fort%20Player%20State%20Object%20Reference) (TeamId :type Byte) (OldTeamId :type Byte) (bTeamChangedWithoutRespawn :type Boolean))
  ;[0 1 0.8098 0.25] Class / Team Changes - If this happens to one of our overlapped players, reassess whether they should be inside.
  (CallFunction|ReassessPlayer (ParticipantInteraction|GetLastParticipantController PlayerState)))

(event Custom|On%20Added%20to%20Minigame
  (params (Minigame :type Fort%20Minigame%20Object%20Reference))
  ;[0 1 0.2123 0.5] Events couldn't correctly bind due to the minigame not being initialized. It has now, attempt to rebind
  (Initialization|BindOrUnbindEvents true))

(event Custom|Empty%20Pawn%20Possessed
  (params (Pawn :type Pawn%20Object%20Reference))
  ;[1 1 1] A Player pawn that entered the zone without a player controller (i.e. one spawning directly into the zone and so hasn't been possessed yet) will call this event as a callback when it gets possessed. We want to unbind the possess bind then run the enter logic as we now have all the info we need to process it.
  (bind _asfort_player_pawn (Utilities|Casting|CastToFortPlayerPawn Pawn))
  (Game|Player|UnbindEventfromOnPawnPossessed _asfort_player_pawn (delegate Custom|Empty%20Pawn%20Possessed))
  (if (Collision|IsOverlappingActor Pawn)
    (bind _updated (CallFunction|UpdateOverlappingActors true Pawn))
    (if _updated
      (PlayerInteraction|PlayerEnterLogic Pawn (Collision|MakeHitResult false false 0.0 0.0 (Transformation|GetActorLocation Pawn) (Transformation|GetActorLocation Pawn))))))

(event EventEndPlay
  (params (EndPlayReason :type EEndPlayReason%20Enum :default "Destroyed"))
  ;[1 0.03409 0] Destruction - Unbind Events
  (Initialization|BindOrUnbindEvents false))

(event Custom|On%20End%20Overlap
  (params (OverlappedComponent :type Primitive%20Component%20Object%20Reference) (OtherActor :type Actor%20Object%20Reference) (OtherComp :type Primitive%20Component%20Object%20Reference) (OtherBodyIndex :type Integer))
  ;[1 1 1] Overlap Events
  (Utilities|CheckVehicle OtherActor
    (:Can%20Interact
      (Utilities|CheckHiding OtherActor
        (:Can%20Interact
          (bind _updated (CallFunction|UpdateOverlappingActors false OtherActor))
          (if _updated
            (PlayerInteraction|PlayerLeavingLogic OtherActor)))))))

(event Custom|On%20Begin%20Overlap
  (params (OverlappedComponent :type Primitive%20Component%20Object%20Reference) (OtherActor :type Actor%20Object%20Reference) (OtherComp :type Primitive%20Component%20Object%20Reference) (OtherBodyIndex :type Integer) (bFromSweep :type Boolean) (SweepResult :type Hit%20Result%20Structure%20%28by%20ref%29))
  ;[1 1 1] Overlap Events
  (bind _true (PlayerInteraction|TestPawnWithoutPlayerController OtherActor)) ; If a player spawns directly into a volume, they'll activate\nthe overlap before they have a controller. This covers that case.
  (if (not _true)
    (Utilities|CheckVehicle OtherActor
      (:Can%20Interact
        (bind _updated (CallFunction|UpdateOverlappingActors true OtherActor))
        (if _updated
          (PlayerInteraction|PlayerEnterLogic OtherActor SweepResult))))))

(event Custom|On%20Play%20Mode%20Changed
  (params (Minigame :type Fort%20Minigame%20Object%20Reference) (bIsInPlayMode :type Boolean))
  ;[0 0.4352 0.8594 0] Play Mode has changed, our next (or currently running) scaling should instantly complete to avoid popping in/out volumes. Always call Refresh Zone State as some playmodes will need a refresh at this time.
  (Effects|ResetInstantlyCompleteNextScale)
  (ZoneState|RefreshZoneState))

(event Custom|On%20Enable%20State%20Changed
  (params (Enabled :type Boolean))
  (ZoneState|SetZoneState Enabled))

(event Custom|Async%20Load%20Shape%20Details
  (params (New%20Shape :type EGameplayVolumeShape%20Enum :default "NewEnumerator0"))
  ;[0 0.7094 1 0.1] Asyncronous Loading of New Mesh (for swapping between multiple meshes at runtime)
  (bind (_found _details) (Shape|GetShapeDetails New%20Shape))
  (if (and (Utilities|Enum|NotEqual(Enum) New%20Shape (Variables|Internal|GetShape)) _found)
    (Variables|Internal|SetShape New%20Shape)
    (Utilities|LoadMaterials (.MaterialOverrides_6_0BA09EBE42205D40584D3ABEB77D5BDD (Utilities|Struct|BreakCreativeMeshDetails _details))
      (:Complete
        (CallFunction|AsyncLoadMesh (Utilities|Struct|BreakCreativeMeshDetails _details)))) ; Async loads all materials and \nstores them in CachedLoadedMaterials.\nCalls Complete when Done
  ))

(event Custom|Async%20Load%20Mesh
  (params (Mesh%20to%20Load :type Static%20Mesh%20Soft%20Object%20Reference))
  ;[0 0.7094 1 0.1] Asyncronous Loading of New Mesh (for swapping between multiple meshes at runtime)
  (if (Utilities|IsValidSoftObjectReference Mesh%20to%20Load)
    (bind _object (Utilities|AsyncLoadAsset Mesh%20to%20Load)
      (:Completed
        (bind _asstatic_mesh (Utilities|Casting|CastToStaticMesh _object)
          (:then
            (bind _returnvalue (Components|StaticMesh|SetStaticMesh _asstatic_mesh))
            (if _returnvalue
              (for (_array_element _array_index) (Variables|Default|GetCachedLoadedMaterials)
                (Utilities|IsValid _array_element
                  (:Is%20Valid
                    (Rendering|Material|SetMaterial _array_index _array_element))))
              (else
                _))
            (Default|CallNewMeshLoaded))
          (:CastFailed
            _))))
    (else
      _)))

(event Custom|PlayOverlap
  (params (OverlapWorldLocation :type Vector))
  ;[1 0.7441 0] Pulse Effect Logic
  (Networking|CanExecuteCosmeticEvents
    (:True
      (if (Variables|Default|GetbEnableOverlapVFX)
        (Effects|EffectScaleVBias OverlapWorldLocation) ; FX Artist Voodoo. Leaving well alone\nother than to encapsulate it in this function.
        (CallFunction|StartTickPulse) ; Start the pulse effect ticking
        (else
          _))
      (if (Variables|Default|GetbEnableOverlapSound)
        (bind _enterzoneaudio (Variables|Settings|GetEnterZoneAudio)
          (:then
            (Audio|PlaySoundatLocation _enterzoneaudio OverlapWorldLocation))
          (:else))))))

(event Custom|Start%20Tick%20Pulse
  ;[1 0.7441 0] Pulse Effect Logic
  (Variables|Transient|SetCurrent%20Pulse%20Time 0.0)
  (CallFunction|TickPulse))

(event Custom|TickPulse
  (bind (_mintime _maxtime) (Math|Curves|GetTimeRange (Variables|Effects|GetPulse%20Curve)))
  ;[1 0.7441 0] Pulse Effect Logic
  (bind _output_get (Variables|Transient|SetCurrent%20Pulse%20Time (+ (Variables|Transient|GetCurrent%20Pulse%20Time) (Utilities|Time|GetWorldDeltaSeconds))))
  (CallFunction|UpdatePulseOverCurve _output_get)
  (if (< (Variables|Transient|GetCurrent%20Pulse%20Time) _maxtime)
    (Utilities|FlowControl|Delay 0.0)))

(event Custom|TickScale
  ;[1 1 1] Scale Mesh (completely cosmetic)
  (bind _complete (Effects|ScaleMesh))
  (if (not _complete)
    (Utilities|FlowControl|Delay 0.0) ; Wait for next frame
  )
  (Variables|Internal|SetInstantlyCompleteNextScale false) ; We've completed our first scale, the next one shouldn't instantly complete
)

(event Custom|Update%20Pulse%20Over%20Curve
  (params (Time :type Float%20%28double-precision%29))
  (bind _self self)
  (bind _mid (Material|GetMID _self))
  (bind _pulse_curve (Variables|Effects|GetPulse%20Curve))
  (bind (_returnvalue_x _returnvalue_y _returnvalue_z) (Math|Vector|BreakVector (Math|Curves|GetVectorValue _pulse_curve Time)))
  ;[1 0.7441 0] Pulse Effect Logic
  (Rendering|Material|SetScalarParameterValue _mid "ImpactParam" _returnvalue_x :!class /Script/Engine.MaterialInstanceDynamic)
  (Rendering|Material|SetScalarParameterValue _mid "Pulse" _returnvalue_y :!class /Script/Engine.MaterialInstanceDynamic)
  (Rendering|Material|SetScalarParameterValue _mid "RingOpacity" _returnvalue_z :!class /Script/Engine.MaterialInstanceDynamic))

(fn AddPawntoOverlappingMap
  (params (Actor :type Actor%20Object%20Reference))
  ;[1 1 1] Keep a list of Overlapping Pawns along with their controllers. This is then used to match an uncontrolled dead pawn to its original controller for sending messages / providing other effects.
  (bind _asfort_player_pawn (Utilities|Casting|CastToFortPlayerPawn Actor))
  (Utilities|Map|Add (Variables|ParticipantInteraction|GetOverlappingPawnPartcipantControllerMap) _asfort_player_pawn (CreativeAI|Controller|GetParticipantControllerfromActor _asfort_player_pawn)))

(fn AddtoPlayerTriggerList
  (params (Controller :type Controller%20Object%20Reference))
  ;[1 1 1] Adds the Player to the Trigger Arrays. This should be the only function that ever adds to these arrays to ensure that they stay in Sync.
  (bind _returnvalue (Utilities|Array|Add (Variables|ZoneTimer|GetParticipantTrigger_Controller) Controller))
  (Utilities|Array|SetArrayElem (Variables|ZoneTimer|GetPlayerTrigger_Times) _returnvalue (+ (Variables|Settings|GetRequiredTimeInZone) (Utilities|Time|GetGameTimeinSeconds)))
  (return))

(fn BindorUnbindEvents
  (params (Bind :type Boolean))
  (bind _outputdelegate (EventDispatchers|CreateEvent :!event On%20Enable%20State%20Changed))
  (bind _outputdelegate_2 (EventDispatchers|CreateEvent :!event On%20Begin%20Overlap))
  (bind _outputdelegate_3 (EventDispatchers|CreateEvent :!event On%20End%20Overlap))
  (bind _outputdelegate_4 (EventDispatchers|CreateEvent :!event On%20Play%20Mode%20Changed))
  (bind _outputdelegate_5 (EventDispatchers|CreateEvent :!event On%20Added%20to%20Minigame))
  (bind _outputdelegate_6 (EventDispatchers|CreateEvent :!event On%20Player%20Team%20Changed))
  (bind _outputdelegate_7 (EventDispatchers|CreateEvent :!event On%20Player%20Class%20Changed))
  (if Bind
    (Collision|BindEventtoOnComponentBeginOverlap _outputdelegate_2)
    (Collision|BindEventtoOnComponentEndOverlap _outputdelegate_3)
    (bind _enabled_component (Variables|Default|GetEnabled%20Component)
      (:then
        (Default|BindEventtoOnEnabledStateChanged _enabled_component _outputdelegate))
      (:else
        _))
    (bind _logic_component (Variables|Default|GetLogic%20Component)
      (:then
        (bind _currentminigame (Variables|/Script/FortniteGame.FortMinigameLogicComponent|GetCurrentMinigame _logic_component)
          (:then
            (bind _output_get (LocalVariables|SetCurrent%20Minigame _currentminigame))
            (Minigame|BindEventtoOnPlayModeChangedDelegate _output_get _outputdelegate_4)
            (Variables|Default|GetUse%20Restriction%20Component
              (:then
                (Minigame|BindEventtoPlayerChangedTeamDelegate _output_get _outputdelegate_6)
                (Minigame|BindEventtoOnPlayerClassApplied _output_get _outputdelegate_7))
              (:else
                _)) ; We only care about team/class changes\nif we care about use restrictions for this zone
          )
          (:else
            ;[1 0.3119 0 0.5] If we don't currently have a valid minigame, wait until we do, then attempt to rebind events.
            (Minigame|BindEventtoOnAddedtoMinigame _logic_component _outputdelegate_5))))
      (:else))
    (else
      (Collision|UnbindEventfromOnComponentBeginOverlap _outputdelegate_2)
      (Collision|UnbindEventfromOnComponentEndOverlap _outputdelegate_3)
      (bind _enabled_component_2 (Variables|Default|GetEnabled%20Component)
        (:then
          (Default|UnbindEventfromOnEnabledStateChanged _enabled_component_2 _outputdelegate)
          (bind _logic_component_2 (Variables|Default|GetLogic%20Component)
            (:then
              (Minigame|UnbindEventfromOnAddedtoMinigame _logic_component_2 _outputdelegate_5)
              (bind _currentminigame_2 (Variables|/Script/FortniteGame.FortMinigameLogicComponent|GetCurrentMinigame _logic_component_2)
                (:then
                  (Minigame|UnbindEventfromOnPlayModeChangedDelegate _currentminigame_2 _outputdelegate_4))
                (:else)))
            (:else)))
        (:else))))
  (Minigame|UnbindEventfromPlayerChangedTeamDelegate _currentminigame_2 _outputdelegate_6)
  (Minigame|UnbindEventfromOnPlayerClassApplied _currentminigame_2 _outputdelegate_7))

(fn CheckZoneTimer
  (Utilities|Time|ClearandInvalidateTimerbyHandle (Variables|ZoneTimer|GetZoneTimer))
  (if (and (> (Variables|Settings|GetRequiredTimeInZone) 0.0) (Math|Conversions|ToBoolean(Integer) (Utilities|Array|Length (Variables|ZoneTimer|GetParticipantTrigger_Controller))))
    (if (<= (Utilities|Array|Get(acopy) (Variables|ZoneTimer|GetPlayerTrigger_Times)) (Utilities|Time|GetGameTimeinSeconds))
      (LocalVariables|SetParticipantController (Utilities|Array|Get(acopy) (Variables|ZoneTimer|GetParticipantTrigger_Controller))) ; Store a copy of the PC \n(we're about to remove it from the list)
      (ZoneTimer|RemoveIndexfromPlayerTriggerList 0)
      ;[1 1 1] If the player is still in the game, activate the zone.
      (bind _participantcontroller (LocalVariables|GetParticipantController)
        (:then
          (PlayerInteraction|PlayerActivatedZone _participantcontroller))
        (:else))
      (else
        (ZoneTimer|StartTimerFromCheck) ; Player in slot 0 still has time remaining till they're due to activate.\nStart the timer again with their remaining time.
      ))))

(fn ClearOverlaps
  ;[1 1 1] Clears Data about Overlapping Players. Called when the Device is deactivated manually.
  (Utilities|Time|ClearandInvalidateTimerbyHandle (Variables|ZoneTimer|GetZoneTimer)) ; Stop the timer that's checking for people in the volume
  (ZoneTimer|ClearPlayerTriggerList)
  (Utilities|Map|Clear (Variables|ParticipantInteraction|GetOverlappingPawnPartcipantControllerMap)))

(fn ClearPlayerTriggerList
  ;[1 1 1] These must stay in sync. They're only two arrays rather than a map in order to facilitate the searching that this function does.
  (Utilities|Array|Clear (Variables|ZoneTimer|GetParticipantTrigger_Controller))
  (Utilities|Array|Clear (Variables|ZoneTimer|GetPlayerTrigger_Times)))

(fn EffectScalevBias
  (params (Hit%20World%20Location :type Vector))
  (bind _returnvalue (Components|GetOwner))
  (bind _returnvalue_2 (Components|GetOwner))
  (bind _returnvalue_3 (Transformation|GetWorldLocation))
  (bind _returnvalue_4 (Transformation|GetActorTransform _returnvalue))
  (bind _returnvalue_5 (Math|Transform|InverseTransformDirection _returnvalue_4 (- (* (- Hit%20World%20Location _returnvalue_3) -1.0) (* (Transformation|GetActorUpVector _returnvalue_2) (Math|Vector|DotProduct (* (- Hit%20World%20Location _returnvalue_3) -1.0) (Transformation|GetActorUpVector _returnvalue_2))))))
  (bind _returnvalue_6 (- Hit%20World%20Location _returnvalue_3))
  (bind _returnvalue_7 (Transformation|GetActorUpVector _returnvalue_2))
  (bind _returnvalue_8 (Math|Vector|DotProduct (* _returnvalue_6 -1.0) _returnvalue_7))
  (bind _returnvalue_9 (* _returnvalue_6 -1.0))
  (bind _returnvalue_10 (* _returnvalue_7 _returnvalue_8))
  (bind (_origin _boxextent) (Collision|GetComponentBounds 0))
  (bind (_boxextent_x _boxextent_y _boxextent_z) (Math|Vector|BreakVector _boxextent))
  (bind _returnvalue_11 (- _returnvalue_9 _returnvalue_10))
  ;[1 1 1] Rotate the class selector mesh to the point where the player entered it so our activation effect plays from the appropriate side
  (if (not (Math|Vector|Equal(Vector) _returnvalue_6 "0, 0, 0" 0.0001))
    ;;[1 1 1] Transform direction into local space and use it to make a rotator to rotate the volume area mesh in local space
    (if (not (Math|Vector|VectorIsNearlyZero _returnvalue_5))
      (Transformation|SetRelativeRotation (Math|Rotator|MakeRotfromX _returnvalue_5))
      (else
        _))
    (Rendering|Material|SetScalarParameterValue (Material|GetMID self) "VBias" (- (/ (- (.z (Variables|GetHit%20World%20Location)) (.z (Transformation|GetWorldLocation))) (Math|Float|SelectFloat 1.0 _boxextent_z (Math|Float|NearlyEqual(Float) _boxextent_z))) 0.5) :!class /Script/Engine.MaterialInstanceDynamic)))

(fn GetLastParticipantController
  (params (Actor :type Actor%20Object%20Reference))
  ;[0 0.4352 0.8594 0] If the actor is a pawn and dead, check to see what was previously controlling it and return that.
  ; This is necessary as the pawn will be unpossessed before the collision is switched off, resulting in dead pawns always having no controller.
  (bind _output_get (LocalVariables|SetCheckActor Actor))
  (bind _asfort_player_controller (Utilities|Casting|CastToFortPlayerController _output_get)
    (:then
      (return _asfort_player_controller))
    (:CastFailed
      (bind _asfort_athena_aibot_controller (Utilities|Casting|CastToFortAthenaAIBotController _output_get)
        (:then
          (return _asfort_athena_aibot_controller))
        (:CastFailed
          (bind _asfort_player_state (Utilities|Casting|CastToFortPlayerState _output_get)
            (:then
              (LocalVariables|SetCheckActor (PlayerState|GetCurrentPawn _asfort_player_state)))
            (:CastFailed
              _))
          (bind _asfort_pawn (Utilities|Casting|CastToFortPawn (LocalVariables|GetCheckActor))
            (:then
              (Utilities|IsValid (CreativeAI|Controller|GetParticipantControllerfromActor _asfort_pawn)
                (:Is%20Valid
                  (return (CreativeAI|Controller|GetParticipantControllerfromActor _asfort_pawn)))
                (:Is%20Not%20Valid
                  (return (Utilities|Map|Find (Variables|ParticipantInteraction|GetOverlappingPawnPartcipantControllerMap) _asfort_pawn)))))
            (:CastFailed
              (return 0))))))))

(fn GetShapeDetails
  (params (Shape :type EGameplayVolumeShape%20Enum%20%28by%20ref%29 :default "NewEnumerator0"))
  (bind (_value _returnvalue) (Utilities|Map|Find (Variables|Settings|GetShapeMeshDetails) Shape))
  ;[0 1 0.4769] Helper function to cache a copy of the shape details, to avoid multiple lookups when individually accessing struct elements.
  (return _returnvalue _value))

(fn GetShapeToMScale
  (bind (_value _returnvalue) (Utilities|Map|Find (Variables|Settings|GetShapeMeshDetails) (Variables|Internal|GetShape)))
  (bind (_value_mesh_2_ba91cdb64dfb95237cec639088ecafeb _value_materialoverrides_6_0ba09ebe42205d40584d3abeb77d5bdd _value_scaletometers_10_a92c273a46f5975d82a56fbc329ee84e _value_scaleinmaterial_12_3a38e9db40fb218347be679d9e425a06 _value_colorparam_15_9a09a70d429488a817bc529c1d48b7e2 _value_scalez_18_8e007c234e08efba3c5bd3a9913007dc) (Utilities|Struct|BreakCreativeMeshDetails _value))
  ;[0 1 0.823 0.05] Convert the scale of the mesh to 1 meter, to ensure no matter the incoming mesh, we're always using a consistent scale. If we have no overrides, assume our mesh is 1UU in size, and multiply by 100
  (return (select _returnvalue _value_scaletometers_10_a92c273a46f5975d82a56fbc329ee84e (Math|Conversions|ToVector(Float) (Variables|Internal|GetM%20To%20CMConversion))) _value_scalez_18_8e007c234e08efba3c5bd3a9913007dc))

(fn InitializeComponent
  (params (Enabled%20Component :type Creative%20Enabled%20Component%20Object%20Reference) (Logic%20Component :type Fort%20Minigame%20Logic%20Component%20Object%20Reference) (Use%20Restriction%20Component :type Creative%20Use%20Restriction%20Component%20Object%20Reference) (Zone%20Scale :type Float%20%28double-precision%29 :default 1.0) (ZoneScaleMultipliers :type Vector :default "1.000000,1.000000,0") (Activation%20Time :type Float%20%28double-precision%29) (EnableOverlapSound :type Boolean) (EnableOverlapVFX :type Boolean) (Shape%20Override :type EGameplayVolumeShape%20Enum :default "NewEnumerator0"))
  (Variables|Default|SetEnabled%20Component Enabled%20Component)
  (Variables|Default|SetLogic%20Component Logic%20Component)
  (Variables|Default|SetUse%20Restriction%20Component Use%20Restriction%20Component)
  (Variables|ScaleOptions|SetZone%20Scale Zone%20Scale)
  (Variables|ScaleOptions|SetZone%20Scale%20Multipliers ZoneScaleMultipliers)
  (Variables|Settings|SetRequiredTimeInZone Activation%20Time)
  (Variables|Default|SetbEnableOverlapSound EnableOverlapSound)
  (Variables|Default|SetbEnableOverlapVFX EnableOverlapVFX)
  (CallFunction|AsyncLoadShapeDetails (Variables|GetShape%20Override)) ; Load a new shape, if shape has changed
  (Effects|ResetInstantlyCompleteNextScale)
  (ZoneState|UpdateScale 1.0)
  (Initialization|BindOrUnbindEvents true)
  (ZoneState|RefreshZoneState))

(fn PlayerActivatedZone
  (params (Controller :type Controller%20Object%20Reference))
  (Utilities|IsValid Controller
    (:Is%20Valid
      (Utilities|Set|Add (Variables|ParticipantInteraction|GetActiveParticipants) Controller)
      (Default|CallOnParticipantActivatedZone Controller))))

(fn PlayerEnterLogic
  (params (Entering%20Actor :type Actor%20Object%20Reference) (Hit :type Hit%20Result%20Structure%20%28by%20ref%29) (Use%20Restriction%20Check :type Boolean :default true))
  (bind _entering_actor (Variables|GetEntering%20Actor))
  (bind _controller (CreativeAI|Controller|GetParticipantControllerfromActor _entering_actor))
  (Utilities|SwitchHasAuthority
    (:Authority
      (if Use%20Restriction%20Check
        (bind _use_restriction_component (Variables|Default|GetUse%20Restriction%20Component)
          (:then
            (if (CreativeUserRestrictions|CanParticipantActorInteract _use_restriction_component (Variables|GetEntering%20Actor) true true false)
              _))
          (:else
            _))
        (PlayerInteraction|AddPawnToOverlappingMap _entering_actor)
        (Default|CallOnParticipantEntered _controller (Variables|GetHit))
        (ZoneTimer|StartZoneActivation _controller)
        (if (Variables|Settings|GetPulse%20on%20Enter)
          (Effects|PulseMesh (Variables|GetHit)))
        (else
          _)))))

(fn PlayerLeavingLogic
  (params (Leaving%20Actor :type Actor%20Object%20Reference) (Use%20Restriction%20Check :type Boolean :default true))
  (Utilities|SwitchHasAuthority
    (:Authority
      (LocalVariables|SetParticipantController (ParticipantInteraction|GetLastParticipantController Leaving%20Actor))
      (Utilities|Set|Remove (Variables|ParticipantInteraction|GetActiveParticipants) (LocalVariables|GetParticipantController))
      (if (Variables|GetUse%20Restriction%20Check)
        (bind _use_restriction_component (Variables|Default|GetUse%20Restriction%20Component)
          (:then
            (if (CreativeUserRestrictions|CanParticipantActorInteract _use_restriction_component (LocalVariables|GetParticipantController) true true false)
              (Default|CallOnParticipantLeft (LocalVariables|GetParticipantController))
              (else
                _))
            (PlayerInteraction|RemovePawnFromOverlappingMap (Variables|GetLeaving%20Actor))
            (ZoneTimer|RemoveFromPlayerTriggerList (LocalVariables|GetParticipantController)))
          (:else
            _))
        (else
          _)))))

(fn PulseMesh
  (params (Hit :type Hit%20Result%20Structure%20%28by%20ref%29))
  (bind (_bblockinghit _binitialoverlap _time _distance _location) (Collision|BreakHitResult Hit))
  (CallFunction|PlayOverlap _location)
  (Networking|FlushNetDormancy (Components|GetOwner)))

(fn ReassessPlayer
  (params (Controller :type Controller%20Object%20Reference))
  (bind _controller (Variables|GetController))
  (bind _localparticipantsinzone (LocalVariables|GetlocalParticipantsInZone))
  (bind _returnvalue (Utilities|Set|ContainsItem _localparticipantsinzone _controller))
  ;[0 1 0.8098 0.25] Check to see if the player is currently active but shouldn't be, or currently is inactive but should be (likely player has changed use restriction values while within area)
  (if (Utilities|Set|ContainsItem (Variables|ParticipantInteraction|GetOverlappingParticipants) Controller)
    (bind _output_get (LocalVariables|SetlocalParticipantsInZone (Variables|ParticipantInteraction|GetActiveParticipants)))
    (Utilities|Set|AddItems _output_get (Variables|ZoneTimer|GetParticipantTrigger_Controller))
    (if (CreativeUserRestrictions|CanParticipantActorInteract (Variables|Default|GetUse%20Restriction%20Component) (Variables|GetController) true true false)
      (if (not _returnvalue)
        (PlayerInteraction|PlayerEnterLogic (Variables|GetController) (Collision|MakeHitResult false false 0.0 0.0 (Transformation|GetActorLocation (Variables|GetController)) (Transformation|GetActorLocation (Variables|GetController))) false))
      (elif _returnvalue
        (PlayerInteraction|PlayerLeavingLogic (Variables|GetController) false)))))

(fn RefreshZoneState
  (bind _enabled_component (Variables|Default|GetEnabled%20Component)
    (:then
      (ZoneState|SetZoneState (Class|CreativeEnabledComponent|IsEnabled _enabled_component)))
    (:else)))

(fn RemovefromPlayerTriggerList
  (params (Controller :type Controller%20Object%20Reference))
  (bind _output_get (LocalVariables|SetIndex (Utilities|Array|FindItem (Variables|ZoneTimer|GetParticipantTrigger_Controller) Controller)))
  (if (!= _output_get -1)
    (ZoneTimer|RemoveIndexfromPlayerTriggerList (LocalVariables|GetIndex))))

(fn RemoveIndexFromPlayerTriggerList
  (params (Index :type Integer))
  ;[1 1 1] These must stay in sync. They're only two arrays rather than a map in order to facilitate the searching that this function does.
  (Utilities|Array|RemoveIndex (Variables|ZoneTimer|GetParticipantTrigger_Controller) Index)
  (Utilities|Array|RemoveIndex (Variables|ZoneTimer|GetPlayerTrigger_Times) Index)
  (if (== Index 0)
    (ZoneTimer|CheckZoneTimer)))

(fn RemovePawnfromOverlappingMap
  (params (Actor :type Actor%20Object%20Reference))
  ;[1 1 1] Keep a list of Overlapping Pawns along with their controllers. This is then used to match an uncontrolled dead pawn to its original controller for sending messages / providing other effects.
  (bind _asfort_player_pawn (Utilities|Casting|CastToFortPlayerPawn Actor))
  (Utilities|Map|Remove (Variables|ParticipantInteraction|GetOverlappingPawnPartcipantControllerMap) _asfort_player_pawn))

(fn ResetInstantlyCompleteNextScale
  (Variables|Internal|SetInstantlyCompleteNextScale true))

(fn ScaleMesh
  (Utilities|FlowControl|Sequence
    (:then_0
      (if (Math|Boolean|ORBoolean (TouchPlatform|IsMobileGame) (TouchPlatform|IsNintendoSwitch) (Variables|Internal|GetInstantlyCompleteNextScale) (not (Variables|ScaleOptions|GetMesh%20Scales)))
        ;[1 0.4501 0] If we're on mobile or switch, select the end Scale Time as the current Scale Time. This will cause it to immediately jump to the end - i.e. not scale visibly. This stops ticking.
        ; In addition, we may also want to stop it in other circumstances, either we don't want the mesh to scale for that device, or it's initialization / sta
        (Variables|Transient|SetCurrent%20Scale%20Time (Math|Float|SelectFloat (Variables|ScaleOptions|GetScale%20Time) 0.0 (Variables|Transient|GetScale%20Direction))))
      (bind (_value _biscomplete) (Math|AnimateFloatCurve (Utilities|Time|GetWorldDeltaSeconds) (Variables|Transient|GetCurrent%20Scale%20Time) (Variables|ScaleOptions|GetScale%20Curve) (Variables|ScaleOptions|GetScale%20Time) (not (Variables|Transient|GetScale%20Direction))))
      (LocalVariables|SetIsComplete _biscomplete)
      (ZoneState|UpdateScale _value))
    (:then_1
      (return (LocalVariables|GetIsComplete)))))

(fn SetDisabledZoneVisible
  (params (DisabledZoneVisible :type EVisibleInGameState%20Enum :default "NewEnumerator0"))
  (Variables|Settings|SetDisabledZoneVisible DisabledZoneVisible)
  (ZoneState|RefreshZoneState))

(fn SetZoneState
  (params (On :type Boolean))
  (Utilities|SwitchHasAuthority
    (:Authority
      ;[1 1 1] Enable or disable collision on the volume on the server, allowing users to interact
      (Collision|SetCollisionEnabled (select On "QueryOnly" "NoCollision"))
      (if (not (Variables|GetOn))
        (PlayerInteraction|ClearOverlaps)))
    (:Remote
      (Effects|StartScaleVisualMesh (or On (Effects|ShouldbeVisibleWhenOff))) ; Scale Visuals on the Client Only
      (Collision|SetCollisionEnabled "NoCollision"))))

(fn ShouldBeVisiblewhenOff
  (bind _logic_component (Variables|Default|GetLogic%20Component)
    (:then
      (bind _currentminigame (Variables|/Script/FortniteGame.FortMinigameLogicComponent|GetCurrentMinigame _logic_component)
        (:then
          (LocalVariables|SetPlay%20Mode (Minigame|IsinPlayMode _currentminigame :!class /Script/FortniteGame.FortMinigame)))
        (:else
          (LocalVariables|SetPlay%20Mode true) ; Default to On if we're not sure\n(edge cases tend to cause problems when Play Mode is On\nbut we think it's off)
        ))
      (return (Utilities|Select :Index (Variables|Settings|GetDisabledZoneVisible) :NewEnumerator0 false :NewEnumerator1 (not (LocalVariables|GetPlay%20Mode)) :NewEnumerator2 (LocalVariables|GetPlay%20Mode) :NewEnumerator3 true :NewEnumerator4 false)))
    (:else
      _)))

(fn StartScaleVisualMesh
  (params (Scale%20Up :type Boolean))
  (Networking|CanExecuteCosmeticEvents
    (:True
      (Variables|Transient|SetScale%20Direction Scale%20Up)
      (if (not (Variables|Rendering|GetbVisible))
        (Variables|Internal|SetInstantlyCompleteNextScale true))
      (CallFunction|TickScale))))

(fn StartTimerfromCheck
  (if (and (Utilities|Array|IsNotEmpty (Variables|ZoneTimer|GetPlayerTrigger_Times)) (Utilities|Array|IsNotEmpty (Variables|ZoneTimer|GetParticipantTrigger_Controller)))
    (Utilities|IsValid (Utilities|Array|Get(acopy) (Variables|ZoneTimer|GetParticipantTrigger_Controller))
      (:Is%20Valid
        (bind _output_get (LocalVariables|SetTime%20to%20Next%20Check (- (Utilities|Array|Get(acopy) (Variables|ZoneTimer|GetPlayerTrigger_Times)) (Utilities|Time|GetGameTimeinSeconds))))
        (if (> _output_get 0.0)
          (bind _returnvalue (Utilities|Time|SetTimerbyEvent (EventDispatchers|CreateEvent :!event Check%20Zone%20Timer) (LocalVariables|GetTime%20to%20Next%20Check)))
          (Variables|ZoneTimer|SetZoneTimer _returnvalue)
          (else
            (ZoneTimer|CheckZoneTimer))))
      (:Is%20Not%20Valid
        ;[1 1 1] Actor is no longer valid, remove from list and try again.
        (ZoneTimer|RemoveIndexfromPlayerTriggerList 0)))))

(fn StartZoneActivation
  (params (Controller :type Controller%20Object%20Reference))
  ;[1 1 1] Rather than hooking to On Player Entered , the goal is for devices to hook logic into On Player Activated Zone. If there is no required time in the zone, we immediately activate, otherwise we add to a list of actors that are waiting to enter the zone. When we think the oldest actor on the list is ready to activate, we check against them and activate before removing them and starting a new timer for the now oldest actor.
  (Utilities|IsValid Controller
    (:Is%20Valid
      (if (> (Variables|Settings|GetRequiredTimeInZone) 0.0)
        (ZoneTimer|AddToPlayerTriggerList (Variables|GetController))
        (if (not (Utilities|Time|IsTimerActivebyHandle (Variables|ZoneTimer|GetZoneTimer)))
          (bind _returnvalue (Utilities|Time|SetTimerbyEvent (EventDispatchers|CreateEvent :!event Check%20Zone%20Timer) (Variables|Settings|GetRequiredTimeInZone)))
          (Variables|ZoneTimer|SetZoneTimer _returnvalue))
        (else
          (PlayerInteraction|PlayerActivatedZone (Variables|GetController)))))))

(fn TestPawnWithoutPlayerController
  (params (Actor :type Actor%20Object%20Reference))
  ;[1 1 1] If a player pawn is directly spawned into a volume, it will not have a player controller on overlap. We need the player controller to check against interaction, so we want to wait until the pawn is possessed.
  (bind _asfort_player_pawn (Utilities|Casting|CastToFortPlayerPawn Actor)
    (:then
      (Utilities|IsValid (Pawn|GetController _asfort_player_pawn)
        (:Is%20Valid
          (return false))
        (:Is%20Not%20Valid
          (Game|Player|BindEventtoOnPawnPossessed _asfort_player_pawn (EventDispatchers|CreateEvent :!event Empty%20Pawn%20Possessed))
          (return true))))
    (:CastFailed
      (return false))))

(fn UpdateOverlappingActors
  (params (Add :type Boolean) (Actor :type Actor%20Object%20Reference%20%28by%20ref%29))
  (bind _outputdelegate (EventDispatchers|CreateEvent :!event Overlapped%20Player%20Pawn%20Died))
  (bind _output_get (LocalVariables|SetLocalParticipant (ParticipantInteraction|GetLastParticipantController Actor)))
  (Utilities|IsValid _output_get
    (:Is%20Valid
      (bind _asfort_player_pawn (Utilities|Casting|CastToFortPlayerPawn (Variables|GetActor))
        (:then
          (if (not (Variables|/Script/FortniteGame.FortPawn|GetbIsDying _asfort_player_pawn))
            (if (Variables|GetAdd)
              (if (Utilities|Set|ContainsItem (Variables|ParticipantInteraction|GetOverlappingParticipants) (LocalVariables|GetLocalParticipant))
                (return false)
                (else
                  (Utilities|Set|Add (Variables|ParticipantInteraction|GetOverlappingParticipants) (LocalVariables|GetLocalParticipant))
                  (bind _asfort_player_pawn_2 (Utilities|Casting|CastToFortPlayerPawn (Variables|GetActor))
                    (:then
                      (EventDispatchers|BindEventtoOnDied _asfort_player_pawn_2 _outputdelegate))
                    (:CastFailed
                      _))
                  (return true)))
              (else
                _))
            (else
              _))
          (bind _returnvalue (Utilities|Set|Remove (Variables|ParticipantInteraction|GetOverlappingParticipants) (LocalVariables|GetLocalParticipant)))
          (if _returnvalue
            (bind _asfort_player_pawn_3 (Utilities|Casting|CastToFortPlayerPawn (Pawn|GetControlledPawn (LocalVariables|GetLocalParticipant)))
              (:then
                (EventDispatchers|UnbindEventfromOnDied _asfort_player_pawn_3 _outputdelegate))
              (:CastFailed
                _))
            (return true)
            (else
              (return false))))
        (:CastFailed
          _)))
    (:Is%20Not%20Valid
      (return false))))

(fn UpdateOverlaps
  (Utilities|SwitchHasAuthority
    (:Authority
      (bind _enabled_component (Variables|Default|GetEnabled%20Component)
        (:then
          (if (Class|CreativeEnabledComponent|IsEnabled _enabled_component)
            (PlayerInteraction|ClearOverlaps) ; Clear overlaps before cycling. Avoids having\nto query the overlap lists for each player triggering\nthe exit overlap events
            ;[1 1 1] Cycle the collision off then on, triggering various leave/enter events.
            (Collision|SetCollisionEnabled "NoCollision")
            (Collision|SetCollisionEnabled "QueryOnly")))
        (:else)))))

(fn UpdateScale
  (params (Multiplier :type Float%20%28double-precision%29 :default 1.0))
  (bind (_returnvalue _scale_z) (ZoneState|UpdateScale|GetShapetoMScale))
  (bind (_returnvalue_x _returnvalue_y _returnvalue_z) (Math|Vector|BreakVector (Math|Conversions|ToVector(Float) (* Multiplier (Variables|ScaleOptions|GetZone%20Scale)))))
  (bind (_returnvalue_x_2 _returnvalue_y_2 _returnvalue_z_2) (Math|Vector|BreakVector (* (* (Variables|ScaleOptions|GetZone%20Scale%20Multipliers) _returnvalue) (LocalVariables|GetScaleMultiplier))))
  (bind (_relativescale3d_x _relativescale3d_y _relativescale3d_z) (Math|Vector|BreakVector (Variables|GetRelativeScale3D_X)))
  ;[0 1 1 0.25] Update the scale of the mesh, optionally by an incoming multiplier (for scaling over time). If any of our Zone Scale Multipliers is set to 0, use the relative scale of the mesh.
  (Variables|SetScaleMultiplier_X (Math|Vector|MakeVector _returnvalue_x _returnvalue_y (select _scale_z _returnvalue_z (Variables|ScaleOptions|GetZone%20Scale)))) ; Some meshes want to scale in all directions when \nscaling up or down. Some always want Z to be max scale\n(so that it scales from middle out) 
  (Transformation|SetRelativeScale3D self (Math|Vector|MakeVector (Math|Float|Max(Float) (Math|Float|SelectFloat _returnvalue_x_2 _relativescale3d_x (!= _returnvalue_x_2 0)) 0.01) (Math|Float|Max(Float) (Math|Float|SelectFloat _returnvalue_y_2 _relativescale3d_y (!= _returnvalue_y_2 0)) 0.01) (Math|Float|Max(Float) (Math|Float|SelectFloat _returnvalue_z_2 _relativescale3d_z (!= _returnvalue_z_2 0)) 0.01)))
  (Rendering|SetVisibility (not (Math|Float|NearlyEqual(Float) Multiplier))))
