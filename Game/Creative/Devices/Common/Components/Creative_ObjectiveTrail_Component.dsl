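; Creative device component that shows an objective trail to a tracked actor for the instigator or all team-eligible players.
; ToggleObjectiveTrail updates a replicated player array or parity-encoded activation counter, granting a tracking ability on PlayerState to survive respawns.
; RepNotifies add or remove the objective in a local breadcrumb registry, rechecked on team changes.
; It binds to minigame state with timer retry, clearing registration and abilities at round or game end.
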(event EventBeginPlay
  (CallFunction|InitializeStateChangeBinding))

(event Custom|InitializeStateChangeBinding
  (bind _returnvalue (Components|GetOwner))
  (bind _returnvalue_2 (Minigame|GetMinigame _returnvalue :!class /Script/FortniteGame.MinigameBlueprintLibrary))
  (Utilities|Time|ClearandInvalidateTimerbyHandle (Variables|Default|GetRetryBindMinigameEventsTimer))
  (Utilities|IsValid _returnvalue_2
    (:Is%20Valid
      (Variables|Default|SetMinigameReference _returnvalue_2)
      ;[1 1 1] Remove this actor from the registry when the game ends.  Bind Events to Players Joining in Progress and Changing Teams
      (Minigame|BindEventtoStateChangedDelegate _returnvalue_2 (delegate Custom|StateChanged)))
    (:Is%20Not%20Valid
      ;[0.5 0 0] Failed to bind delegate - Retry - Mainly occurs in replays when the minigame hasnt been set or created yet.
      (bind _returnvalue_3 (Utilities|Time|SetTimerbyEvent (EventDispatchers|CreateEvent :!event Custom|InitializeStateChangeBinding) 0.1))
      (Variables|Default|SetRetryBindMinigameEventsTimer _returnvalue_3))))

(event Custom|StateChanged
  (params (Minigame :type Fort%20Minigame%20Object%20Reference) (MinigameState :type EFortMinigameState%20Enum :default "PreGame"))
  ;[1 1 1] Remove this actor from the registry when the game ends.  Bind Events to Players Joining in Progress and Changing Teams
  (switch Utilities|FlowControl|Switch|SwitchonEFortMinigameState MinigameState
    (:PostRoundEnd
      _)
    (:PostGameEnd
      _)
    (:PostGameAbandon
      _)
    (:PostGameReset
      _))
  (Class|CreativeIndexedObjectRegisterComp|RemoveObjectfromIndex (Variables|Default|GetObjective%20Register) (Variables|Default|GetObjective%20to%20Track) 0)
  (for _array_element (Variables|Default|GetPlayerArray)
    (if (Utilities|IsValid (Ability|GetAbilitySystemComponent _array_element) :!class /Script/Engine.KismetSystemLibrary)
      (bind _outabilityhandles (GameplayAbilities|FindAllAbilitieswithTags (Ability|GetAbilitySystemComponent _array_element) "(GameplayTags=((TagName=\"Creative.Components.ObjectiveTrail.Granted\")))" true))
      (for _array_element_2 _outabilityhandles
        (GameplayAbilities|ClearAbility (Ability|GetAbilitySystemComponent _array_element) _array_element_2))))
  (Variables|Settings|SetActivatedForAll 0)
  (Variables|Default|SetPlayerArray (Utilities|Array|MakeArray)))

(fn AddAbilitytoPlayer
  (params (AffectedController :type Controller%20Object%20Reference))
  (Utilities|IsValid AffectedController
    (:Is%20Valid
      (bind _activate (CallFunction|ShouldActivateTeamAffiliation AffectedController))
      (if _activate
        ;[1 1 1] Using PlayerState so Ability remains when player respawns
        (bind _asfort_player_state (Utilities|Casting|CastToFortPlayerState (Variables|/Script/Engine.Controller|GetPlayerState AffectedController)))
        (Ability|Advanced|GrantandActivateAbilityonSelf (Ability|GetAbilitySystemComponent _asfort_player_state) "/Game/Creative/Devices/Common/GameplayAbilities/GA_Creative_ObjectiveLocationTracking.GA_Creative_ObjectiveLocationTracking_C" _asfort_player_state)))))

(fn GetActivatedForAll
  (return (and (> (Variables|Settings|GetActivatedForAll) 0) (== (Math|Integer|%25%28Integer%29 (Variables|Settings|GetActivatedForAll) 2) 0))))

(fn ListenforLocalPlayerTeamChange
  (params (Listen :type Boolean))
  (bind _outputdelegate (EventDispatchers|CreateEvent :!event Update%20Registry%20For%20Local%20Player))
  (Networking|CanExecuteCosmeticEvents
    (:True
      (for _array_element (Fortnite|GetLocalFortPlayerControllers true)
        (bind _asfort_player_state_athena (Utilities|Casting|CastToFortPlayerStateAthena (Variables|/Script/Engine.Controller|GetPlayerState _array_element)))
        (if Listen
          (Player|BindEventtoOnTeamIndexChangedDelegate _asfort_player_state_athena _outputdelegate)
          (else
            (Player|UnbindEventfromOnTeamIndexChangedDelegate _asfort_player_state_athena _outputdelegate)))))))

(fn OnRep_ActivatedForAll
  (CallFunction|UpdateRegistryforLocalPlayer) ; Adds or removes the owner to/from a registry\nthat contains all objects that want to be tracked.
  (CallFunction|ListenForLocalPlayerTeamChange (CallFunction|GetActivatedforAll))
  ;[0.02792 1 0.2076 0.2] For authority, if we activate for all, give everyone the ability. (No need to remove if not, it cleans itself up if it can't find any objects to track)
  (Networking|SwitchHasAuthority
    (:Authority
      (if (CallFunction|GetActivatedforAll)
        (bind _minigamereference (Variables|Default|GetMinigameReference)
          (:then
            (bind _outplayers (Minigame|GetParticipatingPlayers_NotPure _minigamereference))
            (for _array_element _outplayers
              (Utilities|IsValid _array_element
                (:Is%20Valid
                  (CallFunction|AddAbilityToPlayer (PlayerState|GetPlayerController _array_element))))))
          (:else))))))

(fn OnRep_PlayerArray
  (CallFunction|UpdateRegistryforLocalPlayer)
  (for _array_element (Fortnite|GetLocalFortPlayerControllers true)
    (CallFunction|ListenForLocalPlayerTeamChange (Utilities|Array|ContainsItem (Variables|Default|GetPlayerArray) _array_element))))

(fn SetActivatedForAll
  (params (Activate :type Boolean))
  (Variables|Settings|SetActivatedForAll (+ (Variables|Settings|GetActivatedForAll) (Math|Integer|SelectInt (Math|Integer|SelectInt 2 1 (== (Math|Integer|%25%28Integer%29 (Variables|Settings|GetActivatedForAll) 2) 0)) (Math|Integer|SelectInt 1 2 (== (Math|Integer|%25%28Integer%29 (Variables|Settings|GetActivatedForAll) 2) 0)) (Variables|GetActivate)))))

(fn ShouldActivateTeamAffiliation
  (params (AffectedController :type Controller%20Object%20Reference))
  (bind _creative_use_restriction_ref (Variables|Default|GetCreative%20Use%20Restriction%20Ref)
    (:then
      (bind _valid_controller (CreativeUserRestrictions|ActorHasValidParticipantController _creative_use_restriction_ref AffectedController true true true))
      (if _valid_controller
        (if (not (Variables|Settings|GetShow%20for%20Friendly))
          (return false))
        (elif (Variables|Settings|GetShow%20for%20Enemy)
          _
          (else
            _)))
      (return true))
    (:else
      (if (or (Knobs|Team|IsTeamAll (Variables|Settings|GetIn%20Team%20Value)) (Knobs|Team|IsTeamNone (Variables|Settings|GetIn%20Team%20Value)))
        (return true)
        (elif (or (Knobs|Team|IsTeamFriendlies (Variables|Settings|GetIn%20Team%20Value)) (Knobs|Team|IsTeamAllies (Variables|Settings|GetIn%20Team%20Value)))
          _
          (elif (or (Knobs|Team|IsTeamEnemies (Variables|Settings|GetIn%20Team%20Value)) (Knobs|Team|IsTeamHostiles (Variables|Settings|GetIn%20Team%20Value)))
            _
            (else
              (switch Utilities|FlowControl|Switch|SwitchonEFortTeamAffiliation (Team|GetTeamAffiliationBetweenTeams (Teams|IndextoTeam (Knobs|Team|GetTeamValueasIndex (Variables|Settings|GetIn%20Team%20Value))) (Team|GetActorTeam AffectedController))
                (:Friendly
                  _)
                (:Neutral
                  (return true))
                (:Hostile
                  _)))))))))

(fn ToggleObjectiveTrail
  (params (Activate :type Boolean) (ObjectiveToTrack :type Actor%20Object%20Reference) (InstigatorController :type Controller%20Object%20Reference) (InTeamValue :type Integer) (ShowForFriendly :type Boolean) (ShowForEnemy :type Boolean) (ShowForInstigatorOnly :type Boolean) (PlayAudio :type Boolean) (PulseRate :type Float%20%28double-precision%29) (creative%20use%20restriction :type Creative%20Use%20Restriction%20Component%20Object%20Reference))
  (bind _instigatorcontroller (Variables|GetInstigatorController))
  (bind _localarray (LocalVariables|GetLocalArray))
  ;[1 1 1] Binding Variables.   It might be better to set up an interface on the device since almost all of these values have been replicated already on the device
  (Variables|Settings|SetActivateTrail Activate)
  (Variables|Default|SetObjective%20to%20Track ObjectiveToTrack)
  (LocalVariables|SetLocalArray (Variables|Default|GetPlayerArray))
  (Variables|Settings|SetIn%20Team%20Value InTeamValue)
  (Variables|Settings|SetShow%20for%20Friendly ShowForFriendly)
  (Variables|Settings|SetShow%20for%20Enemy ShowForEnemy)
  (Variables|Settings|SetShow%20for%20Instigator%20Only ShowForInstigatorOnly)
  (Variables|Settings|SetPlay%20Audio PlayAudio)
  (Variables|Settings|SetPulse%20Rate PulseRate)
  (Variables|Default|SetCreative%20Use%20Restriction%20Ref creative%20use%20restriction)
  ;[1 1 1] We are adding/Removing Players from an array tracking who should be able to track the device here.  Then Setting the Array with a repnotify so the client knows if they can track aswell
  (if (Variables|Settings|GetShow%20for%20Instigator%20Only)
    (Variables|GetInstigatorController
      (:then
        ;;[1 1 1] Activate or Deactivate?
        (if (Variables|Settings|GetActivateTrail)
          (Utilities|Array|AddUnique (LocalVariables|GetLocalArray) _instigatorcontroller)
          (CallFunction|AddAbilityToPlayer _instigatorcontroller)
          (elif (Utilities|Array|ContainsItem _localarray (Variables|GetInstigatorController))
            (Utilities|Array|RemoveItem _localarray (Variables|GetInstigatorController))))
        (Variables|Default|SetPlayerArray (LocalVariables|GetLocalArray))
        (Networking|FlushNetDormancy (Components|GetOwner)) ; Added because without it the component doesn't replicate to the client
      )
      (:else))
    (else
      (Default|SetActivatedforAll (Variables|Settings|GetActivateTrail))
      (Networking|FlushNetDormancy (Components|GetOwner)) ; Added because without it the component doesn't replicate to the client
    )))

(fn UpdateRegistryForLocalPlayer
  (bind _objective_to_track (Variables|Default|GetObjective%20to%20Track))
  (bind _objective_register (Variables|Default|GetObjective%20Register))
  (Networking|CanExecuteCosmeticEvents
    (:True
      (LocalVariables|SetShouldGetOrCreate false)
      (for _array_element (Fortnite|GetLocalFortPlayerControllers true)
        (bind _activate (CallFunction|ShouldActivateTeamAffiliation _array_element))
        (if _activate
          (if (or (Utilities|Array|ContainsItem (Variables|Default|GetPlayerArray) _array_element) (CallFunction|GetActivatedforAll))
            (bind _returnvalue (Minigame|GetMinigameforActor (Components|GetOwner)))
            (if (Utilities|Enum|Equal(Enum) (Variables|/Script/FortniteGame.FortMinigame|GetCurrentState _returnvalue) "InProgress")
              (LocalVariables|SetShouldGetOrCreate true)))))
      (if (LocalVariables|GetShouldGetOrCreate)
        (CreativeObjectRegistry|GetOrCreateObjectRegister _objective_to_track "(TagName=\"Creative.Register.BreadcrumbTargets\")" true 0 _objective_register)
        (else
          (Class|CreativeIndexedObjectRegisterComp|RemoveObjectfromIndex _objective_register _objective_to_track 0))))))
