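; Creative device component that hides registered scene components while the local player captures in Photo Mode, then shows them again.
; On BeginPlay, on cosmetic-capable machines only, it finds the player controller's Photo Mode component and binds to its state-change callback, unbinding on EndPlay.
; Setter events supply an always-hide array and a conditional map, whose entries toggle only when true.
; Setup and setters apply hiding immediately if capture is already active.
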(event EventBeginPlay
  ;[1 1 1] Photo Mode only captures on the client, so don't hide anything on the server!
  (Networking|CanExecuteCosmeticEvents
    (:True
      (bind _output_get (Variables|Default|SetPhotoModeControllerComponent (Actor|GetComponentbyClass (Game|GetPlayerController 0) "/Game/Athena/HUD/Creative/Input/Creative_PhotoMode_ControllerComponent.Creative_PhotoMode_ControllerComponent_C")))
      (Utilities|IsValid _output_get
        (:Is%20Valid
          (EventDispatchers|BindEventtoPhotoModeStateChangedCallback (Variables|Default|GetPhotoModeControllerComponent) (delegate Custom|PhotoModeStateChanged))
          ;[1 1 1] If we are already in photo mode, need to hide everything immediately
          (if (Variables|/Script/FortniteGame.FortControllerComponent_PhotoMode|GetbIsCapturingPhoto (Variables|Default|GetPhotoModeControllerComponent))
            (CallFunction|HideorShowComponents)))))))

(event Custom|PhotoModeStateChanged
  (CallFunction|HideorShowComponents))

(event EventEndPlay
  (params (EndPlayReason :type EEndPlayReason%20Enum :default "Destroyed"))
  (bind _photomodecontrollercomponent (Variables|Default|GetPhotoModeControllerComponent))
  (Utilities|IsValid _photomodecontrollercomponent
    (:Is%20Valid
      (EventDispatchers|UnbindEventfromPhotoModeStateChangedCallback _photomodecontrollercomponent (delegate Custom|PhotoModeStateChanged)))))

(event Custom|SetComponentsToAlwaysHide
  (params (ComponentsToAlwaysHide :type Array%20of%20Scene%20Component%20Object%20References))
  (bind _photomodecontrollercomponent (Variables|Default|GetPhotoModeControllerComponent))
  (Networking|CanExecuteCosmeticEvents
    (:True
      (Variables|Default|SetComponentsToAlwaysHide ComponentsToAlwaysHide)
      ;[1 1 1] If we are already in photo mode, need to hide everything immediately
      (Utilities|IsValid _photomodecontrollercomponent
        (:Is%20Valid
          (if (Variables|/Script/FortniteGame.FortControllerComponent_PhotoMode|GetbIsCapturingPhoto _photomodecontrollercomponent)
            (CallFunction|HideorShowComponents)))))))

(event Custom|SetComponentsToConditionallyHide
  (params (ComponentsToConditionallyHide :type Map%20of%20Scene%20Component%20Object%20References%20to%20Booleans))
  (bind _photomodecontrollercomponent (Variables|Default|GetPhotoModeControllerComponent))
  (Networking|CanExecuteCosmeticEvents
    (:True
      (Variables|Default|SetComponentsToConditionallyHide ComponentsToConditionallyHide)
      ;[1 1 1] If we are already in photo mode, need to hide everything immediately
      (Utilities|IsValid _photomodecontrollercomponent
        (:Is%20Valid
          (if (Variables|/Script/FortniteGame.FortControllerComponent_PhotoMode|GetbIsCapturingPhoto _photomodecontrollercomponent)
            (CallFunction|HideorShowComponents)))))))

(fn HideOrShowComponents
  (for _array_element (Variables|Default|GetComponentsToAlwaysHide)
    (Development|SetHiddeninGame _array_element (Variables|/Script/FortniteGame.FortControllerComponent_PhotoMode|GetbIsCapturingPhoto (Variables|Default|GetPhotoModeControllerComponent)) false))
  (bind _keys (Utilities|Map|Keys (Variables|Default|GetComponentsToConditionallyHide)))
  (for _array_element_2 _keys
    ;[1 1 1] Only make the component visible if we hid it
    (if (Utilities|Map|Find (Variables|Default|GetComponentsToConditionallyHide) _array_element_2)
      (Development|SetHiddeninGame _array_element_2 (Variables|/Script/FortniteGame.FortControllerComponent_PhotoMode|GetbIsCapturingPhoto (Variables|Default|GetPhotoModeControllerComponent)) false))))
