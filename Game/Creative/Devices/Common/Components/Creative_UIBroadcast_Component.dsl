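; Creative UI broadcast component that hooks minigame events to trigger a per-player HUD hook, ApplyHUDGA, which is empty here.
; Initialize takes a minigame progress component and, on authority, binds to the current minigame, or waits for OnAddedToMinigame if none is valid yet.
; Minigame start calls ApplyHUDGA for every participating player's controller.
; A join in progress calls it for the joining pawn's controller.
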(event EventBeginPlay)

(event Custom|Initialize
  (params (MinigameProgressComponent :type Fort%20Minigame%20Progress%20Component%20Object%20Reference) (ShouldInitialize :type Boolean :default true))
  (bind _currentminigame (Variables|/Script/FortniteGame.FortMinigameLogicComponent|GetCurrentMinigame MinigameProgressComponent))
  ;[1 1 1] Initialized
  (if ShouldInitialize
    (Networking|SwitchHasAuthority
      (:Authority
        (Utilities|IsValid _currentminigame
          (:Is%20Valid
            (Variables|Default|SetFort%20Minigame _currentminigame)
            (Minigame|BindEventtoMinigameStartedDelegate (Variables|Default|GetFort%20Minigame) (EventDispatchers|CreateEvent :!event Custom|EventMinigameStarted) :!class /Script/FortniteGame.FortMinigame)
            (Minigame|BindEventtoOnPlayerJoininProgress (Variables|Default|GetFort%20Minigame) (EventDispatchers|CreateEvent :!event Custom|Player%20Joined%20In%20Progress)))
          (:Is%20Not%20Valid
            (Minigame|BindEventtoOnAddedtoMinigame MinigameProgressComponent (delegate Custom|OnAddedToMinigame))))))))

(event Custom|OnAddedToMinigame
  (params (Minigame :type Fort%20Minigame%20Object%20Reference))
  ;[1 1 1] Initialized
  (Variables|Default|SetFort%20Minigame Minigame)
  (Minigame|UnbindEventfromOnAddedtoMinigame (stale Custom|Initialize MinigameProgressComponent) (delegate Custom|OnAddedToMinigame))
  (Minigame|BindEventtoMinigameStartedDelegate (Variables|Default|GetFort%20Minigame) (EventDispatchers|CreateEvent :!event Custom|EventMinigameStarted) :!class /Script/FortniteGame.FortMinigame)
  (Minigame|BindEventtoOnPlayerJoininProgress (Variables|Default|GetFort%20Minigame) (EventDispatchers|CreateEvent :!event Custom|Player%20Joined%20In%20Progress)))

(event Custom|EventMinigameStarted
  ;[1 1 1] Minigame Started
  (for _array_element (Minigame|GetParticipatingPlayers (Variables|Default|GetFort%20Minigame))
    (CallFunction|ApplyHUDGA (PlayerState|GetPlayerController _array_element))))

(event Custom|Player%20Joined%20In%20Progress
  (params (Pawn :type Pawn%20Object%20Reference))
  ;[1 1 1] Join in Progress
  (CallFunction|ApplyHUDGA (Fortnite|GetFortPlayerControllerfromActor Pawn)))

(fn ApplyHUDGA
  (params (PlayerController :type Player%20Controller%20Object%20Reference))
  (return))
