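; Creative device component that decides whether an actor, pawn, vehicle driver or controller may use the owning device.
; StateChecks combines enabled, play mode, same-minigame, and invertible team and class tests, with edit permission as an optional override.
; A cached pawn-to-controller map and possession callbacks handle pawns lacking controllers.
; Participant checks can exclude guard NPCs or fall back to deprecated player-only logic, and restriction settings are mirrored into native variables.
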(event EventBeginPlay
  ;[1 1 1] To Do: Set native variables on all instances and remove BP variables
  (CallFunction|SetNativeVariables))

(fn ActorHasValidController
  (params (Actor :type Actor%20Object%20Reference) (Include%20Restriction%20Check :type Boolean :default true) (Additional%20Checks :type Boolean :default true))
  (bind _controller (LocalVariables|GetController))
  (bind _asfort_player_state (Utilities|Casting|CastToFortPlayerState Actor)
    (:then
      (LocalVariables|SetController (Fortnite|GetControllerfromActor _asfort_player_state)))
    (:CastFailed
      (bind _ascontroller (Utilities|Casting|CastToController Actor)
        (:then
          (LocalVariables|SetController _ascontroller)
          (bind _asfort_player_pawn (Utilities|Casting|CastToFortPlayerPawn (Pawn|GetControlledPawn (LocalVariables|GetController)))
            (:then
              (LocalVariables|SetPawn _asfort_player_pawn)
              (LocalVariables|GetController
                (:then
                  (PlayertoPawnLink|CachePawn/ControllerLink (LocalVariables|GetPawn) (LocalVariables|GetController)))
                (:else
                  (LocalVariables|SetController (Utilities|Map|Find (Variables|PlayertoPawnLink|GetOverlappingPawnControllerMap) (LocalVariables|GetPawn)))
                  (LocalVariables|GetController
                    (:then
                      _)
                    (:else
                      (Game|Player|BindEventtoOnPawnPossessed (LocalVariables|GetPawn) (EventDispatchers|CreateEvent :!event Empty%20Player%20Pawn%20Possessed))
                      (return false (LocalVariables|GetPawn) 0))))))
            (:CastFailed
              (Pawn|BindEventtoOnPossessedPawnChanged (LocalVariables|GetController) (EventDispatchers|CreateEvent :!event ControllerPossessedPawnChanged))))
          (if (Variables|GetInclude%20Restriction%20Check)
            (bind _asplayer_controller (Utilities|Casting|CastToPlayerController _controller)
              (:then
                _)
              (:CastFailed
                _))
            (return (DEPRECATED|CanPlayerControllerInteract (LocalVariables|GetController) (Variables|GetAdditional%20Checks)) (LocalVariables|GetPawn) _asplayer_controller)
            (else
              (bind _asplayer_controller_2 (Utilities|Casting|CastToPlayerController _controller)
                (:then
                  _)
                (:CastFailed
                  _))
              (return true (LocalVariables|GetPawn) _asplayer_controller_2))))
        (:CastFailed
          (bind _asfort_player_pawn_2 (Utilities|Casting|CastToFortPlayerPawn Actor)
            (:then
              (LocalVariables|SetPawn _asfort_player_pawn_2))
            (:CastFailed
              (bind _returnvalue (Vehicle|GetDriver(Message) Actor))
              (LocalVariables|SetPawn _returnvalue)))
          (bind _pawn (LocalVariables|GetPawn)
            (:then
              (LocalVariables|SetController (Pawn|GetController _pawn)))
            (:else
              (return false 0 0))))))))

(fn ActorHasValidParticipantController
  (params (Actor :type Actor%20Object%20Reference) (Include%20Restriction%20Check :type Boolean :default true) (Include%20Guard%20NPC :type Boolean :default true) (Additional%20Checks :type Boolean :default true))
  (bind _localcontroller (LocalVariables|GetLocalController))
  (if (not (Variables|Internal|GetShouldUseLegacyLogic))
    (bind _asfort_player_state (Utilities|Casting|CastToFortPlayerState Actor)
      (:then
        (LocalVariables|SetLocalController (Fortnite|GetControllerfromActor _asfort_player_state)))
      (:CastFailed
        (bind _ascontroller (Utilities|Casting|CastToController Actor)
          (:then
            (LocalVariables|SetLocalController _ascontroller)
            (bind _asfort_player_pawn (Utilities|Casting|CastToFortPlayerPawn (Pawn|GetControlledPawn (LocalVariables|GetLocalController)))
              (:then
                (LocalVariables|SetPawn _asfort_player_pawn)
                (LocalVariables|GetLocalController
                  (:then
                    (PlayertoPawnLink|CachePawn/ControllerLink (LocalVariables|GetPawn) (LocalVariables|GetLocalController)))
                  (:else
                    (LocalVariables|SetLocalController (Utilities|Map|Find (Variables|PlayertoPawnLink|GetOverlappingPawnControllerMap) (LocalVariables|GetPawn)))
                    (LocalVariables|GetLocalController
                      (:then
                        _)
                      (:else
                        (Game|Player|BindEventtoOnPawnPossessed (LocalVariables|GetPawn) (EventDispatchers|CreateEvent :!event Empty%20Player%20Pawn%20Possessed))
                        (return false (LocalVariables|GetPawn) 0))))))
              (:CastFailed
                (Pawn|BindEventtoOnPossessedPawnChanged (LocalVariables|GetLocalController) (EventDispatchers|CreateEvent :!event ControllerPossessedPawnChanged))))
            (if (Variables|GetInclude%20Restriction%20Check)
              (Utilities|Casting|CastToPlayerController _localcontroller
                (:then
                  _)
                (:CastFailed
                  (Utilities|Casting|CastToFortAthenaAIBotController (LocalVariables|GetLocalController)
                    (:then
                      _)
                    (:CastFailed
                      _))))
              (return (CreativeUserRestrictions|CanParticipantControllerInteract (LocalVariables|GetLocalController) (Variables|GetInclude%20Guard%20NPC) (Variables|GetAdditional%20Checks)) (LocalVariables|GetPawn) (LocalVariables|GetLocalController))
              (else
                (Utilities|Casting|CastToPlayerController _localcontroller
                  (:then
                    _)
                  (:CastFailed
                    (Utilities|Casting|CastToFortAthenaAIBotController (LocalVariables|GetLocalController)
                      (:then
                        (if (not (Variables|GetInclude%20Guard%20NPC))
                          (return false (LocalVariables|GetPawn) (LocalVariables|GetLocalController))))
                      (:CastFailed
                        _))
                    (return true (LocalVariables|GetPawn) (LocalVariables|GetLocalController)))))))
          (:CastFailed
            (bind _asfort_player_pawn_2 (Utilities|Casting|CastToFortPlayerPawn Actor)
              (:then
                (LocalVariables|SetPawn _asfort_player_pawn_2))
              (:CastFailed
                (bind _returnvalue (Vehicle|GetDriver(Message) Actor))
                (LocalVariables|SetPawn _returnvalue)))
            (bind _pawn (LocalVariables|GetPawn)
              (:then
                (LocalVariables|SetLocalController (Pawn|GetController _pawn)))
              (:else
                (return false 0 0)))))))
    (else
      (bind (_valid_controller _fortplayerpawn _playercontroller) (DEPRECATED|ActorHasValidController Actor Include%20Restriction%20Check Additional%20Checks))
      (return _valid_controller _fortplayerpawn _playercontroller))))

(fn BP_IsEnabled
  (return (StateChecks|IsEnabled)))

(fn CachePawn/ControllerLink
  (params (Pawn :type Fort%20Player%20Pawn%20Object%20Reference%20%28by%20ref%29) (Controller :type Controller%20Object%20Reference%20%28by%20ref%29))
  (Utilities|Map|Add (Variables|PlayertoPawnLink|GetOverlappingPawnControllerMap) Pawn Controller)
  (Game|BindEventtoOnEndPlay Pawn (EventDispatchers|CreateEvent :!event On%20Pawn%20Destroyed))
  (Game|Player|UnbindEventfromOnPawnPossessed Pawn (EventDispatchers|CreateEvent :!event Empty%20Player%20Pawn%20Possessed)))

(fn CanActorInteract
  (params (Actor :type Actor%20Object%20Reference) (Additional%20Checks :type Boolean :default true) (OR%20can%20Customize%20Device :type Boolean))
  ;[1 1 1] Returns true if all of the checks are true. By default these will be unless restrictions are placed.
  (LocalVariables|SetLastController (PlayertoPawnLink|GetLastPlayerController Actor))
  (bind _lastcontroller (LocalVariables|GetLastController)
    (:then
      (return (DEPRECATED|CanPlayerControllerInteract _lastcontroller (Variables|GetAdditional%20Checks) (Variables|GetOR%20can%20Customize%20Device))))
    (:else
      (if (Variables|RestrictionSettings|GetAllow%20Actors%20without%20Controllers)
        (return (CallFunction|StateChecks (Variables|GetActor) (Variables|GetAdditional%20Checks)))
        (else
          (return false))))))

(fn CanEditOptions
  (params (EditingController :type Controller%20Object%20Reference))
  (Utilities|IsValid EditingController
    (:Is%20Valid
      (bind _toyoptions (Variables|Default|GetToyOptions)
        (:then
          (bind _asfort_pawn (Utilities|Casting|CastToFortPawn (Pawn|GetControlledPawn EditingController))
            (:then
              (return (Interaction|CanEditOptionsComponent _toyoptions _asfort_pawn)))
            (:CastFailed
              (return false))))
        (:else
          (return true))))
    (:Is%20Not%20Valid
      (return false))))

(fn CanPlayerControllerInteract
  (params (Interacting%20Controller :type Controller%20Object%20Reference) (Additional%20Checks :type Boolean :default true) (OR%20can%20Customize%20Device :type Boolean))
  ;[1 1 1] Returns true if all of the checks are true. By default these will be unless restrictions are placed.
  (Utilities|IsValid Interacting%20Controller
    (:Is%20Valid
      (return (or (CallFunction|StateChecks (Variables|GetInteracting%20Controller) (Variables|GetAdditional%20Checks)) (and OR%20can%20Customize%20Device (StateChecks|CanEditOptions (Variables|GetInteracting%20Controller))))))
    (:Is%20Not%20Valid
      (return false))))

(fn CanParticipantActorInteract
  (params (Actor :type Actor%20Object%20Reference) (Include%20Guard%20NPC :type Boolean :default true) (Additional%20Checks :type Boolean :default true) (OR%20can%20Customize%20Device :type Boolean))
  ;[1 1 1] Returns true if all of the checks are true. By default these will be unless restrictions are placed.
  (if (not (Variables|Internal|GetShouldUseLegacyLogic))
    (LocalVariables|SetLastController (PlayertoPawnLink|GetLastPlayerController Actor))
    (bind _lastcontroller (LocalVariables|GetLastController)
      (:then
        (return (CreativeUserRestrictions|CanParticipantControllerInteract _lastcontroller (Variables|GetInclude%20Guard%20NPC) (Variables|GetAdditional%20Checks) (Variables|GetOR%20can%20Customize%20Device))))
      (:else
        (if (Variables|RestrictionSettings|GetAllow%20Actors%20without%20Controllers)
          (return (CallFunction|StateChecks (Variables|GetActor) (Variables|GetAdditional%20Checks)))
          (else
            (return false)))))
    (else
      (return (DEPRECATED|CanActorInteract Actor Additional%20Checks OR%20can%20Customize%20Device)))))

(fn CanParticipantControllerInteract
  (params (Interacting%20Controller :type Controller%20Object%20Reference) (Include%20Guard%20NPC :type Boolean :default true) (Additional%20Checks :type Boolean :default true) (OR%20can%20Customize%20Device :type Boolean))
  ;[1 1 1] Returns true if all of the checks are true. By default these will be unless restrictions are placed.
  (if (not (Variables|Internal|GetShouldUseLegacyLogic))
    (Utilities|IsValid Interacting%20Controller
      (:Is%20Valid
        ;;[1 1 1] If we include Guard & the target is Guard Actor, then do the check, otherwise returns false
        (bind _isguard (CreativeAI|NPCType|IsGuardActor (Variables|GetInteracting%20Controller)))
        (if _isguard
          (if (not (Variables|GetInclude%20Guard%20NPC))
            (return false))
          (else
            _))
        (return (or (CallFunction|StateChecks (Variables|GetInteracting%20Controller) (Variables|GetAdditional%20Checks)) (and (Variables|GetOR%20can%20Customize%20Device) (StateChecks|CanEditOptions (Variables|GetInteracting%20Controller))))))
      (:Is%20Not%20Valid
        (return false)))
    (else
      (return (DEPRECATED|CanPlayerControllerInteract Interacting%20Controller Additional%20Checks OR%20can%20Customize%20Device)))))

(fn ControllerPossessedPawnChanged
  (params (OldPawn :type Pawn%20Object%20Reference) (NewPawn :type Pawn%20Object%20Reference))
  (Utilities|IsValid NewPawn
    (:Is%20Valid
      (PlayertoPawnLink|EmptyPlayerPawnPossessed NewPawn))))

(fn EmptyPlayerPawnPossessed
  (params (NewPawn :type Pawn%20Object%20Reference))
  ;[1 1 1] A Callback that the owning actor can hook into to let it know a pawn is ready (a test was carried out against it when it was unpossessed).
  (bind _asfort_player_pawn (Utilities|Casting|CastToFortPlayerPawn NewPawn))
  (LocalVariables|SetOwningController (Fortnite|GetFortPlayerControllerfromActor NewPawn))
  (bind _owningcontroller (LocalVariables|GetOwningController)
    (:then
      (PlayertoPawnLink|CachePawn/ControllerLink _asfort_player_pawn _owningcontroller)
      (Default|CallPawnNowHasValidParticipantController _asfort_player_pawn (LocalVariables|GetOwningController))
      (bind _asfort_player_controller (Utilities|Casting|CastToFortPlayerController (LocalVariables|GetOwningController)))
      (DEPRECATED|CallPawnNowHasController _asfort_player_pawn _asfort_player_controller) ; TODO: Delete It Once Clean Up
    )
    (:else)))

(fn GetFailureReasons
  (params (Pawn :type Pawn%20Object%20Reference) (Device%20Failure%20Reason :type Text) (Append%20Device%20Reason :type Boolean))
  (bind _returnvalue (Pawn|GetController Pawn))
  (bind _text (LocalVariables|GetText))
  ;[1 0.1959 0 0.2] Appends a list of reasons why the input pawn cannot interact with the device in question.
  (Utilities|IsValid _returnvalue
    (:Is%20Valid
      (LocalVariables|SetText (select Append%20Device%20Reason Device%20Failure%20Reason (Utilities|Text|GetEmptyText)))
      (Utilities|AppendTextIfFail _text (Variables|FailureReasons|GetEnabled%20Fail%20Reason) (StateChecks|IsEnabled))
      (Utilities|AppendTextIfFail _text (Variables|FailureReasons|GetTeam%20Fail%20Reason) (StateChecks|IsValidTeam _returnvalue))
      (Utilities|AppendTextIfFail _text (Variables|FailureReasons|GetClass%20Fail%20Reason) (StateChecks|IsValidClass _returnvalue))
      (Utilities|AppendTextIfFail _text (Variables|FailureReasons|GetPlay%20Mode%20Fail%20Reason) (StateChecks|IsValidPlayState))
      (return (Utilities|Text|TextTrimPrecedingandTrailing (LocalVariables|GetText))))))

(fn GetLastPlayerController
  (params (Actor :type Actor%20Object%20Reference))
  (bind _ascontroller (Utilities|Casting|CastToController Actor)
    (:then
      (return _ascontroller))
    (:CastFailed
      (Utilities|FlowControl|Sequence
        (:then_0
          (bind _output_get (LocalVariables|SetController (Fortnite|GetControllerfromActor (Variables|GetActor))))
          (Utilities|IsValid _output_get
            (:Is%20Not%20Valid
              ;[1 1 1] Check to see if we're a vehicle that is being driven, and use the driver instead.
              (bind _returnvalue (Vehicle|GetDriver(Message) (Variables|GetActor))) ; are we a vehicle with a driver?
              (LocalVariables|SetPawn _returnvalue)
              (bind _pawn (LocalVariables|GetPawn)
                (:then
                  (bind _output_get_2 (LocalVariables|SetController (Pawn|GetController _pawn))))
                (:else
                  _))
              (Utilities|IsValid _output_get_2
                (:Is%20Not%20Valid
                  ;[1 1 1] Check for a Pawn without a valid owner
                  (LocalVariables|GetPawn
                    (:then
                      _)
                    (:else
                      (bind _asfort_player_pawn (Utilities|Casting|CastToFortPlayerPawn (Variables|GetActor))) ; we are a pawn with no controller.
                      (LocalVariables|SetPawn _asfort_player_pawn)
                      (LocalVariables|GetPawn
                        (:then
                          _)
                        (:else)))) ; Check if we've already set a pawn from\nthe driver
                  (LocalVariables|SetController (Utilities|Map|Find (Variables|PlayertoPawnLink|GetOverlappingPawnControllerMap) (LocalVariables|GetPawn))))) ; Do we have a controller yet?
            )) ; Do we have a controller yet?
        )
        (:then_1
          (return (LocalVariables|GetController)))))))

(fn GetNumberOfTeams
  (return (Playlist|GetMaxNumberOfTeams self)))

(fn Initialize
  (params (MinigameLogic :type Fort%20Minigame%20Logic%20Component%20Object%20Reference) (EnabledComponent :type Creative%20Enabled%20Component%20Object%20Reference) (ToyOptionsComponent :type Toy%20Options%20Component%20Object%20Reference) (Team :type Creative%20Team%20Option%20Structure) (Invert%20Team%20Selection :type Boolean) (Class :type Creative%20Class%20Option%20Structure) (Invert%20Class%20Selection :type Boolean) (Use%20Safe%20Class%20Indexing :type Boolean))
  (CallFunction|SetMinigameLogicComponent MinigameLogic)
  (Variables|Default|SetEnabledComponent EnabledComponent)
  (Variables|Default|SetToyOptions ToyOptionsComponent)
  (TeamRestrictions|SetTeamswithAffects (Math|Integer|SelectInt (Creative|TeamOption|GetLegacyTeamIndex Team) (Knobs|Team|GetShiftedAnyTeamValue) (Creative|TeamOption|IsTeamType Team)) Invert%20Team%20Selection)
  (ClassRestrictions|SetClassWithAffects (Creative|ClassOption|GetLegacyClassIndex Class) Invert%20Class%20Selection Use%20Safe%20Class%20Indexing)
  (Internal|SetShouldUseLegacyValidationLogic))

(fn InitializeComponent
  (params (MinigameLogic :type Fort%20Minigame%20Logic%20Component%20Object%20Reference) (EnabledComponent :type Creative%20Enabled%20Component%20Object%20Reference) (ToyOptionsComponent :type Toy%20Options%20Component%20Object%20Reference) (Required%20Single%20Team :type Integer :default 996) (Required%20Class :type Integer :default -1) (Required%20Item :type Fort%20Item%20Definition%20Object%20Reference) (Required%20Item%20Quantity :type Integer :default 1))
  (CallFunction|SetMinigameLogicComponent MinigameLogic)
  (Variables|Default|SetEnabledComponent EnabledComponent)
  (Variables|Default|SetToyOptions ToyOptionsComponent)
  (TeamRestrictions|SetInitialSingleTeam (Math|Integer|SelectInt Required%20Single%20Team (Knobs|Team|GetShiftedAnyTeamValue) (Knobs|Team|IsValidTeamIndex Required%20Single%20Team)))
  (ClassRestrictions|SetInitialSingleClass Required%20Class)
  (Variables|ItemSettings|SetRequired%20Item Required%20Item)
  (Variables|ItemSettings|SetRequired%20Item%20Quantity Required%20Item%20Quantity)
  (Internal|SetShouldUseLegacyValidationLogic))

(fn InitializeComponentwithAffects
  (params (MinigameLogic :type Fort%20Minigame%20Logic%20Component%20Object%20Reference) (EnabledComponent :type Creative%20Enabled%20Component%20Object%20Reference) (ToyOptionsComponent :type Toy%20Options%20Component%20Object%20Reference) (Required%20Single%20Team :type Integer :default 996) (Required%20Class :type Integer :default -1) (Required%20Item :type Fort%20Item%20Definition%20Object%20Reference) (Required%20Item%20Quantity :type Integer :default 1) (TeamAffectsAllButSelected :type Boolean) (ClassAffectsAllButSelected :type Boolean) (UsesSafeClassIndexing :type Boolean))
  (CallFunction|SetMinigameLogicComponent MinigameLogic)
  (Variables|Default|SetEnabledComponent EnabledComponent)
  (Variables|Default|SetToyOptions ToyOptionsComponent)
  (TeamRestrictions|SetTeamswithAffects (Math|Integer|SelectInt Required%20Single%20Team (Knobs|Team|GetShiftedAnyTeamValue) (Knobs|Team|IsValidTeamIndex Required%20Single%20Team)) TeamAffectsAllButSelected)
  (ClassRestrictions|SetClassWithAffects Required%20Class ClassAffectsAllButSelected UsesSafeClassIndexing)
  (Variables|ItemSettings|SetRequired%20Item Required%20Item)
  (Variables|ItemSettings|SetRequired%20Item%20Quantity Required%20Item%20Quantity)
  (Internal|SetShouldUseLegacyValidationLogic))

(fn IsValidTeam
  (params (Actor%20to%20Check :type Actor%20Object%20Reference))
  (if (< (Variables|TeamSettings|GetRequiredTeam) 0)
    (return (not (Variables|TeamSettings|GetInvertTeamRequirement)))
    (else
      (Utilities|IsValid Actor%20to%20Check
        (:Is%20Valid
          (return (!= (== (Variables|TeamSettings|GetRequiredTeam) (Math|Conversions|ToInteger(Byte) (Team|GetActorPvPTeamIndex Actor%20to%20Check))) (Variables|TeamSettings|GetInvertTeamRequirement))))
        (:Is%20Not%20Valid
          (return false))))))

(fn IsEnabled
  (bind _enabledcomponent (Variables|Default|GetEnabledComponent)
    (:then
      (return (Class|CreativeEnabledComponent|IsEnabled _enabledcomponent)))
    (:else
      (return true))))

(fn IsSameMinigame
  (params (Actor :type Actor%20Object%20Reference))
  (bind _minigamelogiccomponent (Utilities|MinigameLogic)
    (:IsValid
      (bind _currentminigame (Variables|/Script/FortniteGame.FortMinigameLogicComponent|GetCurrentMinigame _minigamelogiccomponent)
        (:then
          (LocalVariables|SetController (Fortnite|GetControllerfromActor Actor))
          (bind _controller (LocalVariables|GetController)
            (:then
              (bind _asfort_player_state (Utilities|Casting|CastToFortPlayerState (Variables|/Script/Engine.Controller|GetPlayerState _controller))
                (:then
                  ;[1 1 1] FIX FORT-609324, Return True if the target is an AI Bot (Guard),
                  ; otherwise returns the result from Is Player in Minigame
                  (return (select (Variables|/Script/Engine.PlayerState|GetbIsABot _asfort_player_state) true (Minigame|IsPlayerinMinigame _currentminigame _asfort_player_state false))))
                (:CastFailed
                  _)))
            (:else
              _))
          (return true))
        (:else
          _)))
    (:IsNotValid
      _)))

(fn IsValidClass
  (params (Actor :type Actor%20Object%20Reference))
  (bind _returnvalue (Fortnite|GetControllerfromActor Actor))
  (bind _requiredclass (Variables|ClassSettings|GetRequiredClass))
  (if (>= _requiredclass 0)
    (Utilities|IsValid _returnvalue
      (:Is%20Valid
        (bind _playerstate (Variables|/Script/Engine.Controller|GetPlayerState _returnvalue)
          (:then
            (bind _asfort_player_state (Utilities|Casting|CastToFortPlayerState _playerstate)
              (:then
                (bind _minigamelogiccomponent (Utilities|MinigameLogic)
                  (:IsValid
                    (bind _currentminigame (Variables|/Script/FortniteGame.FortMinigameLogicComponent|GetCurrentMinigame _minigamelogiccomponent)
                      (:then
                        (Utilities|IsValid (Minigame|GetFNEClassComponent _currentminigame)
                          (:Is%20Valid
                            (return (ClassRestrictions|TestMatchingClass (Math|Conversions|ToInteger(Byte) (Minigame|GetClassSlotforPlayer (Minigame|GetFNEClassComponent _currentminigame) _asfort_player_state false)))))
                          (:Is%20Not%20Valid
                            _)))
                      (:else
                        _))
                    (return false))
                  (:IsNotValid
                    (return true))))
              (:CastFailed
                _)))
          (:else
            _)))
      (:Is%20Not%20Valid
        (return false)))
    (else
      (return (!= (== _requiredclass -1) (Variables|ClassSettings|GetInvertClassRequirement))))))

(fn IsValidPlayState
  (if (Variables|RestrictionSettings|GetRequire%20Play%20Mode)
    (bind _minigamelogiccomponent (Utilities|MinigameLogic)
      (:IsValid
        (bind _currentminigame (Variables|/Script/FortniteGame.FortMinigameLogicComponent|GetCurrentMinigame _minigamelogiccomponent)
          (:then
            (return (Minigame|IsinPlayMode _currentminigame :!class /Script/FortniteGame.FortMinigame)))
          (:else
            (return false))))
      (:IsNotValid
        _))
    (else
      _))
  (return true))

(fn OnPawnDestroyed
  (params (Actor :type Actor%20Object%20Reference) (EndPlayReason :type EEndPlayReason%20Enum :default "Destroyed"))
  (bind _asfort_player_pawn (Utilities|Casting|CastToFortPlayerPawn Actor))
  (PlayertoPawnLink|RemovePawnFromCache _asfort_player_pawn))

(fn PrintStateChecks
  (params (Actor :type Actor%20Object%20Reference))
  (Utilities|SwitchHasAuthority
    (:Authority
      (if (or (Variables|Debug|GetDebug%20Server%20Print%20State%20Checks) (Class|CreativeBlueprintLibrary|IsUseRestrictionComponentDebuggingEnabled))
        _))
    (:Remote
      (if (Variables|Debug|GetDebug%20Client%20Print%20State%20Checks)
        _)))
  (Development|PrintString (Utilities|String|ToString(Text) (Utilities|Text|FormatText "NSLOCTEXT(\"[607CFE51407DA4236C42C390DE3FB6EF]\", \"954595F048C68EC010FA508B1BA792F8\", \"Enabled: {e}\\r\\nPlay State: {ps}\\r\\nSame Minigame: {sm}\\r\\nClass: {c}\\r\\nTeam: {t}\\r\\nNumber of teams: {nt}\\r\\nActor Team: {at}\\r\\nOwner Name: {on}\")" (Utilities|Text|ToText(Boolean) (StateChecks|IsEnabled)) (Utilities|Text|ToText(Boolean) (StateChecks|IsValidPlayState)) (Utilities|Text|ToText(Boolean) (StateChecks|IsSameMinigame Actor)) (Utilities|Text|ToText(Boolean) (StateChecks|IsValidClass Actor)) (Utilities|Text|ToText(Boolean) (StateChecks|IsValidTeam Actor)) (Utilities|Text|ToText(Integer) (Playlist|GetMaxNumberOfTeams self)) (Utilities|Text|ToText(Byte) (Team|GetActorPvPTeamIndex Actor)) (Utilities|Text|ToText(String) (Utilities|GetDisplayName (Components|GetOwner))))) true true (Math|Color|SelectColor "(R=0.000000,G=1.000000,B=0.105227,A=1.000000)" "(R=1.000000,G=0.046926,B=0.000000,A=1.000000)" (Math|Boolean|ANDBoolean (StateChecks|IsEnabled) (StateChecks|IsValidPlayState) (select (Variables|RestrictionSettings|GetInvert%20Class%20and%20Team%20Checks) (not (and (StateChecks|IsValidClass Actor) (StateChecks|IsValidTeam Actor))) (and (StateChecks|IsValidClass Actor) (StateChecks|IsValidTeam Actor))))) 1e+01))

(fn RemovePawnfromCache
  (params (Pawn :type Fort%20Player%20Pawn%20Object%20Reference%20%28by%20ref%29))
  (Utilities|Map|Remove (Variables|PlayertoPawnLink|GetOverlappingPawnControllerMap) Pawn)
  (Game|UnbindEventfromOnEndPlay Pawn (EventDispatchers|CreateEvent :!event On%20Pawn%20Destroyed)))

(fn SetClasswithAffects
  (params (Class%20Index :type Integer) (AffectsAllButSelected :type Boolean) (UsesSafeTeamIndexing :type Boolean))
  (if UsesSafeTeamIndexing
    (ClassRestrictions|SetClasswithAffects|SetClasswithAffectsSafeTeamIndex Class%20Index AffectsAllButSelected)
    (else
      (bind _output_get (Variables|ClassSettings|SetInvertClassRequirement AffectsAllButSelected))
      (Variables|ClassSettings|SetbNativeInvertClassRequirement _output_get)
      (ClassRestrictions|SetInitialSingleClass Class%20Index))))

(fn SetInitialSingleClass
  (params (Class%20Index :type Integer))
  (bind _output_get (Variables|ClassSettings|SetRequiredClass Class%20Index))
  (Variables|ClassSettings|SetNativeRequiredClass _output_get))

(fn SetInitialSingleTeam
  (params (Team%20Index :type Integer))
  (bind _output_get (Variables|TeamSettings|SetRequiredTeam (Knobs|Team|GetTeamValueasIndex Team%20Index)))
  (Variables|TeamSettings|SetNativeRequiredTeam _output_get))

(fn SetClassWithAffectsSafeTeamIndex
  (params (Class%20Index :type Integer) (AffectsAllButSelected :type Boolean))
  ;[1 0.08536 0 0.25] This function is mostly deprecated. It's for devices that use the 'Safe Team' logic from the Zone devices, with their slightly wacky way of inferring (None = Everyone, Any = Specifically any class). This shouldn't be used anymore, but is kept for backwards compat.
  (bind _output_get (Variables|ClassSettings|SetInvertClassRequirement AffectsAllButSelected))
  (Variables|ClassSettings|SetbNativeInvertClassRequirement _output_get)
  (ClassRestrictions|SetInitialSingleClass (Math|Integer|Max(Integer) Class%20Index -1)))

(fn SetNativeVariables
  (Variables|RestrictionSettings|SetbNativeRequirePlayMode (Variables|RestrictionSettings|GetRequire%20Play%20Mode))
  (Variables|ClassSettings|SetNativeRequiredClass (Variables|ClassSettings|GetRequiredClass))
  (Variables|TeamSettings|SetNativeRequiredTeam (Variables|TeamSettings|GetRequiredTeam))
  (Variables|ClassSettings|SetbNativeInvertClassRequirement (Variables|ClassSettings|GetInvertClassRequirement))
  (Variables|TeamSettings|SetbNativeInvertTeamRequirement (Variables|TeamSettings|GetInvertTeamRequirement))
  (Variables|RestrictionSettings|SetbNativeInvertClassAndTeamChecks (Variables|RestrictionSettings|GetInvert%20Class%20and%20Team%20Checks)))

(fn SetShouldUseLegacyValidationLogic
  (Variables|Internal|SetShouldUseLegacyLogic (or (not (Class|CreativeBlueprintLibrary|IsAllowingTestPlayeronCreativeDevices)) (ScalableFloat|GetValueasBool (LocalVariables|GetHF_ShouldUseLegacyValidation))))
  (return))

(fn SetTeamsWithAffects
  (params (Team%20Index :type Integer) (AffectsAllButSelected :type Boolean))
  (bind _output_get (Variables|TeamSettings|SetInvertTeamRequirement AffectsAllButSelected))
  (Variables|TeamSettings|SetbNativeInvertTeamRequirement _output_get)
  (TeamRestrictions|SetInitialSingleTeam (Math|Integer|SelectInt Team%20Index (Knobs|Team|GetShiftedAnyTeamValue) (Knobs|Team|IsValidTeamIndex Team%20Index))))

(fn StateChecks
  (params (Actor%20to%20Check :type Actor%20Object%20Reference) (AND :type Boolean :default true))
  (Debug|PrintStateChecks Actor%20to%20Check)
  (return (Math|Boolean|ANDBoolean AND (StateChecks|IsEnabled) (StateChecks|IsValidPlayState) (StateChecks|IsSameMinigame Actor%20to%20Check) (select (Variables|RestrictionSettings|GetInvert%20Class%20and%20Team%20Checks) (not (and (StateChecks|IsValidClass Actor%20to%20Check) (StateChecks|IsValidTeam Actor%20to%20Check))) (and (StateChecks|IsValidClass Actor%20to%20Check) (StateChecks|IsValidTeam Actor%20to%20Check))))))

(fn TestMatchingClass
  (params (Class%20Index :type Integer))
  (if (== (Variables|ClassSettings|GetRequiredClass) -1)
    (return (not (Variables|ClassSettings|GetInvertClassRequirement)))
    (else
      (return (!= (== Class%20Index (Variables|ClassSettings|GetRequiredClass)) (Variables|ClassSettings|GetInvertClassRequirement))))))

(fn TestMatchingTeam
  (params (TeamValue :type Integer))
  (if (Knobs|Team|IsTeamAny TeamValue)
    (return (not (Variables|TeamSettings|GetInvertTeamRequirement)))
    (else
      (return (!= (== (Knobs|Team|GetTeamValueasIndex TeamValue) (Variables|TeamSettings|GetRequiredTeam)) (Variables|TeamSettings|GetInvertTeamRequirement))))))

(fn UnbindEventPawnPossesed
  (params (PlayerPawnToUnbind :type Actor%20Object%20Reference))
  (bind _asfort_player_pawn (Utilities|Casting|CastToFortPlayerPawn PlayerPawnToUnbind))
  (Game|Player|UnbindEventfromOnPawnPossessed _asfort_player_pawn (EventDispatchers|CreateEvent :!event Empty%20Player%20Pawn%20Possessed)))
