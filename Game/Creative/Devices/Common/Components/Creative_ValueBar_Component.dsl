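; Creative device widget component that shows a floating value bar or indicator, with team, distance, text, icon and color options.
; Spawn or Update stores the settings, destroys the component where cosmetics cannot run, and otherwise async-loads the bar widget class and initializes the widget from itself.
; Bar height follows the attached mesh's bounds plus a scale-corrected offset; show/hide sets visible distance and tick.
; Helpers expose team filters and minigame-phase visibility.
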(event Custom|Spawn%20or%20Update%20Value%20Bar
  (params (Attach%20to%20Mesh :type Static%20Mesh%20Component%20Object%20Reference) (Actor :type Actor%20Object%20Reference) (Value%20Bar%20Class :type User%20Widget%20Soft%20Class%20Reference :default "/Game/Creative/Devices/Common/Components/Creative_FloatingValueBar.Creative_FloatingValueBar_C") (Value%20Bar%20Child%20Class :type User%20Widget%20Soft%20Class%20Reference) (ShowDistance :type Float%20%28double-precision%29) (Indicator%20Index :type Integer) (MaxValueManualSet :type Float%20%28double-precision%29) (Vertical%20Offset :type Float%20%28double-precision%29) (Require%20LOS :type Boolean) (Friendly%20Teams :type Array%20of%20Integers) (Hostile%20Teams :type Array%20of%20Integers) (ShowOnlyToTeamIndex :type Integer :default -1) (OnlyShowWhenNotMax :type Boolean) (TextFriendly :type String) (TextNeutral :type String) (TextHostile :type String) (TextScale :type Float%20%28double-precision%29) (Has%20Dynamic%20Value :type Boolean :default true) (DisplayDistanceText :type Boolean) (ClampProjectionToViewport :type Boolean) (DisplayOffscreenArrow :type Boolean) (Indicator%20Icon :type Texture%202D%20Soft%20Object%20Reference) (Display%20Progress%20Icon :type Boolean :default true) (Scale%20Override :type Vector%202D%20Structure :default "(X=1.000000,Y=1.000000)") (Show%20Owning%20Team :type Boolean) (Local%20Player%20Registered :type Boolean))
  ;[0 1 0.1154] Save a bunch of variables (so we never have to refer to the child object, so this object can be applied to anything.
  ; Create the Hud Icon and update if it already exists.
  (Variables|Transient|SetDisplayOwningTeam Show%20Owning%20Team)
  (Networking|CanExecuteCosmeticEvents
    (:True
      (Utilities|FlowControl|Sequence
        (:then_0
          (Variables|Internal|SetAttached%20Mesh Attach%20to%20Mesh)
          (Variables|Internal|SetActor%20Ref Actor)
          (Variables|Transient|SetVisible%20Distance ShowDistance)
          (Default|SetIndicatorIcon Indicator%20Icon)
          (Variables|Transient|SetMaxValueManualSet MaxValueManualSet)
          (Variables|Transient|SetVertical%20Offset Vertical%20Offset)
          (Variables|Transient|SetRequire%20LOS Require%20LOS)
          (Variables|Transient|SetFriendly%20Teams Friendly%20Teams)
          (Variables|Transient|SetHostile%20Teams Hostile%20Teams)
          (Variables|Transient|SetShowTeams ShowOnlyToTeamIndex)
          (Variables|Transient|SetOnlyShowWhenNotMax OnlyShowWhenNotMax)
          (Variables|Transient|SetText%20Overrides (Utilities|Array|MakeArray (Utilities|String|Trim TextFriendly) (Utilities|String|Trim TextNeutral) (Utilities|String|Trim TextHostile)))
          (Variables|Transient|SetTextScale TextScale)
          (Variables|Transient|SetbDisplayDistanceText DisplayDistanceText)
          (Variables|Transient|SetHas%20Dynamic%20Value Has%20Dynamic%20Value)
          (Variables|UserInterface|SetbClampProjectedLocalPositionToViewport ClampProjectionToViewport)
          (Variables|Transient|SetbDisplayOffscreenArrow DisplayOffscreenArrow)
          (Variables|Transient|SetbDisplayProgressIcon Display%20Progress%20Icon)
          (Variables|Transient|SetScaleOverride Scale%20Override)
          (Variables|Transient|SetLocal%20Player%20Registered Local%20Player%20Registered))
        (:then_1
          (Components|Tick|SetComponentTickEnabled true)
          (bind _valuebarwidget (Variables|Transient|GetValueBarWidget)
            (:then
              (if (Utilities|Equal(SoftClassReference) (Variables|Transient|GetCurrentChildClass) Value%20Bar%20Child%20Class)
                (Default|CallOnValueBarComponentUpdated self)
                (else
                  (Widget|RemovefromParent _valuebarwidget) ; If the container widget exists, but\nwe've received a new child widget,\nreload everything.
                )))
            (:else
              _))
          (Variables|Transient|SetCurrentChildClass Value%20Bar%20Child%20Class)
          (bind _class (Utilities|AsyncLoadClassAsset Value%20Bar%20Class)
            (:Completed
              (bind _asuser_widget (Utilities|Casting|CastToUserWidgetClass _class))
              (bind _returnvalue (UserInterface|CreateWidget _asuser_widget (Game|GetPlayerController 0)))
              (bind _output_get (Variables|Transient|SetValueBarWidget _returnvalue))
              (UserInterface|SetWidget _output_get)
              (Init|InitfromObject(Message) _returnvalue self)
              (CallFunction|UpdateValueBarHeight)
              (Default|CallOnWidgetInitialized)
              (Default|CallOnValueBarComponentUpdated self))))) ; Update user variables then spawn a health bar if there isn't one spawned already.
    )
    (:False
      (Components|DestroyComponent) ; If we're the server, we don't care about health bars.\nDestroy ourselves.
    )))

(event Custom|Set%20Custom%20Color
  (params (UseCustomColor :type Boolean) (CustomColor :type Linear%20Color%20Structure))
  ;[1 1 1] Widget Color and Font
  (Variables|Default|SetUse%20Custom%20Color UseCustomColor)
  (Variables|Default|SetCustom%20Color CustomColor)
  (Default|CallCustomColorChanged))

(event Custom|Hide%20Value%20Bar
  ;[1 0.0625 0.0625] Hiding Value Bar
  (Variables|Transient|SetVisible%20Distance 0.0)
  (Components|Tick|SetComponentTickEnabled false))

(event Custom|SetCustomFontStyle
  (params (FontStyle :type EBeaconFontStyle%20Enum :default "NewEnumerator0"))
  ;[1 1 1] Widget Color and Font
  (Variables|Transient|SetFontStyle FontStyle))

(event Custom|ShowValueBar
  (params (DistanceToShow :type Float%20%28double-precision%29))
  ;[1 0.0625 0.0625] Showing Value Bar
  (Variables|Transient|SetVisible%20Distance DistanceToShow)
  (Components|Tick|SetComponentTickEnabled true))

(fn GetIndicator
  (return (not (Utilities|IsValidSoftObjectReference (Variables|Internal|GetIndicator%20Texture))) (Variables|Internal|GetIndicator%20Label) (Variables|Internal|GetIndicator%20Texture)))

(fn GetShowTeamAsIndex
  (return (Knobs|Team|GetTeamValueasIndex (Variables|Transient|GetShowTeams))))

(fn IsLocalPlayerRegistered
  (return (Variables|Transient|GetLocal%20Player%20Registered)))

(fn IsShowingValueBarAllowed
  (bind _returnvalue (Components|GetOwner))
  (bind _returnvalue_2 (Minigame|GetMinigame _returnvalue :!class /Script/FortniteGame.MinigameBlueprintLibrary))
  (Utilities|IsValid _returnvalue_2
    (:Is%20Valid
      (return (Utilities|Set|ContainsItem (Variables|Default|GetShowDuringPhases self) (Variables|/Script/FortniteGame.FortMinigame|GetCurrentState _returnvalue_2))))
    (:Is%20Not%20Valid
      (return false))))

(fn IsShowTeamAny
  (return (Knobs|Team|IsTeamAny (Variables|Transient|GetShowTeams))))

(fn IsShowTeamFriendlies
  (return (Knobs|Team|IsTeamFriendlies (Variables|Transient|GetShowTeams))))

(fn IsShowTeamHostiles
  (return (Knobs|Team|IsTeamHostiles (Variables|Transient|GetShowTeams))))

(fn IsShowTeamNeutral
  (return (Knobs|Team|IsTeamNeutral (Variables|Transient|GetShowTeams))))

(fn IsShowTeamValid
  (return (Knobs|Team|IsValidTeamIndex (Variables|Transient|GetShowTeams))))

(fn SetHasValue
  (params (HasValue :type Boolean))
  (bind _ascreative_floating_value_bar (Utilities|Casting|CastToCreative_FloatingValueBar (Variables|Transient|GetValueBarWidget)))
  (Class|CreativeFloatingValueBar|ShowValue _ascreative_floating_value_bar HasValue))

(fn SetIndicatorIcon
  (params (Icon%20Texture :type Texture%202D%20Soft%20Object%20Reference))
  ;[1 1 1] While using the new indicator logic:
  ; - Apply the texture from valid icon picker entry to the Indicator Texture
  ; - If not, assign an empty texture insteasd.
  (Variables|Internal|SetIndicator%20Texture (select (Utilities|IsValidSoftObjectReference Icon%20Texture) Icon%20Texture "/Game/Athena/UI/InGame/Creative/PropertyEditor/Icon_Picker/Textures/T_UI_IconPicker_EMPTY.T_UI_IconPicker_EMPTY")))

(fn SetIndicatorLabel
  (params (Index :type Integer) (Icon%20Texture :type Texture%202D%20Soft%20Object%20Reference))
  (bind _indicator_labels (Variables|Default|GetIndicator%20Labels))
  (bind _returnvalue (Utilities|Array|Length _indicator_labels))
  (bind _index (Variables|GetIndex))
  (bind _returnvalue_2 (- _index _returnvalue))
  (bind _indicator_labels_2 (Variables|Default|GetIndicator%20Labels))
  (bind _indicator_icons (Variables|Default|GetIndicator%20Icons))
  ;[1 1 1] Index can either be a text label or an icon, based on option chosen in data asset.
  (if (Utilities|Array|IsValidIndex _indicator_labels_2 Index)
    (Variables|Internal|SetIndicator%20Label (Utilities|Array|Get(acopy) _indicator_labels_2 Index))
    (Variables|Internal|SetIndicator%20Texture "")
    (else
      (Variables|Internal|SetIndicator%20Label "")
      (if (Utilities|Array|IsValidIndex _indicator_icons _returnvalue_2)
        (Variables|Internal|SetIndicator%20Texture (Utilities|Array|Get(acopy) _indicator_icons _returnvalue_2))
        (else
          (Variables|Internal|SetIndicator%20Texture ""))))))

(fn UpdateValueBarHeight
  ;[0 0 0] ValueBar, update content and visibility
  (bind _attached_mesh (Variables|Internal|GetAttached%20Mesh)
    (:then
      (Utilities|IsValid (Variables|/Script/Engine.StaticMeshComponent|GetStaticMesh _attached_mesh)
        (:Is%20Valid
          (Transformation|SetRelativeLocation 0 (Math|Vector|MakeVector 0.0 0.0 (+ (/ (Variables|Transient|GetVertical%20Offset) (.z (.scale (Transformation|GetWorldTransform _attached_mesh)))) (.z (.Max (Utilities|Struct|BreakBox (StaticMesh|GetBoundingBox (Variables|/Script/Engine.StaticMeshComponent|GetStaticMesh _attached_mesh))))))) false false))))
    (:else)))
