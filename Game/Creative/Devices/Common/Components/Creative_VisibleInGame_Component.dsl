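; Creative device component that shows or hides its owner or listed scene components by minigame phase, editor state and a forced in-game override.
; Updates also drive collision, audio, particle and skeletal mesh activation, and optional foliage refresh.
; Authority sets the phase and override, whose OnRep functions refresh visibility; the override resets at PostGameReset.
; Minigame event binding falls back from direct lookup to the owner's logic-component delegate, then a retry timer.
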(event Custom|Initialize%20Component
  ;[0.03244 1 0.2556 0.2] Construction - Update Visibility based on Settings and attempt to bind to the minigame so we can fire events on Game State changes.
  (Visibility|UpdateVisibility)
  (CallFunction|AttemptBindMinigameEvents))

(event EventBeginPlay
  ;[0.03244 1 0.2556 0.2] Construction - Update Visibility based on Settings and attempt to bind to the minigame so we can fire events on Game State changes.
  (|CallParentFunction)
  (Visibility|UpdateVisibility)
  (CallFunction|AttemptBindMinigameEvents))

(event Custom|OnPlayModeChangedDelegate
  (params (Minigame :type Fort%20Minigame%20Object%20Reference) (bIsInPlayMode :type Boolean))
  ;[0.2086 0.3763 1 0.25] Game State Changes.
  (Visibility|UpdateVisibility))

(event Custom|Attempt%20Bind%20Minigame%20Events
  (bind _returnvalue (Components|GetOwner))
  (bind _returnvalue_2 (Actor|GetComponentbyClass _returnvalue "/Script/FortniteGame.FortMinigameLogicComponent"))
  ;[0 0.3766 1 0.078] Binding to the Minigame, and falling back if it cannot find one.
  ; This is a multi-stage issue. It'll attempt to bind from the Owner Minigame. If that doesn't exist, it'll grab the owner's MinigameProgress component and listen to the OnAddedToMinigame delegate.
  ; If this fails or the component doesn't exist, straight up use a timer.
  (Utilities|Time|ClearandInvalidateTimerbyHandle (Variables|Default|GetRetryBindMiniGameEventsTimer))
  (bind _bind_success (CallFunction|UpdateMinigameEventBinds true (Minigame|GetMinigame (Components|GetOwner) :!class /Script/FortniteGame.MinigameBlueprintLibrary)))
  (if _bind_success
    (Visibility|UpdateVisibility)
    (else
      ;;[0.07286 0.5762 1] Can't Find Minigame yet. Grab the owners Minigame Logic Component and listen for being added to the minigame
      (Utilities|IsValid _returnvalue_2
        (:Is%20Valid
          (Minigame|BindEventtoOnAddedtoMinigame _returnvalue_2 (delegate Custom|OnOwnerAddedToMinigame)))
        (:Is%20Not%20Valid
          ;;[0.5 0 0] Major fallback - trying to avoid this one if possible. This is used if minigame fails and owner doesn't have a logic component.
          (bind _returnvalue_3 (Utilities|Time|SetTimerbyEvent (delegate Custom|Retry%20Bind%20Events%20on%20Timer) 0.5 true))
          (Variables|Default|SetRetryBindMiniGameEventsTimer _returnvalue_3))))))

(event Custom|OnMinigameStateChanged
  (params (Minigame :type Fort%20Minigame%20Object%20Reference) (MinigameState :type EFortMinigameState%20Enum :default "PreGame"))
  ;[0.2086 0.3763 1 0.25] Game State Changes.
  (if (Utilities|Enum|Equal(Enum) MinigameState "PostGameReset")
    (Variables|Visibility|SetForcedVisibilityDuringGame "NewEnumerator0")))

(event EventEndPlay
  (params (EndPlayReason :type EEndPlayReason%20Enum :default "Destroyed"))
  ;[0.9301 0.1119 0.1095 0.2] Destruction - Clear Minigame Binds.
  (|CallParentFunction EndPlayReason)
  (CallFunction|UpdateMinigameEventBinds false))

(event Custom|OnOwnerAddedToMinigame
  (params (Minigame :type Fort%20Minigame%20Object%20Reference))
  ;[0 0.3766 1 0.078] Binding to the Minigame, and falling back if it cannot find one.
  ; This is a multi-stage issue. It'll attempt to bind from the Owner Minigame. If that doesn't exist, it'll grab the owner's MinigameProgress component and listen to the OnAddedToMinigame delegate.
  ; If this fails or the component doesn't exist, straight up use a timer.
  ;;[0.07286 0.5762 1] Can't Find Minigame yet. Grab the owners Minigame Logic Component and listen for being added to the minigame
  (Minigame|UnbindEventfromOnAddedtoMinigame (Actor|GetComponentbyClass (Components|GetOwner) "/Script/FortniteGame.FortMinigameLogicComponent") (delegate Custom|OnOwnerAddedToMinigame))
  (bind _bind_success (CallFunction|UpdateMinigameEventBinds true Minigame))
  (if _bind_success
    (Visibility|UpdateVisibility)
    (else
      ;;[0.5 0 0] Major fallback - trying to avoid this one if possible. This is used if minigame fails and owner doesn't have a logic component.
      (bind _returnvalue (Utilities|Time|SetTimerbyEvent (delegate Custom|Retry%20Bind%20Events%20on%20Timer) 0.5 true))
      (Variables|Default|SetRetryBindMiniGameEventsTimer _returnvalue))))

(event Custom|Retry%20Bind%20Events%20on%20Timer
  ;[0 0.3766 1 0.078] Binding to the Minigame, and falling back if it cannot find one.
  ; This is a multi-stage issue. It'll attempt to bind from the Owner Minigame. If that doesn't exist, it'll grab the owner's MinigameProgress component and listen to the OnAddedToMinigame delegate.
  ; If this fails or the component doesn't exist, straight up use a timer.
  (bind _bind_success (CallFunction|UpdateMinigameEventBinds true (Minigame|GetMinigame (Components|GetOwner) :!class /Script/FortniteGame.MinigameBlueprintLibrary)))
  (if _bind_success
    (Visibility|UpdateVisibility)
    (Utilities|Time|ClearandInvalidateTimerbyHandle (Variables|Default|GetRetryBindMiniGameEventsTimer))))

(fn GetVisibilityComponents
  (LocalVariables|SetLocal%20Combined%20Array (Variables|Visibility|GetVisibleSceneComponents))
  (bind _keys (Utilities|Map|Keys (Variables|PrimitiveComponent|GetVisibility%20Components%20with%20Collision)))
  (Utilities|Array|AppendArray (LocalVariables|GetLocal%20Combined%20Array) _keys)
  (return (LocalVariables|GetLocal%20Combined%20Array)))

(fn InitializewithSceneComponents
  (params (Visibility%20Components :type Array%20of%20Scene%20Component%20Object%20References) (Components%20with%20Collision :type Map%20of%20Scene%20Component%20Object%20References%20to%20ECollisionEnabled%20Enums) (Visible%20During%20Phase :type EVisibleInGameState%20Enum :default "NewEnumerator4"))
  (Variables|Visibility|SetVisibleSceneComponents Visibility%20Components)
  (Variables|PrimitiveComponent|SetVisibility%20Components%20with%20Collision Components%20with%20Collision)
  (if (Utilities|Enum|NotEqual(Enum) Visible%20During%20Phase "NewEnumerator4")
    (bind _visibility_changed (Visibility|SetVisiblityDuringPhase Visible%20During%20Phase))
    (if (not _visibility_changed)
      (Visibility|UpdateVisibility))
    (else
      (CallFunction|InitializeComponent) ; Even if we're not setting values, we'll need to initialize anyway.\n(no need to call this from the true branch, it'll happen in that flow)
    )))

(fn IsInPlayMode
  (bind _returnvalue (Components|GetOwner))
  (bind _returnvalue_2 (Minigame|GetMinigame _returnvalue :!class /Script/FortniteGame.MinigameBlueprintLibrary))
  (Utilities|IsValid _returnvalue
    (:Is%20Valid
      (Utilities|IsValid _returnvalue_2
        (:Is%20Valid
          (return (or (Minigame|IsinPlayMode _returnvalue_2 :!class /Script/FortniteGame.FortMinigame) (Utilities|Enum|Equal(Enum) (Variables|/Script/FortniteGame.FortMinigame|GetCurrentState _returnvalue_2) "Transitioning"))))
        (:Is%20Not%20Valid
          _)))
    (:Is%20Not%20Valid
      _))
  (return (Minigame|IsinPlayMode _returnvalue :!class /Script/FortniteGame.MinigameBlueprintLibrary)))

(fn OnRep_ForcedVisibilityDuringGame
  (Visibility|UpdateVisibility))

(fn OnRep_VisibleDuringPhase
  (Visibility|UpdateVisibility))

(fn SetVisibilityComponents
  (params (Scene%20Components :type Array%20of%20Scene%20Component%20Object%20References) (Update%20Visibility :type Boolean :default true))
  (Variables|Visibility|SetVisibleSceneComponents Scene%20Components)
  (if Update%20Visibility
    (Visibility|UpdateVisibility)))

(fn SetVisibilityComponentswithCollision
  (params (Collision%20Overrides :type Map%20of%20Scene%20Component%20Object%20References%20to%20ECollisionEnabled%20Enums) (Update%20Visibility :type Boolean :default true))
  (Variables|PrimitiveComponent|SetVisibility%20Components%20with%20Collision Collision%20Overrides)
  (bind _keys (Utilities|Map|Keys Collision%20Overrides))
  (if (Math|Conversions|ToBoolean(Integer) (Utilities|Array|Length _keys))
    (if Update%20Visibility
      (Visibility|UpdateArrayVisibility))))

(fn SetVisiblityDuringPhase
  (params (New%20Visible%20During%20Phase :type EVisibleInGameState%20Enum :default "NewEnumerator0"))
  (Utilities|SwitchHasAuthority
    (:Authority
      (if (and (Utilities|Enum|NotEqual(Enum) (Variables|GetNew%20Visible%20During%20Phase) "NewEnumerator4") (Math|Byte|NotEqual(Byte) (Variables|GetNew%20Visible%20During%20Phase) (Variables|Visibility|GetVisibleDuringPhase)))
        (Variables|Visibility|SetVisibleDuringPhase (Variables|GetNew%20Visible%20During%20Phase))
        (return true)
        (else
          _)))
    (:Remote
      _))
  (return false))

(fn SetForcedVisibilityDuringGame
  (params (NewForcedVisibilityDuringGame :type Enum_ForcedValueDuringGame%20Enum :default "NewEnumerator0"))
  (Utilities|SwitchHasAuthority
    (:Authority
      (if (Utilities|Enum|NotEqual(Enum) NewForcedVisibilityDuringGame (Variables|Visibility|GetForcedVisibilityDuringGame))
        (Networking|FlushNetDormancy (Components|GetOwner)) ; Need else it sometimes doesn't replicate to client
        (Variables|Visibility|SetForcedVisibilityDuringGame (Variables|GetNewForcedVisibilityDuringGame))))))

(fn ShouldBeVisible
  (bind _returnvalue (Editor|InEditor))
  (bind _returnvalue_2 (CallFunction|IsinPlayMode))
  (bind _returnvalue_3 (not _returnvalue_2))
  (bind _visibleduringphase (Variables|Visibility|GetVisibleDuringPhase))
  (bind _returnvalue_4 (Utilities|Enum|Equal(Enum) _visibleduringphase "NewEnumerator2"))
  (bind _returnvalue_5 (and _returnvalue_4 _returnvalue_2))
  (bind _returnvalue_6 (Utilities|Enum|Equal(Enum) _visibleduringphase "NewEnumerator1"))
  (bind _returnvalue_7 (and _returnvalue_3 _returnvalue_6))
  (bind _returnvalue_8 (Utilities|Enum|Equal(Enum) _visibleduringphase "NewEnumerator3"))
  (bind _returnvalue_9 (Math|Boolean|ORBoolean _returnvalue_8 _returnvalue_5 _returnvalue_7))
  (bind _visibleduringphase_2 (Variables|Visibility|GetVisibleDuringPhase))
  (bind _returnvalue_10 (Utilities|Enum|Equal(Enum) _visibleduringphase_2 "NewEnumerator3"))
  (bind _returnvalue_11 (Utilities|Enum|Equal(Enum) _visibleduringphase_2 "NewEnumerator2"))
  (bind _returnvalue_12 (or _returnvalue_11 _returnvalue_10))
  (bind _returnvalue_13 (select _returnvalue _returnvalue_12 _returnvalue_9))
  ;[1 1 1] We don't care about forced visibility during game if we aren't in a game.
  (if (not (CallFunction|IsinPlayMode))
    (return _returnvalue_13)
    (else
      (return (Utilities|Select :Index (Variables|Visibility|GetForcedVisibilityDuringGame) :NewEnumerator0 _returnvalue_13 :NewEnumerator1 true :NewEnumerator2 false)))))

(fn ShouldBeVisible(NonEditor)
  ;[1 0.5171 0] Whether the component should be visible (not including any 'Hidden but visible' alterations in the editor, which count as 'not visible' normally.
  (return (Utilities|Select :Index (Variables|Visibility|GetForcedVisibilityDuringGame) :NewEnumerator0 (Math|Boolean|ORBoolean (Utilities|Enum|Equal(Enum) (Variables|Visibility|GetVisibleDuringPhase) "NewEnumerator3") (and (Utilities|Enum|Equal(Enum) (Variables|Visibility|GetVisibleDuringPhase) "NewEnumerator2") (CallFunction|IsinPlayMode)) (and (not (CallFunction|IsinPlayMode)) (Utilities|Enum|Equal(Enum) (Variables|Visibility|GetVisibleDuringPhase) "NewEnumerator1"))) :NewEnumerator1 true :NewEnumerator2 false)))

(fn ShouldUpdateActivationBasedOnVisibility
  (params (Component :type Actor%20Component%20Object%20Reference))
  (Utilities|Casting|CastToParticleSystemComponent Component
    (:then
      _)
    (:CastFailed
      (Utilities|Casting|CastToSkeletalMeshComponent Component
        (:then
          _)
        (:CastFailed
          (return false)))))
  (return true))

(fn UpdateArrayVisibility
  (bind _output_get (LocalVariables|SetUpdate%20Array (Visibility|GetVisibilityComponents)))
  (if (Math|Conversions|ToBoolean(Integer) (Utilities|Array|Length _output_get))
    (LocalVariables|SetCacheVisible (Visibility|ShouldbeVisible))
    (for _array_element _output_get
      (Visibility|UpdateComponentVisibility _array_element (LocalVariables|GetCacheVisible)))
    (CallFunction|UpdateFoliage)
    (Visibility|CallOnVisibilityChanged (Visibility|ShouldbeVisible(NonEditor)))))

(fn UpdateAudioComponent
  (params (Audio%20Component :type Audio%20Component%20Object%20Reference) (Visible :type Boolean))
  (bind _audio_component (Variables|GetAudio%20Component))
  (Utilities|IsValid Audio%20Component
    (:Is%20Valid
      (if Visible
        (Components|Activation|Activate _audio_component (Variables|ActorComponent|GetbResetOnComponentActivation))
        (else
          (Components|Activation|Deactivate _audio_component))))))

(fn UpdateComponentVisibility
  (params (Component :type Scene%20Component%20Object%20Reference) (Visible :type Boolean))
  (bind _component (Variables|GetComponent))
  (Utilities|IsValid Component
    (:Is%20Valid
      (if (not (Variables|Visibility|GetUseSimplifiedActorVisibilityAndCollision))
        (Development|SetHiddeninGame (Variables|GetComponent) (not (Variables|GetVisible)) false)
        (Rendering|SetVisibility (Variables|GetComponent) (and (Utilities|Enum|NotEqual(Enum) (Variables|Visibility|GetVisibleDuringPhase) "NewEnumerator0") (or (Variables|GetVisible) (Editor|InEditor))) false) ; If the option should be always invisible, we hide the visibility straight up.\nThis allows optional components that have been disabled to hide correctly\nin UEFN. Note: This doesn't cover the 'Game Only' case in editor, but I don't think \nwe actually use that much (and there's no way to accurately cover that, as there's no\n'Hidden in Editor but not Game' variable.
        (else
          _))
      (Utilities|FlowControl|Sequence
        (:then_0
          (bind _asprimitive_component (Utilities|Casting|CastToPrimitiveComponent (Variables|GetComponent)))
          (if (and (Variables|PrimitiveComponent|GetbEnableCollisionUpdates) (not (Variables|Visibility|GetUseSimplifiedActorVisibilityAndCollision)))
            (Visibility|UpdatePrimitiveComponentCollision _asprimitive_component (Variables|GetVisible))))
        (:then_1
          (bind _asaudio_component (Utilities|Casting|CastToAudioComponent (Variables|GetComponent)))
          (Visibility|UpdateAudioComponent _asaudio_component (Variables|GetVisible)))
        (:then_2
          (bind _shouldupdateactivation (Visibility|ShouldUpdateActivationBasedonVisibility _component))
          (if _shouldupdateactivation
            (Visibility|UpdateActivationBasedonVisibility _component (Variables|GetVisible))))))))

(fn UpdateFoliage
  (Networking|CanExecuteCosmeticEvents
    (:True
      (if (Variables|Visibility|GetUpdate%20Foliage%20on%20Visibility%20Change)
        (bind _asbuilding_actor (Utilities|Casting|CastToBuildingActor (Components|GetOwner)))
        (Building|RefreshRemovedFoliageAroundBounds _asbuilding_actor false)))))

(fn UpdateMinigameEventBinds
  (params (Bind :type Boolean) (Minigame :type Fort%20Minigame%20Object%20Reference))
  (bind _outputdelegate (EventDispatchers|CreateEvent :!event OnPlayModeChangedDelegate))
  (bind _outputdelegate_2 (EventDispatchers|CreateEvent :!event OnMinigameStateChanged))
  (bind _local_minigame (LocalVariables|Getlocal%20Minigame))
  (LocalVariables|Setlocal%20Minigame Minigame)
  (Utilities|FlowControl|Sequence
    (:then_0
      ;[0 0.7473 1] Unbind Events on Cached Minigame if different than current (or if told to unbind)
      (bind _cachedminigame (Variables|Default|GetCachedMinigame)
        (:then
          (Utilities|IsValid (Utilities|ResolveSoftReference _cachedminigame)
            (:Is%20Valid
              (LocalVariables|Getlocal%20Minigame
                (:then
                  (if (or (!= (Utilities|ResolveSoftReference (LocalVariables|Getlocal%20Minigame)) (Variables|Default|GetCachedMinigame)) (not (Variables|GetBind)))
                    _))
                (:else
                  _))
              (Minigame|UnbindEventfromOnPlayModeChangedDelegate (Utilities|ResolveSoftReference _cachedminigame) _outputdelegate)
              (Minigame|UnbindEventfromStateChangedDelegate (Utilities|ResolveSoftReference _cachedminigame) _outputdelegate_2))))
        (:else)))
    (:then_1
      (if (Variables|GetBind)
        (bind _local_minigame_2 (LocalVariables|Getlocal%20Minigame)
          (:then
            (Variables|Default|SetCachedMinigame (Utilities|ResolveSoftReference _local_minigame_2))
            (Minigame|BindEventtoOnPlayModeChangedDelegate _local_minigame _outputdelegate)
            (Minigame|BindEventtoStateChangedDelegate _local_minigame _outputdelegate_2)
            (return true))
          (:else
            (Variables|Default|SetCachedMinigame 0)
            (return false)))))))

(fn UpdatePrimitiveComponentCollision
  (params (Primitive%20Component :type Primitive%20Component%20Object%20Reference) (Visible :type Boolean))
  (bind (_value _returnvalue) (Utilities|Map|Find (Variables|PrimitiveComponent|GetVisibility%20Components%20with%20Collision) Primitive%20Component))
  (Utilities|IsValid Primitive%20Component
    (:Is%20Valid
      (Collision|SetCollisionEnabled Primitive%20Component (select (Variables|GetVisible) (select _returnvalue _value (Variables|PrimitiveComponent|GetDefault%20Collision%20Enabled%20State)) (Variables|PrimitiveComponent|GetCollision%20Disabled%20State))))))

(fn UpdateVisibility
  (if (Creative|UEFN|IsUEFNEditModeTransitionPending (Components|GetOwner))
    _
    (elif (Variables|Visibility|GetUseSimplifiedActorVisibilityAndCollision)
      (Visibility|UpdateActorVisibility)
      (Visibility|UpdateArrayVisibility)
      (else
        _))))

(fn UpdateActivationBasedOnVisibility
  (params (Component :type Actor%20Component%20Object%20Reference) (Visible :type Boolean))
  (if Visible
    (Components|Activation|Activate Component (Variables|ActorComponent|GetbResetOnComponentActivation))
    (else
      (Components|Activation|Deactivate Component))))

(fn UpdateActorVisibility
  (bind _returnvalue (Components|GetOwner))
  (bind _visible (Visibility|ShouldbeVisible))
  (Rendering|SetActorHiddenInGame _returnvalue (not _visible))
  (if (Variables|PrimitiveComponent|GetbEnableCollisionUpdates)
    (Collision|SetActorEnableCollision _returnvalue _visible)))
