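; Creative device component that stores wood, stone, metal and gold costs and charges them to a player.
; PassesCostCheck compares inventory counts against the cached cost map, failing at the first shortfall with an optional failed-action response.
; On success it also removes the items through AttemptConsumeCosts; infinite-resource players skip both check and charge.
; SetPickupCosts copies the cached costs onto Creative pickups.
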(fn AttemptConsumeCosts
  (params (FortPC :type Fort%20Player%20Controller%20Object%20Reference))
  (bind _cached_costs (Variables|Default|GetCached%20Costs))
  (bind _keys (Utilities|Map|Keys (Variables|Default|GetCached%20Costs)))
  (for _array_element _keys
    (if (Math|Integer|InRange(Integer) (Utilities|Map|Find _cached_costs _array_element) 0 10000 false)
      (bind _asfort_world_item_definition (Utilities|Casting|CastToFortWorldItemDefinition (Utilities|Map|Find (Variables|Default|GetResourceToIDMap) _array_element)))
      (Inventory|RemoveItemFromPlayer :PlayerController (Variables|GetFortPC) :ItemDefinition _asfort_world_item_definition :AmountToRemove (Utilities|Map|Find _cached_costs _array_element) :bForceRemoval false :bSkipVerbRemoval false)))
  (return))

(fn GetCost
  (return (Variables|Default|GetCached%20Costs)))

(fn PassesCostCheck
  (params (Check%20Pawn :type Pawn%20Object%20Reference) (bPlayMontageOnFail :type Boolean))
  ;[0 0.1252 0.15 0.5] Checks to see if the player controller has the resources listed in cached costs.
  ; If all checks pass, return success, else early out as soon as we fail a check.
  (bind _asfort_game_state_athena (Utilities|Casting|CastToFortGameStateAthena (Game|GetGameState)))
  (LocalVariables|SetCachedFortPC (Fortnite|GetFortPlayerControllerfromActor Check%20Pawn))
  (if (not (Class|FortGameStateAthena|ShouldHaveInfiniteResources _asfort_game_state_athena Check%20Pawn "None"))
    (bind _keys (Utilities|Map|Keys (Variables|Default|GetCached%20Costs)))
    (for _array_element _keys
      (bind _returnvalue (InventoryManagement|GetItemCountforItemDefinition (LocalVariables|GetCachedFortPC) (Utilities|Map|Find (Variables|Default|GetResourceToIDMap) _array_element)))
      (if (< _returnvalue (Utilities|Map|Find (Variables|Default|GetCached%20Costs) _array_element))
        (if bPlayMontageOnFail
          (bind _asfort_player_pawn (Utilities|Casting|CastToFortPlayerPawn (Variables|GetCheck%20Pawn))
            (:then
              (FailedAction|UnabletoPerformAction _asfort_player_pawn true (GameplayTags|MakeGameplayTagContainerfromTag "(TagName=\"Activate.Failed.Cost\")")) ; Needs to be called on client to work.
            )
            (:CastFailed
              _))
          (return false)
          (else
            _))))
    (Default|AttemptConsumeCosts (LocalVariables|GetCachedFortPC))
    (else
      _))
  (return true))

(fn SetPickupCosts
  (params (Pickup :type Fort%20Pickup%20Object%20Reference))
  ;[0 0.15 0.1346 0.5] Sets the pickup's costs to the values defined by this component, if the pickup is a creative pickup (using Fort Pickup type as an input as the SpawnPickupInWorld function always outputs that type).
  (bind _asfort_pickup_creative (Utilities|Casting|CastToFort_Pickup_Creative Pickup))
  (bind _keys (Utilities|Map|Keys (Variables|Default|GetCached%20Costs)))
  (for _array_element _keys
    (Class|FortPickupCreative|SetItemResourceCost _asfort_pickup_creative _array_element (Utilities|Map|Find (Variables|Default|GetCached%20Costs) _array_element))))

(fn SetCosts
  (params (Wood :type Integer) (Stone :type Integer) (Metal :type Integer) (Gold :type Integer))
  (Variables|Default|SetCached%20Costs (Utilities|Map|MakeMap "Wood" Wood "Stone" Stone "Metal" Metal "GoldCurrency" Gold)))
